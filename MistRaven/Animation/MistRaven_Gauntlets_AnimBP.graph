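# Animation Blueprint for the MistRaven gauntlets, which follows the attached parent's pose and adds secondary motion to the feathers.
# The copied pose and curves pass through DefaultSlot into a gauntlet Control Rig, then through spring-driven, root-relative Anim Dynamics nodes, one per left and right feather bone (05–10).
# On character customization load, a custom part's gender and body type set the Medium Male, Medium Female and Large Male flags fed to the rig.
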AnimBlueprint /MistRaven/Animation/MistRaven_Gauntlets_AnimBP
target_skeleton=/MistRaven/Assets/Mesh/Flamethrower_MistRaven_Gauntlet_Skeleton
graphs [
 AnimGraph {nodes [
   n1 = Control Rig!(
    Source=Slot 'DefaultSlot'(Source=Copy Pose From Mesh() {bCopyCurves=true, bUseAttachedParent=true})
    Alpha
    Male_Med:bool=Get Medium Male() -> (Medium Male:bool)
    Female_Med:bool=Get Medium Female() -> (Medium Female:bool)
    Male_Lrg:bool=Get Large Male() -> (Large Male:bool)
   ) -> (Pose) {
    alphaBoolBlend {blendOption=Linear}
    controlRigAssetReference {blueprintRigClass=/MistRaven/Assets/Mesh/Flamethrower_MistRaven_Gauntlet_CtrlRig.Flamethrower_MistRaven_Gauntlet_CtrlRig_C}
   }
   n2 = Anim Dynamics - Bone: dyn_feather_l_10(ComponentPose=Local To Component(LocalPose=@n1), Alpha=1, ExternalForce="0,0,0") {
    angularDampingOverride=1
    angularSpringConstant=8000
    bAngularSpring=true
    bOverrideAngularDamping=true
    bUseGravityOverride=true
    gravityScale=0.35
    simulationSpace=RootRelative
    alphaBoolBlend {blendOption=Linear}
    boundBone {boneName=dyn_feather_l_10}
    componentLinearAccScale=<1 1 1>
    componentLinearVelScale=<1 1 1>
    gravityOverride {z=-980}
    physicsBodyDefinitions [{
      boundBone {boneName=dyn_feather_l_10}
      boxExtents=<28 28 28>
      constraintSetup {angularLimitsMax=<70 30 30>, angularLimitsMin=<-60 -30 -30>, angularTarget=<1 0 0>}
      localJointOffset=<-3 0 0>
     }]
   }
   n3 = Anim Dynamics - Bone: dyn_feather_l_09(ComponentPose=@n2, Alpha=1, ExternalForce="0,0,0") {
    angularDampingOverride=1
    angularSpringConstant=8000
    bAngularSpring=true
    bOverrideAngularDamping=true
    bUseGravityOverride=true
    gravityScale=0.35
    simulationSpace=RootRelative
    alphaBoolBlend {blendOption=Linear}
    boundBone {boneName=dyn_feather_l_09}
    componentLinearAccScale=<1 1 1>
    componentLinearVelScale=<1 1 1>
    gravityOverride {z=-980}
    physicsBodyDefinitions [{
      boundBone {boneName=dyn_feather_l_09}
      boxExtents=<28 28 28>
      constraintSetup {angularLimitsMax=<70 30 30>, angularLimitsMin=<-60 -30 -30>, angularTarget=<1 0 0>}
      localJointOffset=<-3 0 0>
     }]
   }
   n4 = Anim Dynamics - Bone: dyn_feather_l_08(ComponentPose=@n3, Alpha=1, ExternalForce="0,0,0") {
    angularDampingOverride=1
    angularSpringConstant=8000
    bAngularSpring=true
    bOverrideAngularDamping=true
    bUseGravityOverride=true
    gravityScale=0.35
    simulationSpace=RootRelative
    alphaBoolBlend {blendOption=Linear}
    boundBone {boneName=dyn_feather_l_08}
    componentLinearAccScale=<1 1 1>
    componentLinearVelScale=<1 1 1>
    gravityOverride {z=-980}
    physicsBodyDefinitions [{
      boundBone {boneName=dyn_feather_l_08}
      boxExtents=<28 28 28>
      constraintSetup {angularLimitsMax=<70 30 30>, angularLimitsMin=<-60 -30 -30>, angularTarget=<1 0 0>}
      localJointOffset=<-3 0 0>
     }]
   }
   n5 = Anim Dynamics - Bone: dyn_feather_l_07(ComponentPose=@n4, Alpha=1, ExternalForce="0,0,0") {
    angularDampingOverride=1
    angularSpringConstant=8000
    bAngularSpring=true
    bOverrideAngularDamping=true
    bUseGravityOverride=true
    gravityScale=0.35
    simulationSpace=RootRelative
    alphaBoolBlend {blendOption=Linear}
    boundBone {boneName=dyn_feather_l_07}
    componentLinearAccScale=<1 1 1>
    componentLinearVelScale=<1 1 1>
    gravityOverride {z=-980}
    physicsBodyDefinitions [{
      boundBone {boneName=dyn_feather_l_07}
      boxExtents=<28 28 28>
      constraintSetup {angularLimitsMax=<70 30 30>, angularLimitsMin=<-60 -30 -30>, angularTarget=<1 0 0>}
      localJointOffset=<-3 0 0>
     }]
   }
   n6 = Anim Dynamics - Bone: dyn_feather_l_06(ComponentPose=@n5, Alpha=1, ExternalForce="0,0,0") {
    angularDampingOverride=1
    angularSpringConstant=8000
    bAngularSpring=true
    bOverrideAngularDamping=true
    bUseGravityOverride=true
    gravityScale=0.35
    simulationSpace=RootRelative
    alphaBoolBlend {blendOption=Linear}
    boundBone {boneName=dyn_feather_l_06}
    componentLinearAccScale=<1 1 1>
    componentLinearVelScale=<1 1 1>
    gravityOverride {z=-980}
    physicsBodyDefinitions [{
      boundBone {boneName=dyn_feather_l_06}
      boxExtents=<28 28 28>
      constraintSetup {angularLimitsMax=<70 30 30>, angularLimitsMin=<-60 -30 -30>, angularTarget=<1 0 0>}
      localJointOffset=<-3 0 0>
     }]
   }
   n7 = Anim Dynamics - Bone: dyn_feather_l_05(ComponentPose=@n6, Alpha=1, ExternalForce="0,0,0") {
    angularDampingOverride=1
    angularSpringConstant=8000
    bAngularSpring=true
    bOverrideAngularDamping=true
    bUseGravityOverride=true
    gravityScale=0.35
    simulationSpace=RootRelative
    alphaBoolBlend {blendOption=Linear}
    boundBone {boneName=dyn_feather_l_05}
    componentLinearAccScale=<1 1 1>
    componentLinearVelScale=<1 1 1>
    gravityOverride {z=-980}
    physicsBodyDefinitions [{
      boundBone {boneName=dyn_feather_l_05}
      boxExtents=<28 28 28>
      constraintSetup {angularLimitsMax=<70 30 30>, angularLimitsMin=<-60 -30 -30>, angularTarget=<1 0 0>}
      localJointOffset=<-3 0 0>
     }]
   }
   n8 = Anim Dynamics - Bone: dyn_feather_r_10(ComponentPose=@n7, Alpha=1, ExternalForce="0,0,0") {
    angularDampingOverride=1
    angularSpringConstant=7000
    bAngularSpring=true
    bOverrideAngularDamping=true
    bUseGravityOverride=true
    gravityScale=0.35
    simulationSpace=RootRelative
    alphaBoolBlend {blendOption=Linear}
    boundBone {boneName=dyn_feather_r_10}
    componentLinearAccScale=<1 1 1>
    componentLinearVelScale=<1 1 1>
    gravityOverride {z=-980}
    physicsBodyDefinitions [{
      boundBone {boneName=dyn_feather_r_10}
      boxExtents=<29 29 29>
      constraintSetup {angularLimitsMax=<80 30 30>, angularLimitsMin=<-60 -30 -30>, angularTarget=<1 0 0>}
      localJointOffset=<3 0 0>
     }]
   }
   n9 = Anim Dynamics - Bone: dyn_feather_r_09(ComponentPose=@n8, Alpha=1, ExternalForce="0,0,0") {
    angularDampingOverride=1
    angularSpringConstant=7000
    bAngularSpring=true
    bOverrideAngularDamping=true
    bUseGravityOverride=true
    gravityScale=0.35
    simulationSpace=RootRelative
    alphaBoolBlend {blendOption=Linear}
    boundBone {boneName=dyn_feather_r_09}
    componentLinearAccScale=<1 1 1>
    componentLinearVelScale=<1 1 1>
    gravityOverride {z=-980}
    physicsBodyDefinitions [{
      boundBone {boneName=dyn_feather_r_09}
      boxExtents=<29 29 29>
      constraintSetup {angularLimitsMax=<80 30 30>, angularLimitsMin=<-60 -30 -30>, angularTarget=<1 0 0>}
      localJointOffset=<3 0 0>
     }]
   }
   n10 = Anim Dynamics - Bone: dyn_feather_r_08(ComponentPose=@n9, Alpha=1, ExternalForce="0,0,0") {
    angularDampingOverride=1
    angularSpringConstant=20000
    bAngularSpring=true
    bOverrideAngularDamping=true
    bUseGravityOverride=true
    gravityScale=0.35
    simulationSpace=RootRelative
    alphaBoolBlend {blendOption=Linear}
    boundBone {boneName=dyn_feather_r_08}
    componentLinearAccScale=<1 1 1>
    componentLinearVelScale=<1 1 1>
    gravityOverride {z=-980}
    physicsBodyDefinitions [{
      boundBone {boneName=dyn_feather_r_08}
      boxExtents=<33 33 33>
      constraintSetup {angularLimitsMax=<30 30 30>, angularLimitsMin=<-40 -30 -30>, angularTarget=<1 0 0>}
      localJointOffset=<2 0 0>
     }]
   }
   n11 = Anim Dynamics - Bone: dyn_feather_r_07(ComponentPose=@n10, Alpha=1, ExternalForce="0,0,0") {
    angularDampingOverride=1
    angularSpringConstant=7000
    bAngularSpring=true
    bOverrideAngularDamping=true
    bUseGravityOverride=true
    gravityScale=0.35
    simulationSpace=RootRelative
    alphaBoolBlend {blendOption=Linear}
    boundBone {boneName=dyn_feather_r_07}
    componentLinearAccScale=<1 1 1>
    componentLinearVelScale=<1 1 1>
    gravityOverride {z=-980}
    physicsBodyDefinitions [{
      boundBone {boneName=dyn_feather_r_07}
      boxExtents=<29 29 29>
      constraintSetup {angularLimitsMax=<80 30 30>, angularLimitsMin=<-60 -30 -30>, angularTarget=<1 0 0>}
      localJointOffset=<3 0 0>
     }]
   }
   n12 = Anim Dynamics - Bone: dyn_feather_r_06(ComponentPose=@n11, Alpha=1, ExternalForce="0,0,0") {
    angularDampingOverride=1
    angularSpringConstant=7000
    bAngularSpring=true
    bOverrideAngularDamping=true
    bUseGravityOverride=true
    gravityScale=0.35
    simulationSpace=RootRelative
    alphaBoolBlend {blendOption=Linear}
    boundBone {boneName=dyn_feather_r_06}
    componentLinearAccScale=<1 1 1>
    componentLinearVelScale=<1 1 1>
    gravityOverride {z=-980}
    physicsBodyDefinitions [{
      boundBone {boneName=dyn_feather_r_06}
      boxExtents=<29 29 29>
      constraintSetup {angularLimitsMax=<80 30 30>, angularLimitsMin=<-60 -30 -30>, angularTarget=<1 0 0>}
      localJointOffset=<3 0 0>
     }]
   }
   n13 = Anim Dynamics - Bone: dyn_feather_r_05(ComponentPose=@n12, Alpha=1, ExternalForce="0,0,0") {
    angularDampingOverride=1
    angularSpringConstant=7000
    bAngularSpring=true
    bOverrideAngularDamping=true
    bUseGravityOverride=true
    gravityScale=0.35
    simulationSpace=RootRelative
    alphaBoolBlend {blendOption=Linear}
    boundBone {boneName=dyn_feather_r_05}
    componentLinearAccScale=<1 1 1>
    componentLinearVelScale=<1 1 1>
    gravityOverride {z=-980}
    physicsBodyDefinitions [{
      boundBone {boneName=dyn_feather_r_05}
      boxExtents=<29 29 29>
      constraintSetup {angularLimitsMax=<80 30 30>, angularLimitsMin=<-60 -30 -30>, angularTarget=<1 0 0>}
      localJointOffset=<3 0 0>
     }]
   }
   Output Pose(Result=Component To Local(ComponentPose=@n13))
  ]}
 EventGraph {nodes [
   n1 = Event OnCharacterCustomizationLoaded()
   n2 = Get Array Item with EFort Custom Part Type(Array:Array of Custom Character Part Object References=@n1.CustomParts) -> (Output:Custom Character Part Object Reference)
   n3 = Get BodyTypesPermitted!(self:Custom Character Part Object Reference=@n2) -> (BodyTypesPermitted:EFortCustomBodyType Enum=NONE)
   n4 = "Equal (Enum)"(
    A:EFortCustomGender Enum=Get GenderPermitted!(self:Custom Character Part Object Reference=@n2) -> (GenderPermitted:EFortCustomGender Enum=Invalid)
    B:EFortCustomGender Enum=Male
   )
   > Set Large Male(Large Male:bool=AND Boolean(A="Equal (Enum)"(A:EFortCustomBodyType Enum=@n3, B:EFortCustomBodyType Enum=Large), B=@n4)) -> (Output_Get:bool)
   n5 = "Equal (Enum)"(A=Get BodyTypesPermitted(self=@n2) -> (BodyTypesPermitted=NONE), B=Medium)
   > Set Medium Male(Medium Male:bool=AND Boolean(A=@n5, B=@n4)) -> (Output_Get:bool)
   n6 = Get GenderPermitted(self=@n2) -> (GenderPermitted=Invalid)
   > Set Medium Female(Medium Female:bool=AND Boolean(A=@n5, B="Equal (Enum)"(A:EFortCustomGender Enum=@n6, B:EFortCustomGender Enum=Female))) -> (Output_Get:bool)
   TryGetPawnOwner()
   Event BlueprintUpdateAnimation()
  ]}
]
variables [
 Large Male {type=boolean}
 Medium Female {type=boolean}
 Medium Male {type=boolean}
]
