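# MistRaven player Animation Blueprint implementing Item Override layers for the drain animations on the Fortnite player skeleton.
# ItemUpperBody runs a state machine that passes input through, plays normal, crouch-moving or sprint-moving drain outros from reload-outro flags, then returns automatically.
# ItemFullBodyAimOffsets uses bIsTendrilDrain to select a neck_01 layered blend containing the drain aim offset, whose two-way blend alpha is zero.
# ItemFullBodyOverride passes through; remaining layers are unwired.
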AnimBlueprint /MistRaven/Animation/MistRaven_Player_AnimBP
target_skeleton=&Fortnite_M_Avg_Player_Skeleton
graphs [
 AnimGraph {nodes [Output Pose()]}
 EventGraph {links [], nodes []}
 ItemAimOffset {nodes [
   Input Pose!() -> (Pose, AimOffsetAlpha:double, Yaw:double, Pitch:double) {name=InPose_AimOffset}
   Input Pose!() -> (Pose) {name=InPose_PreAimOffsets}
   Output Pose() {layerGroup=Item Override}
  ]}
 ItemFallAdditive {nodes [
   Input Pose() {name=InPoseItemFallAdditive}
   Output Pose() {layerGroup=Item Override}
  ]}
 ItemFallLandAdditive {nodes [
   Input Pose() {name=InPoseItemFallLandAdditive}
   Output Pose() {layerGroup=Item Override}
  ]}
 ItemFallLevel2Override {nodes [
   Input Pose() {name=InPose_FallLevel2}
   Output Pose() {layerGroup=Item Override}
  ]}
 ItemFinalPoseOverride {nodes [
   Input Pose() {name=InFinalPose}
   Output Pose() {layerGroup=Item Override}
  ]}
 ItemFullBodyAimOffsets {nodes [
   n1 = Input Pose!() -> (Pose, FullBody_AimOffsetAlpha:double, Aim_Yaw:double, Aim_Pitch:double) {name=InPoseFullBodyAimOffsets}
   n2 = Two Way Blend(
    A=Use cached pose 'InPose AimOffset'()
    B=AimOffset Player 'MistRaven_Drain_AO'!(
     X:float=@n1.Aim_Yaw
     Y:float=@n1.Aim_Pitch
     BasePose:Pose Link Structure=Input Pose!() -> (Pose) {name=InPosePreFullBodyAimOffsets}
     Alpha:float=1
    ) -> (Pose:Pose Link Structure) {blendSpace=&/MistRaven_Drain_AO, alphaBoolBlend {blendOption=Linear}}
    Alpha=0
   ) {blendNode {alphaBoolBlend {blendOption=Linear}}}
   n3 = Use cached pose 'InPose AimOffset'()
   Output Pose(Result=Blend Poses by bool(
     BlendPose_0=Layered blend per bone(BasePose=@n2, BlendPoses_0=@n3) {bMeshSpaceScaleBlend=true, layerSetup [{branchFilters [{boneName=neck_01}]}]}
     BlendPose_1=Use cached pose 'InPose AimOffset'()
     bActiveValue=Get bIsTendrilDrain() -> (bIsTendrilDrain:bool)
    ) {blendType=Linear}) {layerGroup=Item Override}
   Save cached pose 'InPose AimOffset'(Pose=@n1) {cacheName=InPose AimOffset}
  ]}
 ItemFullBodyOverride {nodes [Output Pose(Result=Input Pose() {name=InPoseFullBody}) {layerGroup=Item Override}]}
 ItemIdleOverride {nodes [
   Input Pose() {name=InPose_Idle}
   Output Pose() {layerGroup=Item Override}
  ]}
 ItemJetPackJumpAdditive {nodes [Output Pose() {layerGroup=Item Override}]}
 ItemJetPackStartAdditive {nodes [Output Pose() {layerGroup=Item Override}]}
 ItemJumpLoopAdditive {nodes [
   Input Pose() {name=InPoseItemJumpLoopAdditive}
   Output Pose() {layerGroup=Item Override}
  ]}
 ItemJumpOverride {nodes [
   Input Pose() {name=InPoseJump}
   Output Pose() {layerGroup=Item Override}
  ]}
 ItemJumpUpAdditive {nodes [
   Input Pose() {name=InPoseItemJumpUpAdditive}
   Output Pose() {layerGroup=Item Override}
  ]}
 ItemPreAimOffsets {nodes [
   Input Pose() {name=InPosePreAimOffsets}
   Output Pose() {layerGroup=Item Override}
  ]}
 ItemPreIK {nodes [
   Input Pose() {name=InPose_PreIK}
   Output Pose() {layerGroup=Item Override}
  ]}
 ItemRightHandIK {nodes [
   Input Pose() {name=InPosePreRightHandIK}
   Input Pose() {name=InPosePostRightHandIK}
   Output Pose() {layerGroup=Item Override}
  ]}
 ItemSkeletalControl {nodes [
   Input Pose()
   Output Pose() {layerGroup=Item Override}
  ]}
 ItemSwimJumpFallLoop {nodes [Output Pose() {layerGroup=Item Override}]}
 ItemSwimJumpSurfaceEnd {nodes [Output Pose() {layerGroup=Item Override}]}
 ItemSwimJumpSurfaceLoop {nodes [Output Pose() {layerGroup=Item Override}]}
 ItemUpperBody {nodes [
   AnimGraphNode_StateMachine_0 = Reload Outro SM() {is=title AnimGraphNode_StateMachine}
   Output Pose(Result=@AnimGraphNode_StateMachine_0) {layerGroup=Item Override}
   Save cached pose 'UpperBody Input Pose'(Pose=Input Pose!() -> (Pose, InputParam:Fort Anim Input Adjusted Aim Structure) {name=InPoseUpperBody}) {cacheName=UpperBody Input Pose}
  ]}
 Reload Outro SM {
  entry=Passthrough
  path=ItemUpperBody.AnimGraphNode_StateMachine_0.Reload Outro SM
  ids=id
  nodes [
   AnimStateNode_0 = Passthrough() {is=title AnimStateNode}
   AnimStateNode_1 = Outro Normal() {is=title AnimStateNode}
   AnimStateNode_2 = Outro Crouch Moving() {is=title AnimStateNode}
   AnimStateNode_3 = Outro Sprint Moving() {is=title AnimStateNode}
   AnimStateTransitionNode_2 = AnimStateTransitionNode()
   AnimStateTransitionNode_4 = AnimStateTransitionNode() {bAutomaticRuleBasedOnSequencePlayerInState=true}
   AnimStateTransitionNode_5 = AnimStateTransitionNode()
   AnimStateTransitionNode_6 = AnimStateTransitionNode() {bAutomaticRuleBasedOnSequencePlayerInState=true}
   AnimStateTransitionNode_7 = AnimStateTransitionNode()
   AnimStateTransitionNode_8 = AnimStateTransitionNode() {bAutomaticRuleBasedOnSequencePlayerInState=true, crossfadeDuration=0.4}
  ]
  links [
   {
    dst=Passthrough
    rule=ItemUpperBody.AnimGraphNode_StateMachine_0.Reload Outro SM.AnimStateTransitionNode_8.Transition
    src=Outro Crouch Moving
    via=AnimStateTransitionNode_8
   }
   {
    dst=Passthrough
    rule=ItemUpperBody.AnimGraphNode_StateMachine_0.Reload Outro SM.AnimStateTransitionNode_4.Transition
    src=Outro Normal
    via=AnimStateTransitionNode_4
   }
   {
    dst=Passthrough
    rule=ItemUpperBody.AnimGraphNode_StateMachine_0.Reload Outro SM.AnimStateTransitionNode_6.Transition
    src=Outro Sprint Moving
    via=AnimStateTransitionNode_6
   }
   {
    dst=Outro Crouch Moving
    rule=ItemUpperBody.AnimGraphNode_StateMachine_0.Reload Outro SM.AnimStateTransitionNode_7.Transition
    src=Passthrough
    via=AnimStateTransitionNode_7
   }
   {
    dst=Outro Normal
    rule=ItemUpperBody.AnimGraphNode_StateMachine_0.Reload Outro SM.AnimStateTransitionNode_2.Transition
    src=Passthrough
    via=AnimStateTransitionNode_2
   }
   {
    dst=Outro Sprint Moving
    rule=ItemUpperBody.AnimGraphNode_StateMachine_0.Reload Outro SM.AnimStateTransitionNode_5.Transition
    src=Passthrough
    via=AnimStateTransitionNode_5
   }
  ]
 }
 Passthrough {
  path=ItemUpperBody.AnimGraphNode_StateMachine_0.Reload Outro SM.AnimStateNode_0.Passthrough
  nodes [Output Animation Pose(Result=Use cached pose 'UpperBody Input Pose'())]
 }
 Outro Normal {
  path=ItemUpperBody.AnimGraphNode_StateMachine_0.Reload Outro SM.AnimStateNode_1.Outro Normal
  nodes [Output Animation Pose(Result=Sequence Player 'MistRaven_Drain_Outro'() {bLoopAnimation=false, sequence=&/MistRaven_Drain_Outro})]
 }
 Outro Crouch Moving {
  path=ItemUpperBody.AnimGraphNode_StateMachine_0.Reload Outro SM.AnimStateNode_2.Outro Crouch Moving
  nodes [Output Animation Pose(Result=Sequence Player 'MistRaven_Drain_Outro_Crouch'() {bLoopAnimation=false, sequence=&/MistRaven_Drain_Outro_Crouch})]
 }
 Outro Sprint Moving {
  path=ItemUpperBody.AnimGraphNode_StateMachine_0.Reload Outro SM.AnimStateNode_3.Outro Sprint Moving
  nodes [Output Animation Pose(Result=Sequence Player 'MistRaven_DrainOutro_Moving'() {bLoopAnimation=false, sequence=&/MistRaven_DrainOutro_Moving})]
 }
 Transition {
  path=ItemUpperBody.AnimGraphNode_StateMachine_0.Reload Outro SM.AnimStateTransitionNode_2.Transition
  nodes [Result(bCanEnterTransition=Get bIsReloadOutroNormal() -> (bIsReloadOutroNormal:bool))]
 }
 Transition {path=ItemUpperBody.AnimGraphNode_StateMachine_0.Reload Outro SM.AnimStateTransitionNode_4.Transition, nodes [Result()]}
 Transition {
  path=ItemUpperBody.AnimGraphNode_StateMachine_0.Reload Outro SM.AnimStateTransitionNode_5.Transition
  nodes [Result(bCanEnterTransition=Get bIsReloadOutroSprint() -> (bIsReloadOutroSprint:bool))]
 }
 Transition {path=ItemUpperBody.AnimGraphNode_StateMachine_0.Reload Outro SM.AnimStateTransitionNode_6.Transition, nodes [Result()]}
 Transition {
  path=ItemUpperBody.AnimGraphNode_StateMachine_0.Reload Outro SM.AnimStateTransitionNode_7.Transition
  nodes [Result(bCanEnterTransition=Get bIsReloadOutroCrouchMove() -> (bIsReloadOutroCrouchMove:bool))]
 }
 Transition {path=ItemUpperBody.AnimGraphNode_StateMachine_0.Reload Outro SM.AnimStateTransitionNode_8.Transition, nodes [Result()]}
 ItemZipLineStartAdditive {nodes [Output Pose() {layerGroup=Item Override}]}
]
variables []
