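; Player pawn component that runs the MistRaven Grim Strut Wraith Form: temporary invulnerability, body hiding, weapon holstering and teleport support.
; GameplayTag waits set replicated bRepWFActive, whose handler toggles invulnerability, knockback block and camera modifier, and starts the server duration timer.
; Separate RepNotify bools drive the mist material, looping Niagara trail and camera lens.
; Timeout, weapon swap, vehicle entry or Tendril Drain end the form, with ground-traced exit montage and attack lockout.
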(event Custom|CleanupCamera
  ;[0.15 0.1107 0.07482 0.5] CAMERAFX
  ;; CLEANUP CAMERA FX
  ;;
  (Utilities|SwitchHasAuthority
    (:Remote
      (LensEffect|IsValidCameraLens (Variables|camera|GetCameraLens)
        (:Valid
          (CallFunction|CleanupCameraLens (Variables|camera|GetCameraLens))
          (Variables|camera|SetCameraLens 0)
          (CallFunction|BPLogMistRaven "PComp GS Cam Remove"))
        (:Invalid
          (CallFunction|BPLogMistRaven "PComp GS invalid camera lense"))))))

(event Custom|AddCameraFX
  ;[0.15 0.1107 0.07482 0.5] CAMERAFX
  ;; ADDING CAMERA FX
  (Utilities|SwitchHasAuthority
    (:Remote
      (if (Variables|Default|GetbScreenFxEnable)
        (bind _returnvalue (Camera|AddGenericCameraLensEffect (Variables|/Script/Engine.PlayerController|GetPlayerCameraManager (Fortnite|GetFortPlayerControllerfromActor (Variables|Default|GetAs%20Fort%20Player%20Pawn))) "/MistRaven/Assets/VFX/CameraFX/BP_CameraLens_MistRaven_GrimStrut_Active.BP_CameraLens_MistRaven_GrimStrut_Active_C"))
        (Variables|camera|SetCameraLens _returnvalue)
        (CallFunction|BPLogMistRaven "PComp GS Cam Lens Add")))))

(event Custom|SetHiddenExplicit
  (params (NewHidden :type Boolean))
  ;[0.1468 0.15 0.07217 0.5] Hide/Unhide body but not fx
  (for _array_element (Components|GetChildrenComponents (Variables|/Script/Engine.Character|GetMesh (Variables|Default|GetAs%20Fort%20Player%20Pawn)) true)
    (bind _asniagara_particle_system_component (Utilities|Casting|CastToNiagaraComponent _array_element)
      (:then
        (if (not (Utilities|String|Contains (Utilities|GetDisplayName _asniagara_particle_system_component) "NS_MistRaven_WraithForm_Active_Loop"))
          (Development|SetHiddeninGame _asniagara_particle_system_component NewHidden false)))
      (:CastFailed
        (if (== (Variables|Default|GetAs%20Fort%20Player%20Pawn) (Components|GetOwner _array_element))
          (Development|SetHiddeninGame _array_element NewHidden false))
        (if (> (ScalableFloat|GetValueatLevel (Variables|Scalables|GetShouldHideCape)) 0)
          (bind _asskeletal_mesh_component_budgeted (Utilities|Casting|CastToSkeletalMeshComponentBudgeted _array_element))
          (if (== (Variables|/Script/Engine.SkinnedMeshComponent|GetSkeletalMesh _asskeletal_mesh_component_budgeted) "/MistRaven/Assets/Mesh/WraithForm_MistRaven.WraithForm_MistRaven")
            (Development|SetHiddeninGame _array_element (not NewHidden) false)))))))

(event Custom|Event_Invulnerable
  ; Wraith Form: Invulnerable State Projectiles cant hit this player anymore
  (CallFunction|BPLogMistRaven "PComp GS invulnerable")
  (bind _returnvalue (GameplayTags|HasMatchingGameplayTag(Message) (Components|GetOwner) "(TagName=\"MistRaven.GrimStrut.Active\")"))
  (if (or _returnvalue (ScalableFloat|GetValueasBool (Variables|Scalables|GetIgnoreInvulActivateCheck)))
    (Collision|SetCollisionResponsetoChannel (Variables|/Script/Engine.Character|GetMesh (Variables|Default|GetAs%20Fort%20Player%20Pawn)) "ECC_GameTraceChannel7" "ECR_Ignore")
    (Variables|/Script/FortniteGame.FortPawn|SetbIsInvulnerable true (Variables|Default|GetAs%20Fort%20Player%20Pawn))
    ;; knockback immunity
    (bind _returnvalue_2 (GameplayEffects|ApplyGameplayEffectToTarget :self (Ability|GetAbilitySystemComponent (Variables|Default|GetAs%20Fort%20Player%20Pawn)) :GameplayEffectClass "/MistRaven/Gameplay/Ascended/GrimStrutAscended/Blueprints/GE_GrimStrut_BlockKnockback.GE_GrimStrut_BlockKnockback_C" :Target (Ability|GetAbilitySystemComponent (Variables|Default|GetAs%20Fort%20Player%20Pawn)) :Level 0.0))
    (Variables|Handles|SetKnockbackBlock_Handle _returnvalue_2)
    (CallFunction|SetHiddenExplicit true)
    (else
      (CallFunction|BPLogMistRaven "PComp Invul NotActiveButTrying")
      (Utilities|Time|ClearandInvalidateTimerbyHandle (Variables|Handles|GetWraithFormActiveTimer))
      (Variables|MistRavenPlayerComponent|SetbRepWFActive false))))

(event Custom|Event_Vulnerable
  (bind _as_fort_player_pawn (Variables|Default|GetAs%20Fort%20Player%20Pawn))
  ; Wraith Form : Vulnerable State
  (CallFunction|BPLogMistRaven "PComp GS vulnerable")
  (Networking|FlushNetDormancy _as_fort_player_pawn)
  (if (ScalableFloat|GetValueasBool (Variables|Default|GetUseExplicitHiddenRestoration))
    (CallFunction|SetHiddenExplicit false)
    (else
      (Development|SetHiddeninGame (Variables|/Script/Engine.Character|GetMesh _as_fort_player_pawn) false true)))
  (Collision|SetCollisionResponsetoChannel (Variables|/Script/Engine.Character|GetMesh (Variables|Default|GetAs%20Fort%20Player%20Pawn)) "ECC_GameTraceChannel7" "ECR_Block")
  (Development|PrintString "GS CLEANUP" true true "(R=1.000000,G=0.206779,B=0.951215,A=1.000000)" 5.0)
  (GameplayEffects|RemoveActiveGameplayEffect (Ability|GetAbilitySystemComponent (Variables|Default|GetAs%20Fort%20Player%20Pawn)) (Variables|Handles|GetKnockbackBlock_Handle) -1)
  (Variables|/Script/FortniteGame.FortPawn|SetbIsInvulnerable false (Variables|Default|GetAs%20Fort%20Player%20Pawn))
  (Variables|RepNotifyVars|SetbEnableMistOverrideMaterial false) ; onrep material swap
  (CallFunction|BPLogMistRaven "PComp GS vulnerable FINISHED"))

(event Custom|EquipWeaponAgain
  ; Allow the player to attack again
  (Weapon|SetHolsterWeapon (Variables|Default|GetAs%20Fort%20Player%20Pawn) 0 false false (Variables|Default|GetWraithFormHolsterId))
  (Utilities|Time|ClearandInvalidateTimerbyHandle (Variables|Handles|GetReequipTimer)))

(event Custom|EndOfWraithForm_Cleanup
  (bind _as_fort_player_pawn (Variables|Default|GetAs%20Fort%20Player%20Pawn))
  (bind _wraithformfxhandle (Variables|Default|GetWraithFormFXHandle))
  ;[0.15 0.06641 0.06798 0.5] END WRAITH FORM - now separated from invul switching
  (CallFunction|BPLogMistRaven "PComp GS Cleanup - Tag Remove v1")
  (Ability|GameplayEffect|RemoveLooseGameplayTags (Variables|Default|GetAs%20Fort%20Player%20Pawn) (Variables|Tags|GetGameplay%20Tags) true)
  (Ability|Async|EndAction (Variables|Handles|GetDrainWatch))
  (Variables|RepNotifyVars|SetbEnableMistOverrideMaterial false) ; onrep material swap
  (Utilities|SwitchHasAuthority
    (:Authority
      _)
    (:Remote
      (bind _returnvalue (Utilities|Time|SetTimerbyEvent (delegate Custom|Event_TimerCleanupCamera) 0.6))
      (Variables|Handles|SetClientCamerCleanupHandle _returnvalue)))
  (Ability|SendGameplayEventtoActor _as_fort_player_pawn "(TagName=\"MistRaven.GrimStrut.Ascended.EndAbility\")")
  ;[0.9896 0.05386 0.0964 0.5] VFX. Exit Wraith but player is in air, wraith trail FX would keep spawning.
  (if (not (Variables|Default|GetNearGroundOnExit))
    (Utilities|IsValid _wraithformfxhandle
      (:Is%20Valid
        (Effects|Components|ParticleSystem|SetFloatParameter _wraithformfxhandle "Is Air" 0.0))
      (:Is%20Not%20Valid
        _))
    (EventDispatchers|BindEventtoOnPawnEnterWater _as_fort_player_pawn (EventDispatchers|CreateEvent :!event Custom|Enter%20Water))
    (EventDispatchers|AssignOnPawnLanded _as_fort_player_pawn (EventDispatchers|CreateEvent :!event Custom|OnPawnLanded_Event))
    (bind _returnvalue_2 (Utilities|Time|SetTimerbyEvent (delegate Custom|Failsafe) 2.5))
    (Variables|Handles|SetFailsafeHandle _returnvalue_2)
    (else
      (Variables|RepNotifyVars|SetbRepApplyWFFX false))))

(event Custom|Enter%20Water
  (params (FortPlayerPawn :type Fort%20Player%20Pawn%20Object%20Reference))
  ;[0.9896 0.05386 0.0964 0.5] VFX. Exit Wraith but player is in air, wraith trail FX would keep spawning.
  (Variables|RepNotifyVars|SetbRepApplyWFFX false))

(event Custom|EndWraithFormWithAnimCost
  ; Call to end ability but also stop the player fromm attacking for a moment
  (Utilities|SwitchHasAuthority
    (:Authority
      _)
    (:Remote
      (CallFunction|SendEndEventtoPPCamera)))
  (Utilities|Time|ClearandInvalidateTimerbyHandle (Variables|Handles|GetWraithFormActiveTimer))
  (CallFunction|EndOfWraithFormCleanup)
  (bind _returnvalue (CallFunction|SetExitAnim))
  (Variables|Default|SetNearGroundOnExit _returnvalue)
  (Development|PrintString (Utilities|GetDisplayName (Animation|GetCurrentMontage (Variables|Default|GetAs%20Fort%20Player%20Pawn) :!class /Script/Engine.Character)))
  (if (not (GameplayTags|HasMatchingGameplayTag (Variables|Default|GetAs%20Fort%20Player%20Pawn) "(TagName=\"MistRaven.TendrilDrain.Ascended.Active\")"))
    (Animation|Utilities|PlayAnimMontage (Variables|Default|GetAs%20Fort%20Player%20Pawn) (Variables|Default|GetExitAnimMontage)))
  (if (not (GameplayTags|HasAnyMatchingGameplayTags (Ability|GetAbilitySystemComponent (Variables|Default|GetAs%20Fort%20Player%20Pawn)) (Variables|Tags|GetTC_TendrilActive)))
    (Utilities|Time|SetTimerbyEvent (delegate Custom|SetLandHolster) 0.15)
    (GameplayEffects|ApplyGameplayEffectToTarget :self (Ability|GetAbilitySystemComponent (Variables|Default|GetAs%20Fort%20Player%20Pawn)) :GameplayEffectClass "/MistRaven/Gameplay/Ascended/GrimStrutAscended/Blueprints/GE_GrimStrutAscended_CantAttack.GE_GrimStrutAscended_CantAttack_C" :Target (Ability|GetAbilitySystemComponent (Variables|Default|GetAs%20Fort%20Player%20Pawn)) :Level 0.0)))

(event Custom|Failsafe
  (if (not (GameplayTags|HasMatchingGameplayTag (Variables|Default|GetAs%20Fort%20Player%20Pawn) "(TagName=\"MistRaven.GrimStrut.Active\")"))
    (Variables|RepNotifyVars|SetbRepApplyWFFX false)))

(event Custom|Event_TimerCleanupCamera
  (Variables|RepNotifyVars|SetbCameraLensActive false))

(event Custom|OnPawnLanded_Event
  (params (Hit :type Hit%20Result%20Structure%20%28by%20ref%29))
  ;[0.9896 0.05386 0.0964 0.5] VFX. Exit Wraith but player is in air, wraith trail FX would keep spawning.
  (Variables|RepNotifyVars|SetbRepApplyWFFX false))

(event Custom|SetLandHolster
  (bind _as_fort_player_pawn (Variables|Default|GetAs%20Fort%20Player%20Pawn))
  (bind _returnvalue (Ability|GetAbilitySystemComponent _as_fort_player_pawn))
  (bind _as_fort_player_pawn_2 (Variables|Default|GetAs%20Fort%20Player%20Pawn))
  ; Call to end ability but also stop the player fromm attacking for a moment
  (if (and (> (ScalableFloat|GetValueatLevel (Variables|Scalables|GetInAirHolsterEnabled)) 0) (not (Variables|Default|GetNearGroundOnExit)))
    (Utilities|Casting|CastToB_MistRaven_WeaponAscended (Weapon|GetCurrentWeapon (Variables|Default|GetAs%20Fort%20Player%20Pawn))
      (:then
        _)
      (:CastFailed
        (GameplayEffects|ApplyGameplayEffectToTarget :self _returnvalue :GameplayEffectClass "/Game/Athena/Items/Parents/WeaponHolster/GE_TillLandHolster_Grant_Generic.GE_TillLandHolster_Grant_Generic_C" :Target _returnvalue :Level 0.0)))
    (GameplayEffects|ApplyGameplayEffectToTarget :self _returnvalue :GameplayEffectClass "/MistRaven/Gameplay/Ascended/GrimStrutAscended/Blueprints/HolsterTech/GE_TillLandManager_Grant_MistRavenGrimStrutSpecific.GE_TillLandManager_Grant_MistRavenGrimStrutSpecific_C" :Target _returnvalue :Level 0.0)
    (bind _returnvalue_2 (GameplayEffects|ApplyGameplayEffectToTarget :self _returnvalue :GameplayEffectClass "/MistRaven/Gameplay/Ascended/GrimStrutAscended/Blueprints/GE_GrimStrut_DontShowHolster.GE_GrimStrut_DontShowHolster_C" :Target _returnvalue :Level 0.0))
    (Variables|Handles|SetDontShowHolsterEffect _returnvalue_2)
    (if (ScalableFloat|GetValueasBool (Variables|Scalables|GetHolsterZiplineForceCancelEnabled))
      (bind _asynctaskproxy (Ability|Tasks|WaitGameplayTagAddToActor _as_fort_player_pawn_2 "(TagName=\"Abilities.Generic.Athena.ZipLine\")")
        (:then
          (Variables|Handles|SetZiplineHandle _asynctaskproxy))
        (:Added
          (Utilities|IsValid (Ability|GetFortAbilitySystemComponent _as_fort_player_pawn_2)
            (:Is%20Valid
              (Ability|SendGameplayEventtoActor _as_fort_player_pawn_2 "(TagName=\"Athena.Event.Gameplay.ForceCancelTillLand\")"))))))))

(event Custom|ApplyWFFx
  (bind _wraithformfxhandle (Variables|Default|GetWraithFormFXHandle))
  ; APPLY WF LOOPING FX
  ;
  (Utilities|IsValid _wraithformfxhandle
    (:Is%20Valid
      (Components|DestroyComponent _wraithformfxhandle))
    (:Is%20Not%20Valid
      _))
  (bind _returnvalue (Effects|SpawnFXSystemAttached "/MistRaven/VFX/NS_MistRaven_WraithForm_Active_Loop.NS_MistRaven_WraithForm_Active_Loop" (Variables|/Script/Engine.Character|GetMesh (Variables|Default|GetAs%20Fort%20Player%20Pawn)) "pelvis"))
  (Variables|Default|SetWraithFormFXHandle _returnvalue)
  (Effects|Components|ParticleSystem|SetFloatParameter _returnvalue "Looping Duration" (ScalableFloat|GetValueatLevel (Variables|Scalables|GetDuration)))
  (Effects|Components|ParticleSystem|SetBoolParameter _returnvalue "Ascended" (Variables|Default|GetIS%20Ascended))
  (Effects|Components|ParticleSystem|SetFloatParameter _returnvalue "Is Air" 1.0))

(event Custom|DeactivateFX
  (bind _wraithformfxhandle (Variables|Default|GetWraithFormFXHandle))
  (Utilities|IsValid _wraithformfxhandle
    (:Is%20Valid
      (Components|Activation|Deactivate _wraithformfxhandle))
    (:Is%20Not%20Valid
      _))
  (GameplayEffects|RemoveActiveGameplayEffect (Ability|GetAbilitySystemComponent (Variables|Default|GetAs%20Fort%20Player%20Pawn)) (Variables|Handles|GetDontShowHolsterEffect) -1))

(event EventBeginPlay
  (bind _returnvalue (Components|GetOwner))
  ;[0.07885 0.15 0.08442 0.5] BEGINPLAY
  (bind _asfort_player_pawn (Utilities|Casting|CastToFortPlayerPawn (Components|GetOwner)))
  (bind _output_get (Variables|Default|SetAs%20Fort%20Player%20Pawn _asfort_player_pawn))
  (FortPawn|BindEventtoOnWeaponAttachmentChangedDelegate _output_get (EventDispatchers|CreateEvent :!event Custom|OnWeaponAttachmentChangedDelegate_Event))
  (bind _asplayer_pawn_athena_generic_parent (Utilities|Casting|CastToPlayerPawn_Athena_Generic_Parent (Components|GetOwner)))
  (Default|AssignOnEnteredVehicleDispatcher _asplayer_pawn_athena_generic_parent (EventDispatchers|CreateEvent :!event Custom|Event_OnVehicleEnter))
  (Utilities|IsValid (Ability|GetAbilitySystemComponent (Variables|Default|GetAs%20Fort%20Player%20Pawn))
    (:Is%20Valid
      (Ability|Tasks|WaitGameplayTagAddToActor _returnvalue "(TagName=\"MistRaven.GrimStrut.Active\")"
        (:then
          (CallFunction|SetupCancelTags _returnvalue))
        (:Added
          (bind _returnvalue_2 (GameplayTags|HasMatchingGameplayTag(Message) _returnvalue "(TagName=\"MistRaven.AllAscendedEquipped\")"))
          (Variables|Default|SetIS%20Ascended _returnvalue_2)
          (Variables|MistRavenPlayerComponent|SetStartGrimStrutLocation (Transformation|GetActorLocation (Components|GetOwner))) ; grab the location where we start from\nbefore the dash, so we can fall back\nto using it if no input pressed.
          (Weapon|SetHolsterWeapon (Variables|Default|GetAs%20Fort%20Player%20Pawn) 0 true false (Variables|Default|GetWraithFormHolsterId))
          (Ability|Tasks|WaitGameplayTagAddToActor (Variables|Default|GetAs%20Fort%20Player%20Pawn) "(TagName=\"MistRaven.GrimStrut.Notify_Activated\")" true
            (:then
              ;[0 0.15 0.000738 0.5] PLACEHOLDER FOR ACTIVATION MATERIAL WIPE
              (Variables|RepNotifyVars|SetbEnableMistOverrideMaterial true) ; TODO: can we do something like this from a GC?\nor perhaps from a GE we can gaurantee in all\nscenarios will be cleaned up?
            )
            (:Added
              (if (and (GameplayTags|HasMatchingGameplayTag (Variables|Default|GetAs%20Fort%20Player%20Pawn) "(TagName=\"MistRaven.GrimStrut.Active\")") (not (GameplayTags|HasMatchingGameplayTag (Variables|Default|GetAs%20Fort%20Player%20Pawn) "(TagName=\"MistRaven.GrimStrut.Cancelled\")")))
                (Utilities|ResetVars)
                ; GRIM STRUT TRIGGERED
                ;
                (Variables|MistRavenPlayerComponent|SetbRepWFActive true)
                (Variables|RepNotifyVars|SetbRepApplyWFFX true)
                (Variables|RepNotifyVars|SetbCameraLensActive true)
                (bind _returnvalue_3 (Utilities|Time|SetTimerbyEvent (delegate Custom|EndingSoundGameplayCue) (- (ScalableFloat|GetValueasDouble (Variables|Scalables|GetDuration)) (ScalableFloat|GetValueasDouble (Variables|Scalables|GetDelayBetweenWipeAndEnd)))))
                (Variables|Handles|SetEndingSoundGameplayCueTimer _returnvalue_3)
                (EventDispatchers|UnbindEventfromOnPawnEnterWater (Variables|Default|GetAs%20Fort%20Player%20Pawn) (EventDispatchers|CreateEvent :!event Custom|Enter%20Water))
                (EventDispatchers|UnbindEventfromOnPawnLanded (Variables|Default|GetAs%20Fort%20Player%20Pawn) (EventDispatchers|CreateEvent :!event Custom|OnPawnLanded_Event))
                (bind _returnvalue_4 (Utilities|Time|SetTimerbyEvent (delegate Custom|EVENT_Unholster) 0.1))
                (Variables|Handles|SetUnholster_Handle _returnvalue_4)
                (bind _returnvalue_5 (Utilities|Time|SetTimerbyEvent (delegate Custom|SafeToStartInterruptCheck) (ScalableFloat|GetValueatLevel (Variables|Scalables|GetInterruptDelay))))
                (Variables|Handles|SetInterruptHandle _returnvalue_5)
                (else
                  (CallFunction|Failsafe)
                  (Utilities|Time|ClearandInvalidateTimerbyHandle (Variables|Handles|GetWraithFormActiveTimer))
                  (Variables|MistRavenPlayerComponent|SetbRepWFActive false))))))))
    (:Is%20Not%20Valid
      (EventDispatchers|AssignOnAbilitySystemComponentInitialized (Variables|Default|GetAs%20Fort%20Player%20Pawn) (delegate Custom|OnAbilitySystemComponentInitialized_Event)))))

(event Custom|OnAbilitySystemComponentInitialized_Event
  (params (AbilitySystemComponent :type Fort%20Ability%20System%20Component%20Object%20Reference) (AffectedPawn :type Fort%20Player%20Pawn%20Object%20Reference))
  (bind _returnvalue (Components|GetOwner))
  ;[0.07885 0.15 0.08442 0.5] BEGINPLAY
  (Ability|Tasks|WaitGameplayTagAddToActor _returnvalue "(TagName=\"MistRaven.GrimStrut.Active\")"
    (:then
      (CallFunction|SetupCancelTags _returnvalue))
    (:Added
      (bind _returnvalue_2 (GameplayTags|HasMatchingGameplayTag(Message) _returnvalue "(TagName=\"MistRaven.AllAscendedEquipped\")"))
      (Variables|Default|SetIS%20Ascended _returnvalue_2)
      (Variables|MistRavenPlayerComponent|SetStartGrimStrutLocation (Transformation|GetActorLocation (Components|GetOwner))) ; grab the location where we start from\nbefore the dash, so we can fall back\nto using it if no input pressed.
      (Weapon|SetHolsterWeapon (Variables|Default|GetAs%20Fort%20Player%20Pawn) 0 true false (Variables|Default|GetWraithFormHolsterId))
      (Ability|Tasks|WaitGameplayTagAddToActor (Variables|Default|GetAs%20Fort%20Player%20Pawn) "(TagName=\"MistRaven.GrimStrut.Notify_Activated\")" true
        (:then
          ;[0 0.15 0.000738 0.5] PLACEHOLDER FOR ACTIVATION MATERIAL WIPE
          (Variables|RepNotifyVars|SetbEnableMistOverrideMaterial true) ; TODO: can we do something like this from a GC?\nor perhaps from a GE we can gaurantee in all\nscenarios will be cleaned up?
        )
        (:Added
          (if (and (GameplayTags|HasMatchingGameplayTag (Variables|Default|GetAs%20Fort%20Player%20Pawn) "(TagName=\"MistRaven.GrimStrut.Active\")") (not (GameplayTags|HasMatchingGameplayTag (Variables|Default|GetAs%20Fort%20Player%20Pawn) "(TagName=\"MistRaven.GrimStrut.Cancelled\")")))
            (Utilities|ResetVars)
            ; GRIM STRUT TRIGGERED
            ;
            (Variables|MistRavenPlayerComponent|SetbRepWFActive true)
            (Variables|RepNotifyVars|SetbRepApplyWFFX true)
            (Variables|RepNotifyVars|SetbCameraLensActive true)
            (bind _returnvalue_3 (Utilities|Time|SetTimerbyEvent (delegate Custom|EndingSoundGameplayCue) (- (ScalableFloat|GetValueasDouble (Variables|Scalables|GetDuration)) (ScalableFloat|GetValueasDouble (Variables|Scalables|GetDelayBetweenWipeAndEnd)))))
            (Variables|Handles|SetEndingSoundGameplayCueTimer _returnvalue_3)
            (EventDispatchers|UnbindEventfromOnPawnEnterWater (Variables|Default|GetAs%20Fort%20Player%20Pawn) (EventDispatchers|CreateEvent :!event Custom|Enter%20Water))
            (EventDispatchers|UnbindEventfromOnPawnLanded (Variables|Default|GetAs%20Fort%20Player%20Pawn) (EventDispatchers|CreateEvent :!event Custom|OnPawnLanded_Event))
            (bind _returnvalue_4 (Utilities|Time|SetTimerbyEvent (delegate Custom|EVENT_Unholster) 0.1))
            (Variables|Handles|SetUnholster_Handle _returnvalue_4)
            (bind _returnvalue_5 (Utilities|Time|SetTimerbyEvent (delegate Custom|SafeToStartInterruptCheck) (ScalableFloat|GetValueatLevel (Variables|Scalables|GetInterruptDelay))))
            (Variables|Handles|SetInterruptHandle _returnvalue_5)
            (else
              (CallFunction|Failsafe)
              (Utilities|Time|ClearandInvalidateTimerbyHandle (Variables|Handles|GetWraithFormActiveTimer))
              (Variables|MistRavenPlayerComponent|SetbRepWFActive false))))))))

(event Custom|EVENT_Unholster
  (Weapon|SetHolsterWeapon (Variables|Default|GetAs%20Fort%20Player%20Pawn) 0 false false (Variables|Default|GetWraithFormHolsterId)))

(event Custom|EndingSoundGameplayCue
  (bind _endingsoundcue (Variables|Tags|GetEndingSoundCue))
  (bind _as_fort_player_pawn (Variables|Default|GetAs%20Fort%20Player%20Pawn))
  ; GRIM STRUT TRIGGERED
  ;
  (Networking|SwitchHasAuthority
    (:Remote
      (GameplayCue|AddGameplayCue _as_fort_player_pawn _endingsoundcue)
      (GameplayCue|ExecuteGameplayCue _as_fort_player_pawn _endingsoundcue)
      (GameplayCue|ExecuteGameplayCue%28Burst%29Local _as_fort_player_pawn _endingsoundcue (Ability|GameplayCue|MakeGameplayCueParameters 0.0)) ; Cam shake + controller vibration
      (Utilities|Time|ClearandInvalidateTimerbyHandle (Variables|Handles|GetEndingSoundGameplayCueTimer)))))

(event Custom|SafeToStartInterruptCheck
  ;[0.15 0.1024 0.09377 0.5] Tendril Interrupt
  (bind _asynctaskproxy (Ability|Tasks|WaitGameplayTagAddToActor (Components|GetOwner) "(TagName=\"MistRaven.TendrilDrain.Ascended.Active\")" true)
    (:then
      (Variables|Handles|SetDrainWatch _asynctaskproxy))
    (:Added
      (if (GameplayTags|HasMatchingGameplayTag (Variables|Default|GetAs%20Fort%20Player%20Pawn) "(TagName=\"MistRaven.GrimStrut.Active\")")
        (Utilities|Time|ClearandInvalidateTimerbyHandle (Variables|Handles|GetWraithFormActiveTimer))
        (Variables|MistRavenPlayerComponent|SetbRepWFActive false)
        (else
          (Development|PrintString "GSnotactive"))))))

(event Custom|EventStartServerDurationTimer
  (params (DurationTime :type Float%20%28double-precision%29))
  ; SERVER TIMER DURATION TIME
  (bind _returnvalue (Utilities|Time|SetTimerbyEvent (delegate Custom|Event_EndWraithFormTimesUp) DurationTime))
  (Variables|Handles|SetWraithFormActiveTimer _returnvalue))

(event Custom|OnWeaponAttachmentChangedDelegate_Event
  (params (NewWeapon :type Fort%20Weapon%20Object%20Reference) (PrevWeapon :type Fort%20Weapon%20Object%20Reference))
  ; Handle Weapon Switching
  (if (GameplayTags|HasMatchingGameplayTag (Variables|Default|GetAs%20Fort%20Player%20Pawn) "(TagName=\"MistRaven.GrimStrut.HolsterGrimStrutSpecific\")")
    (GameplayEffects|ApplyGameplayEffectToTarget :self (Ability|GetAbilitySystemComponent (Variables|Default|GetAs%20Fort%20Player%20Pawn)) :GameplayEffectClass "/Game/Athena/Items/Parents/WeaponHolster/GE_TillLandHolster_Grant_Generic.GE_TillLandHolster_Grant_Generic_C" :Target (Ability|GetAbilitySystemComponent (Variables|Default|GetAs%20Fort%20Player%20Pawn)) :Level 0.0)
    (else
      _))
  (if (GameplayTags|HasMatchingGameplayTag (Variables|Default|GetAs%20Fort%20Player%20Pawn) "(TagName=\"MistRaven.GrimStrut.Active\")")
    (Utilities|Time|ClearandInvalidateTimerbyHandle (Variables|Handles|GetWraithFormActiveTimer))
    (Variables|MistRavenPlayerComponent|SetbRepWFActive false))
  (if (GameplayTags|HasMatchingGameplayTag (Variables|Default|GetAs%20Fort%20Player%20Pawn) "(TagName=\"MistRaven.Connector.Ascended.PlacementModeActive\")")
    (Utilities|Casting|CastToB_MistRaven_WeaponAscended NewWeapon
      (:CastFailed
        (Ability|GameplayEffect|RemoveLooseGameplayTags (Variables|Default|GetAs%20Fort%20Player%20Pawn) (Variables|Default|GetPlacementModeActiveTag) true)))))

(event Custom|Event_EndWraithFormTimesUp
  ; SERVER TIMER DURATION TIME
  (Variables|MistRavenPlayerComponent|SetbRepWFActive false))

(event Custom|SetLocalTeleportLocations
  (params (InStart :type Vector) (InEnd :type Vector))
  ; Set these values so camera modifier can easily read data.
  (Variables|Default|SetLocal_LastTeleportStartPos InStart)
  (Variables|Default|SetLocal_LastTeleportEndPos InEnd))

(event Custom|Add%20Camera%20Modifier
  ;[0.15 0.1177 0.1001 0.5] CAMERA MODIFIER
  (if (Variables|Default|GetUsePlayerCompForCameraModifer)
    (bind _returnvalue (Game|Player|AddNewCameraModifier (Variables|/Script/Engine.PlayerController|GetPlayerCameraManager (Fortnite|GetFortPlayerControllerfromActor (Variables|Default|GetAs%20Fort%20Player%20Pawn))) (Variables|camera|GetCamera%20Modifier%20Class)))
    (bind _ascamera_modifier_grim_strut (Utilities|Casting|CastToCameraModifier_GrimStrut_ _returnvalue))
    (Variables|camera|SetCamera%20Modifier%20Grim%20Strut _ascamera_modifier_grim_strut)
    (CallFunction|BPLogMistRaven "PComp GS Cam Mod Add")))

(event Custom|ClientToServer_TeleportNow
  (params (Position :type Vector) (Rotation :type Rotator))
  (Transform|Actor|TeleportPlayerPawn (Variables|Default|GetAs%20Fort%20Player%20Pawn) Position Rotation))

(event Custom|Remove%20Camera%20Mod
  (bind _camera_modifier_grim_strut (Variables|camera|GetCamera%20Modifier%20Grim%20Strut))
  ;[0.15 0.1177 0.1001 0.5] CAMERA MODIFIER
  ;;[1 1 1] Remove Camera Modifier
  (if (Variables|Default|GetUsePlayerCompForCameraModifer)
    (Utilities|IsValid _camera_modifier_grim_strut
      (:Is%20Valid
        (Class|CameraModifierGrimStrut|WindDownandRemove _camera_modifier_grim_strut)
        (CallFunction|BPLogMistRaven "PComp GS Cam Mod Remove")))))

(event Custom|SetupCancelTags
  (params (TargetActor :type Actor%20Object%20Reference))
  (Utilities|FlowControl|Sequence
    (:then_0
      (Ability|Tasks|WaitGameplayTagAddToActor TargetActor "(TagName=\"MistRaven.GrimStrut.Cancelled\")"
        (:Added
          _)))
    (:then_1
      (Ability|Tasks|WaitGameplayTagAddToActor TargetActor "(TagName=\"Conversation.InteractingWithNPC\")"
        (:Added
          _)))
    (:then_2
      (Ability|Tasks|WaitGameplayTagAddToActor TargetActor "(TagName=\"Gameplay.Status.DBNO\")"
        (:Added
          _)))
    (:then_3
      (Ability|Tasks|WaitGameplayTagAddToActor TargetActor "(TagName=\"Granted.Riding.Player.IsRiding\")"
        (:Added
          _)))
    (:then_4
      (if (< (ScalableFloat|GetValueasDouble (Variables|Default|GetCanWraithFormPickUpDBNO)) 1.0)
        (Ability|Tasks|WaitGameplayTagAddToActor TargetActor "(TagName=\"Gameplay.Action.Player.CarryingPlayer\")"
          (:Added
            _))))
    (:then_5
      (if (< (ScalableFloat|GetValueasDouble (Variables|Default|GetAllowWraithFormInteract)) 1.0)
        (Utilities|FlowControl|Sequence
          (:then_0
            (Ability|Tasks|WaitGameplayTagAddToActor TargetActor "(TagName=\"Gameplay.Action.Player.InteractAny\")"
              (:Added
                _)))
          (:then_1
            (Ability|Tasks|WaitGameplayTagAddToActor TargetActor "(TagName=\"Gameplay.Action.Player.InteractActive\")"
              (:Added
                _)))
          (:then_2
            (Ability|Tasks|WaitGameplayTagAddToActor TargetActor "(TagName=\"Gameplay.Action.Player.InteractUse\")"
              (:Added
                _)))
          (:then_3
            (Ability|Tasks|WaitGameplayTagAddToActor TargetActor "(TagName=\"Gameplay.Action.Player.InteractSearch\")"
              (:Added
                _)))))))
  (CallFunction|EventCancelWraithForm))

(event Custom|Event_OnVehicleEnter
  ; Cancel OnEnterVehicle
  (bind _returnvalue (GameplayTags|HasMatchingGameplayTag(Message) (Ability|GetAbilitySystemComponent (Variables|Default|GetAs%20Fort%20Player%20Pawn)) "(TagName=\"MistRaven.GrimStrut.Active\")"))
  (if _returnvalue
    (Utilities|Time|ClearandInvalidateTimerbyHandle (Variables|Handles|GetWraithFormActiveTimer))
    (Variables|MistRavenPlayerComponent|SetbRepWFActive false)))

(event Custom|Event_CancelWraithForm
  ; Cancel OnEnterVehicle
  (bind _returnvalue (GameplayTags|HasMatchingGameplayTag(Message) (Ability|GetAbilitySystemComponent (Variables|Default|GetAs%20Fort%20Player%20Pawn)) "(TagName=\"MistRaven.GrimStrut.Active\")"))
  (if _returnvalue
    (Utilities|Time|ClearandInvalidateTimerbyHandle (Variables|Handles|GetWraithFormActiveTimer))
    (Variables|MistRavenPlayerComponent|SetbRepWFActive false)))

(event Custom|HandleReplicated_bRepWFActive
  (if (Variables|MistRavenPlayerComponent|GetbRepWFActive)
    (CallFunction|BPLogMistRaven "Rep WFActive true")
    (CallFunction|EventInvulnerable)
    (CallFunction|AddCameraModifier)
    (Networking|SwitchHasAuthority
      (:Authority
        ;[0.06794 0.15 0.06901 0.5] Start Duration Timer On Server
        (bind _output_get (Variables|Default|SetOriginalGameDataDuration (ScalableFloat|GetValueatLevel (Variables|Scalables|GetDuration))))
        (CallFunction|EventStartServerDurationTimer _output_get))
      (:Remote
        _))
    (else
      (CallFunction|BPLogMistRaven "Rep WFActive false")
      (CallFunction|EndWraithFormwithAnimCost)
      (CallFunction|RemoveCameraMod)
      (CallFunction|EventVulnerable)
      (CallFunction|EVENTUnholster))))

(event EventOnNativeReplicatedWFActive
  (params (bNewValue :type Boolean))
  (CallFunction|HandleReplicatedBRepWFActive))

(fn CleanupCameraLens
  (params (CameraLens :type Camera%20Lens%20Effect%20Interface%20Interface))
  (bind _returnvalue (CameraLensEffect|GetParticleComponents CameraLens))
  (if (Utilities|IsValidInterface CameraLens)
    ; TODO these should just delete when their lens deletes instead
    (if (Utilities|Array|IsNotEmpty _returnvalue)
      (for _array_element _returnvalue
        (Rendering|SetVisibility _array_element false true))
      (else
        _))
    (Camera|RemoveGenericCameraLensEffect (Variables|/Script/Engine.PlayerController|GetPlayerCameraManager (Fortnite|GetFortPlayerControllerfromActor (Variables|Default|GetAs%20Fort%20Player%20Pawn))) CameraLens)
    (return)
    (else
      (return))))

(fn GetLastLocalTeleportLocations
  (return (Variables|Default|GetLocal_LastTeleportStartPos) (Variables|Default|GetLocal_LastTeleportEndPos)))

(fn OnRep_bCameraLensActive
  (if (Variables|RepNotifyVars|GetbCameraLensActive)
    (CallFunction|AddCameraFX)
    (else
      (CallFunction|CleanupCamera))))

(fn OnRep_bEnableMistOverrideMaterial
  (if (and (Variables|RepNotifyVars|GetbEnableMistOverrideMaterial) (Variables|Default|GetSetting_EnableMistMatOverrideFeature))
    (if (not (Guid|IsValidGuid (Variables|Default|GetMistMatOverrideId)))
      (bind _returnvalue (Class|FortPlayerPawn|ApplyMaterialOverrideLocalwithFlags :self (Variables|Default|GetAs%20Fort%20Player%20Pawn) :Material (Variables|Default|GetMI_MistRaven_WraithFrom) :Priority 0.0 :Flags 0))
      (Variables|Default|SetMistMatOverrideId _returnvalue)
      (Class|FortPlayerPawn|OverrideScalarMaterialParameterLocal (Variables|Default|GetAs%20Fort%20Player%20Pawn) "TL_Time" (Utilities|Time|GetGameTimeinSeconds))
      (CallFunction|BPLogMistRaven "Mist Mat Set" "Verbose")
      (else
        (CallFunction|BPLogMistRaven "Mist Mat Already Valid" "Verbose")))
    (elif (Guid|IsValidGuid (Variables|Default|GetMistMatOverrideId))
      (Class|FortPlayerPawn|RemoveMaterialOverrideLocal (Variables|Default|GetAs%20Fort%20Player%20Pawn) (Variables|Default|GetMistMatOverrideId))
      (Guid|InvalidateGuid (Variables|Default|GetMistMatOverrideId))
      (CallFunction|BPLogMistRaven "Mist Mat Removed" "Verbose")
      (else
        (CallFunction|BPLogMistRaven "No Mist Mat to invalidate" "Verbose")))))

(fn OnRep_bRepApplyWFFX
  (if (Variables|RepNotifyVars|GetbRepApplyWFFX)
    (CallFunction|ApplyWFFx)
    (else
      (CallFunction|DeactivateFX))))

(fn SendEndEventToPPCamera
  (bind _cameralens (Variables|camera|GetCameraLens))
  (bind _endingsoundcue (Variables|Tags|GetEndingSoundCue))
  (bind _as_fort_player_pawn (Variables|Default|GetAs%20Fort%20Player%20Pawn))
  (if (Utilities|IsValidInterface _cameralens)
    (bind _asbp_camera_lens_mist_raven_grim_strut_active (Utilities|Casting|CastToBP_CameraLens_MistRaven_GrimStrut_Active _cameralens)
      (:then
        (Class|BPCameraLensMistRavenGrimStrutActive|EVENTEndingGrimStrutCameraPP _asbp_camera_lens_mist_raven_grim_strut_active))
      (:CastFailed
        _))
    (if (Utilities|Time|IsValidTimerHandle (Variables|Handles|GetEndingSoundGameplayCueTimer))
      (GameplayCue|AddGameplayCue _as_fort_player_pawn _endingsoundcue)
      (GameplayCue|ExecuteGameplayCue _as_fort_player_pawn _endingsoundcue)
      (GameplayCue|ExecuteGameplayCue%28Burst%29Local _as_fort_player_pawn _endingsoundcue (Ability|GameplayCue|MakeGameplayCueParameters 0.0)) ; Cam shake + controller vibration
      (Utilities|Time|ClearandInvalidateTimerbyHandle (Variables|Handles|GetEndingSoundGameplayCueTimer)))
    (return)))

(fn SetExitAnim
  (bind _as_fort_player_pawn (Variables|Default|GetAs%20Fort%20Player%20Pawn))
  (bind (_returnvalue_x _returnvalue_y _returnvalue_z) (Math|Vector|BreakVector (Transformation|GetActorLocation _as_fort_player_pawn)))
  (Variables|SetBelowVector_X (Math|Vector|MakeVector _returnvalue_x _returnvalue_y (- _returnvalue_z 1.5e+02)))
  (bind (_outhit _returnvalue) (Collision|LineTraceByChannel (Transformation|GetActorLocation _as_fort_player_pawn) (LocalVariables|GetBelowVector) "TraceTypeQuery10"))
  (if _returnvalue
    (Variables|Default|SetExitAnimMontage "/MistRaven/Animation/MistRaven_GrimStrutOutro.MistRaven_GrimStrutOutro")
    (else
      (Variables|Default|SetExitAnimMontage "/MistRaven/Animation/MistRaven_GrimStrutOutroInAir.MistRaven_GrimStrutOutroInAir")))
  (return _returnvalue))

(fn UpdateTimerValue
  (params (CurrentTime :type Float%20%28double-precision%29) (bTimeDialate :type Boolean))
  (if bTimeDialate
    (return (* CurrentTime 2.0))
    (else
      (return (/ CurrentTime 2.0)))))
