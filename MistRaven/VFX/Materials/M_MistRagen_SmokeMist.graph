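# Unlit translucent Niagara sprite material for MistRaven smoke and mist, tinted by particle color.
# Opacity blends two smoke flipbook textures, modulated by tiled noise, particle alpha and depth fade.
# The IS_Decal switch scales emissive and squares opacity against an inverted depth fade, confining smoke near intersecting surfaces.
# The Alter Pos switch halves opacity and builds an offset from particle position to a Dynamic Parameter position, plus camera offset.
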Material /MistRaven/VFX/Materials/M_MistRagen_SmokeMist
graph {
 properties {bUsedWithNiagaraSprites=true, blendMode=BLEND_Translucent, pixelDepthOffsetMode=PDOM_Legacy, shadingModel=MSM_Unlit, usedShadingModels=Unlit}
 nodes [
  n1 = DynamicParameter() {defaultValue=<rgba 100 100 200 10>, paramNames [X, Y, Z, WPO Noise Strength]}
  Add(A=@n1.RGB, B=TransformPosition(None=VectorParameter() {parameterName=Ofs}.RGB) {transformType=TRANSFORMPOSSOURCE_World})
  n2 = ScalarParameter() {defaultValue=5, parameterName=Decal Emisive Scale}
  n3 = ParticleColor()
  MaterialFunctionCall(Emissive=StaticSwitchParameter(True=Multiply(A=@n2, B=@n3.RGB), False=@n3.RGB) {parameterName=IS_Decal}) {materialFunction=&MF_FX_EmissiveAdjustments}
  n4 = DepthFade(FadeDistance=ScalarParameter() {defaultValue=50, parameterName=Inv})
  Saturate(None=@n4)
  n5 = ParticlePositionWS()
  n6 = ScalarParameter() {defaultValue=15}
  n7 = WorldPosition()
  StaticSwitchParameter(
   True=Add(
    A=Subtract(A=@n1.RGB, B=@n5)
    B=MaterialFunctionCall(Offset Amount=@n6, World Position=@n7.XYZ) {materialFunction=/Engine/Functions/Engine_MaterialFunctions02/WorldPositionOffset/CameraOffset}.Camera Offset
   )
   False=Constant()
  ) {parameterName=Alter Pos}
  n8 = TextureSample() {samplerType=SAMPLERTYPE_LinearColor, texture=/Game/Effects/Fort_Effects/Gameplay/SmokeDustExplosions/Textures/T_SmokeDeath_8x8_01}
  n9 = Multiply(A=@n8.R, B=@n8.R)
  n10 = TextureSampleParameter2D() {parameterName=SubUV_Tex, samplerType=SAMPLERTYPE_LinearColor, texture=/Game/Effects/Fort_Effects/Textures/Smoke/T_Inky_Smoke_SubUV_8x8}
  n11 = StaticComponentMaskParameter(None=@n10.RGB) {defaultR=true, parameterName=Mask}
  n12 = ScalarParameter() {defaultValue=0.8, parameterName=Blend Lerp}
  n13 = ScalarParameter() {parameterName=LerpMin}
  n14 = ScalarParameter() {defaultValue=1, parameterName=LerpMax}
  n15 = LinearInterpolate(
   A=TextureSample(UVs=TextureCoordinate() {uTiling=1.5, vTiling=1.5}) {texture=/Game/Effects/Fort_Effects/Textures/Tile/T_Inky_Smoke_Tile2}.R
   Alpha=ScalarParameter() {defaultValue=1.25, parameterName=Noise Lerp}
  ) {constAlpha=0.1}
  n16 = Multiply(A=@n3.A, B=Multiply(A=LinearInterpolate(A=@n9, B=@n11, Alpha=@n12) {constAlpha=0.75}, B=LinearInterpolate(A=@n13, B=@n14, Alpha=@n15)))
  n17 = DepthFade(
   Opacity=StaticSwitchParameter(True=Multiply(A=@n16) {constB=0.5}, False=@n16) {parameterName=Alter Pos}
   FadeDistance=ScalarParameter() {defaultValue=100, parameterName=FadeDistance}
  ) {fadeDistanceDefault=5}
  StaticSwitchParameter(True=Multiply(A=OneMinus(None=@n4), B=Multiply(A=@n17, B=@n17)), False=@n17) {parameterName=IS_Decal}
 ]
}
