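# Material function for Fort effects that computes a noisy 3D world position offset, with a switch for Niagara use.
# It samples an animated, rotating Voronoi volume texture to shape normal-direction displacement, scaled by a dynamic parameter, and adds panning explosion-noise distortion.
# The output combines displacement, optionally faded toward the UV U edges, with a non-Niagara normal thinning term.
# Velocity bend, cloth-noise and levels branches remain unconnected to the output.
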MaterialFunction /Game/Effects/Fort_Effects/Materials/_Functions/MF_WPO
graph {nodes [
  n1 = NamedRerouteUsage()
  n2 = ComponentMask(None=TextureCoordinate()) {r=true}
  n3 = Abs(None=Multiply(A=Subtract(A=@n2) {constB=0.5}) {constB=2})
  n4 = StaticSwitchParameter(True=Multiply(A=@n1, B=OneMinus(None=@n3)), False=@n1) {parameterName=Displace Fade}
  Add(A=@n4)
  n5 = Constant()
  n6 = Multiply(A=VertexNormalWS(), B=ScalarParameter() {parameterName=Thin Control})
  FunctionOutput(None=Subtract(A=StaticSwitchParameter(True=@n5, False=@n6) {parameterName=IS Niagara}, B=@n4))
  n7 = VectorParameter() {parameterName=Volonoi Noise Scale, defaultValue=<rgba 0.01 0.01 0.01 0>}
  n8 = ScalarParameter() {defaultValue=4, parameterName=Main Noise Scale}
  n9 = Multiply(A=Time(), B=ScalarParameter() {defaultValue=10, parameterName=Main Displacement Anim Speed}) {constB=100}
  n10 = Transform(None=NamedRerouteUsage()) {transformSourceType=TRANSFORMSOURCE_World, transformType=TRANSFORM_Local}
  n11 = ScalarParameter() {defaultValue=0.4, parameterName=Main Noise Rotate Rate}
  n12 = AppendVector(
   A=Rotator(Coordinate=ComponentMask(None=@n10) {g=true, r=true}, Time=Multiply(A=@n9, B=@n11) {constB=0.15})
   B=ComponentMask(None=@n10) {b=true}
  )
  n13 = Multiply(A=Multiply(A=Multiply(A=@n7.RGB, B=@n8), B=Add(A=@n9, B=@n12))) {constB=0.5}
  n14 = TextureObjectParameter() {
   group=Volume Textures
   parameterName=Displacement Volume Tex
   samplerType=SAMPLERTYPE_LinearColor
   texture=/Game/Effects/Explosion_Update/Textures/VT_Voronoi_Seamless
  }
  n15 = OneMinus(None=Power(
    Base=TextureSample(UVs=@n13, Tex=@n14) {samplerType=SAMPLERTYPE_LinearColor, texture=&VT_Voronoi_Seamless}.R
    Exp=ScalarParameter() {defaultValue=3, parameterName=Shape Unevenness Scale}
   ) {constExponent=1.5})
  n16 = Constant() {r=0.4}
  n17 = Subtract(A=@n15, B=@n16)
  n18 = Constant()
  n19 = Constant() {r=1}
  Min(A=Divide(A=Max(A=@n17, B=@n18), B=Subtract(A=@n19, B=@n16)), B=Constant() {r=1}) {desc=LevelsControlInputRange}
  n20 = NamedRerouteUsage()
  n21 = Multiply(A=VectorParameter() {parameterName=Curl Noise Pan Dir, defaultValue=<rgba 1 5 -1 0>}.RGB, B=Time())
  n22 = TextureSample(UVs=Multiply(A=Add(A=@n20, B=@n21), B=ScalarParameter() {defaultValue=0.002, parameterName=Noise Size})) {samplerType=SAMPLERTYPE_LinearColor, texture=/Game/Global/MaterialFunctions/VT_ClothSimWPO}
  Multiply(A=Multiply(A=Subtract(A=@n22.R) {constB=0}, B=ScalarParameter() {defaultValue=1, parameterName=Second Noise Intensity}), B=@n15)
  n23 = Add(A=NamedRerouteUsage(), B=Multiply(A=VectorParameter() {parameterName=Curl Noise Pan Dir, defaultValue=<rgba 50 100 -30 0>}.RGB, B=Time()))
  n24 = ScalarParameter() {defaultValue=0.0011, parameterName=Noise Size}
  n25 = Subtract(A=TextureSample(UVs=Multiply(A=@n23, B=@n24)) {texture=/Game/Effects/Explosion_Update/Textures/VT_ExplosionNoise}.RGB) {constB=0.5}
  n26 = ScalarParameter() {defaultValue=1.5, parameterName=Noise Scale}
  n27 = VertexNormalWS()
  n28 = Multiply(
   A=StaticSwitchParameter(True=Transform(None=@n27) {transformSourceType=TRANSFORMSOURCE_World, transformType=TRANSFORM_Local}, False=@n27) {parameterName=IS Niagara}
   B=@n15
  )
  n29 = Multiply(
   A=ScalarParameter() {defaultValue=8, parameterName=Displacement}
   B=DynamicParameter() {parameterIndex=3, paramNames [Displacement,Mask_R,Mask_G,Mask_B]}.Displacement
  )
  NamedRerouteDeclaration(None=Add(A=Multiply(A=@n25, B=@n26), B=Multiply(A=@n28, B=@n29))) {name=3D Displacement, nodeColor=<rgba 0 1 0.9766 1>}
  NamedRerouteDeclaration(None=Subtract(A=WorldPosition().XYZ, B=ActorPositionWS())) {name=CurrentPos, nodeColor=<rgba 1 0.6563 0 1>}
  ObjectPositionWS()
  ParticlePositionWS()
  n30 = DynamicParameter() {parameterIndex=2, defaultValue=<rgba 1 1 1 0.1>, paramNames [X, Y, Z, Bend WPO Intensity]}
  n31 = MaterialFunctionCall(X=@n30.X, Y=@n30.X, Z=@n30.X) {materialFunction=&MakeFloat3}
  n32 = ParticlePositionWS()
  n33 = TextureCoordinate()
  n34 = Power(Base=OneMinus(None=ComponentMask(None=@n33) {g=true}), Exp=ScalarParameter() {defaultValue=2.5, group=WPO, parameterName=WPOBendPower})
  StaticSwitchParameter(True=Multiply(A=Subtract(A=@n31.Result, B=@n32), B=Multiply(A=@n34, B=@n30.Bend WPO Intensity)), False=Constant()) {defaultValue=true, group=WPO, parameterName=WPO From Velocity}
 ]}
