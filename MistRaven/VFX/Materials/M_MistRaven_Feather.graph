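# Translucent, two-sided, default-lit Niagara sprite material for MistRaven feather particles.
# Particle color tints a grayscale feather texture borrowed from the shared Fortnite effects motes.
# Particle alpha multiplies the texture alpha, then a squared Near_Camera_Fade result, driven by the Camera Fade Distance parameter, to fade sprites close to the camera.
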Material /MistRaven/VFX/Materials/M_MistRaven_Feather
graph {
 properties {
  bAutomaticallySetUsageInEditor=false
  bUsedWithNiagaraSprites=true
  blendMode=BLEND_Translucent
  twoSided=true
  usedShadingModels=Default Lit
 }
 nodes [
  n1 = ParticleColor()
  n2 = TextureSample() {samplerType=SAMPLERTYPE_LinearGrayscale, texture=/Game/Effects/Fort_Effects/Textures/Motes/T_Glider_Monarch_Feathers}
  Multiply(A=@n1.RGB, B=@n2.RGB)
  n3 = MaterialFunctionCall(Distance=ScalarParameter() {defaultValue=200, parameterName=Camera Fade Distance}) {materialFunction=&Near_Camera_Fade}
  Multiply(A=Multiply(A=@n1.A, B=@n2.A), B=Multiply(A=@n3.Result, B=@n3.Result))
 ]
}
