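# Translucent unlit flare material for Niagara sprites.
# A bright, desaturated radial core plus soft edge glow and flare mask texture are tinted by particle color, with switchable hue shift and two-color radial gradient, then passed through emissive adjustments.
# Particle alpha is shaped by depth fade from a scalar or Niagara dynamic parameter, with optional near-camera fade and mask-power opacity.
# A view-angle fade branch is also present.
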Material /MistRaven/VFX/Materials/M_MistRaven_Flare
graph {
 properties {
  bAutomaticallySetUsageInEditor=false
  bUsedWithNiagaraSprites=true
  blendMode=BLEND_Translucent
  shadingModel=MSM_Unlit
  usedShadingModels=Unlit
 }
 nodes [
  n1 = TextureCoordinate()
  n2 = Constant() {r=0.5}
  n3 = VectorParameter() {parameterName=RadialGradientColorSetting, channelNames {a="", b=P, g=H, r=R}, defaultValue=<rgba 1 0 1 1>}
  n4 = LinearInterpolate(
   A=VectorParameter() {parameterName=EdgeColor, defaultValue=<rgba 1 0 0.6939 1>}.RGB
   B=VectorParameter() {parameterName=CenterColor, defaultValue=<rgba 0 0.8019 1 1>}.RGB
   Alpha=Power(Base=SphereMask(A=@n1, B=@n2, Radius=@n3.R, Hardness=@n3.R) {attenuationRadius=1, hardnessPercent=0}, Exp=@n3.P)
  )
  n5 = ScalarParameter() {defaultValue=2, parameterName=CenterBrightness}
  n6 = Constant() {r=0.5}
  n7 = SphereMask(A=TextureCoordinate(), B=@n6, Radius=@n6) {hardnessPercent=0}
  n8 = Power(Base=@n7, Exp=ScalarParameter() {defaultValue=25, parameterName=CentertPower}) {constExponent=25}
  n9 = ParticleColor()
  n10 = Desaturation(None=@n9.RGB, Fraction=ScalarParameter() {parameterName=inputDesaturation})
  n11 = ScalarParameter() {defaultValue=0.3, parameterName=CenterDesaturation}
  n12 = Multiply(
   A=Power(Base=@n7, Exp=ScalarParameter() {defaultValue=4, parameterName=EdgePower}) {constExponent=4}
   B=ScalarParameter() {defaultValue=0.15, parameterName=EdgeBrightness}
  ) {constB=0.15}
  n13 = TextureSampleParameter2D(UVs=TextureCoordinate()) {parameterName=FlareMask, texture=/Game/Effects/Fort_Effects/Textures/Flares/T_Flare_Spikeball}
  n14 = Add(A=Multiply(A=Multiply(A=@n5, B=@n8) {constB=2}, B=Desaturation(None=@n10, Fraction=@n11)), B=Multiply(A=Add(A=@n12, B=@n13.RGB), B=@n10))
  n15 = StaticSwitchParameter(True=MaterialFunctionCall(Texture=@n14) {materialFunction=&HueShift}.Result, False=@n14) {parameterName=UseColorJitter?}
  MaterialFunctionCall(Emissive=StaticSwitchParameter(True=Multiply(A=@n4, B=@n15), False=@n15) {parameterName=UseRadialGradientColor}) {materialFunction=&MF_FX_EmissiveAdjustments}
  n16 = ScalarParameter() {defaultValue=30, parameterName=DepthFade}
  n17 = DynamicParameter() {desc=Index 1 Params for Niagara, parameterIndex=1, paramNames [DepthFadeDist, MaskPow, Param3, Param4]}
  n18 = StaticSwitchParameter(True=@n16, False=If(A=@n17.DepthFadeDist, B=Constant() {r=1}, "A > B"=@n17.DepthFadeDist, "A == B"=@n16, "A < B"=@n16)) {parameterName=SimpleDepthFade}
  n19 = Multiply(A=@n9.A, B=ShadingPathSwitch(Default=DepthFade(FadeDistance=@n18)))
  n20 = ScalarParameter() {parameterName=CameraFadeDistance}
  n21 = StaticSwitchParameter(
   True=Multiply(A=@n19, B=MaterialFunctionCall(Distance=@n20) {materialFunction=&Near_Camera_Fade}.Result)
   False=@n19
  ) {group=DetailedGradientMask, parameterName=UseCameraNearFade}
  Multiply(
   A=@n21
   B=MaterialFunctionCall("FadeSensitivity (optional)"=ScalarParameter() {parameterName="FadeSensitivity (optional) (S)"}) {materialFunction=&FadeBasedOnViewAngle}.Result
  )
  NamedRerouteDeclaration(None=@n13.R) {name=Flare Mask, nodeColor=<rgba 1 0 0.2109 1>}
  n22 = NamedRerouteUsage()
  StaticSwitchParameter(True=Multiply(A=@n21, B=Power(Base=@n22, Exp=@n17.MaskPow) {constExponent=1}), False=@n21) {parameterName=Use FlareMaskOpacity}
 ]
}
