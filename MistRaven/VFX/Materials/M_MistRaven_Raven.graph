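# Translucent, two-sided lit material for smoky mist ravens rendered as Niagara mesh particles, animating the birds entirely in the shader.
# UV-masked sine offsets and axis rotations animate wings, body, legs and tail, with dynamic parameters controlling glide-to-flap blend, speed and time offset.
# Vertex color G and B separate second and third ravens for offsets, pivots and phase.
# Panning noise distorts the smoke texture, while smoothstepped noise and optional near-camera fade shape opacity.
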Material /MistRaven/VFX/Materials/M_MistRaven_Raven
graph {
 properties {
  bAutomaticallySetUsageInEditor=false
  bOutputTranslucentVelocity=true
  bUsedWithNiagaraMeshParticles=true
  blendMode=BLEND_Translucent
  previewMesh=/MistRaven/VFX/Meshes/SM_MistRaven_L1
  twoSided=true
  usedShadingModels=Default Lit
 }
 nodes [
  n1 = NamedRerouteUsage()
  n2 = Add(A=NamedRerouteUsage(), B=Add(A=NamedRerouteUsage(), B=NamedRerouteUsage()))
  n3 = Transform(None=VectorParameter() {parameterName=Raven2 Offset}.RGB) {transformSourceType=TRANSFORMSOURCE_Local}
  n4 = VertexColor()
  n5 = DynamicParameter() {parameterIndex=3, paramNames [Raven2 Ofs Rand, Raven3 Ofs Rand, Emissive Scale, Fresnel Ctrl]}
  n6 = Multiply(A=@n4.G, B=@n5.Raven2 Ofs Rand)
  n7 = Transform(None=VectorParameter() {parameterName=Raven3 Offset}.RGB) {transformSourceType=TRANSFORMSOURCE_Local}
  n8 = Multiply(A=@n4.B, B=@n5.Raven3 Ofs Rand)
  Add(
   A=LinearInterpolate(A=Add(A=@n1, B=@n2), B=NamedRerouteUsage(), Alpha=NamedRerouteUsage())
   B=Add(A=Multiply(A=@n3, B=@n6), B=Multiply(A=@n7, B=@n8))
  )
  Constant() {r=0.9}
  n9 = ScalarParameter() {defaultValue=0.25, parameterName=Wing Smokey Turb}
  n10 = Multiply(A=ComponentMask(None=NamedRerouteUsage()) {g=true, r=true}, B=NamedRerouteUsage())
  n11 = TextureSample(UVs=Add(A=TextureCoordinate(), B=Multiply(A=@n9, B=@n10))) {texture=&/T_MistRaven_Smokey}
  n12 = ParticleColor()
  Multiply(
   A=MaterialFunctionCall(DiffuseColor=@n11.RGB, SpecularColor=@n11.RGB) {materialFunction=&ConvertFromDiffSpec}.BaseColor
   B=@n12.RGB
  )
  n13 = TextureCoordinate() {coordinateIndex=1}
  NamedRerouteDeclaration(None=Clamp(None=ComponentMask(None=@n13) {r=true}) {desc=Fasten Mask}) {name=Gradient Mask, nodeColor=<rgba 0 0.0625 1 1>}
  n14 = Transform(None=LinearInterpolate(A=Constant3Vector() {constant=<rgba 1 0.25 -0.5 0>}, B=Constant3Vector() {constant=<rgba -1 -0.25 0.5 0>}, Alpha=NamedRerouteUsage())) {transformSourceType=TRANSFORMSOURCE_Local}
  n15 = NamedRerouteUsage()
  n16 = Constant() {r=0.2}
  n17 = Multiply(A=Clamp(None=Add(A=@n15, B=@n16)), B=NamedRerouteUsage())
  n18 = Constant() {desc=Time Delay for Wing Top, r=0.15}
  n19 = OneMinus(None=@n15)
  n20 = VertexColor()
  n21 = Add(A=Multiply(A=@n20.G) {constB=0.4}, B=Multiply(A=@n20.B) {constB=0.3})
  n22 = Add(A=Add(A=NamedRerouteUsage(), B=Multiply(A=@n18, B=@n19)), B=@n21)
  n23 = Multiply(A=NamedRerouteUsage(), B=Add(A=Constant() {desc=Wing Anim Speed, r=1}, B=@n21))
  n24 = ConstantBiasScale(None=Sine(None=Multiply(A=@n22, B=@n23)))
  n25 = ScalarParameter() {defaultValue=2.3333, parameterName=WingsWPOExp}
  n26 = Multiply(A=ConstantBiasScale(None=Power(Base=@n24, Exp=@n25)) {bias=-0.5, scale=2}, B=Constant() {desc=Wing Anim Amount, r=0.1})
  n27 = VertexColor()
  n28 = ParticlePositionWS()
  n29 = Add(A=@n28, B=Transform(None=VectorParameter() {defaultValue=<rgba -100 100 0 0>}.RGB) {transformSourceType=TRANSFORMSOURCE_Local})
  n30 = Add(
   A=@n28
   B=Transform(None=VectorParameter() {parameterName=Parama, defaultValue=<rgba -100 -100 0 0>}.RGB) {transformSourceType=TRANSFORMSOURCE_Local}
  )
  n31 = If(A=@n27.G, "A > B"=@n30, "A == B"=@n30, "A < B"=@n28) {constB=0.5}
  NamedRerouteDeclaration(None=RotateAboutAxis(
    NormalizedRotationAxis=Normalize(VectorInput=@n14)
    RotationAngle=Multiply(A=@n17, B=@n26)
    PivotPoint=If(A=@n27.B, "A > B"=@n29, "A == B"=@n29, "A < B"=@n31) {constB=0.5}
    Position=WorldPosition().XYZ
   )) {name=Wings WPO Anim, nodeColor=<rgba 0.3828 1 0 1>}
  n32 = DynamicParameter() {parameterIndex=1, defaultValue=<rgba 0 1.6 0 1>, paramNames [Animation Lerp | Glide - Flap, Animation Speed, StepMin, StepMax]}
  NamedRerouteDeclaration(None=@n32.Animation Lerp | Glide - Flap) {name=Animation Lerp, nodeColor=<rgba 0.7344 0 1 1>}
  n33 = Constant3Vector() {constant=<rgba 0.25 1 0.25 0>}
  n34 = Constant() {desc=Wing Anim Speed, r=0.6}
  n35 = NamedRerouteUsage()
  n36 = Sine(None=Multiply(A=NamedRerouteUsage(), B=Multiply(A=@n34, B=@n35)))
  n37 = Constant() {desc=Wing Anim Amount, r=0.025}
  n38 = NamedRerouteUsage()
  n39 = Multiply(A=Constant() {r=0.2}, B=NamedRerouteUsage())
  n40 = RotateAboutAxis(
   NormalizedRotationAxis=Normalize(VectorInput=Transform(None=@n33) {transformSourceType=TRANSFORMSOURCE_Local})
   RotationAngle=Multiply(A=Multiply(A=@n36, B=@n37), B=Add(A=@n38, B=@n39))
   PivotPoint=MaterialFunctionCall() {materialFunction=&ComponentPivotLocation}.Result
   Position=WorldPosition().XYZ
  )
  n41 = NamedRerouteUsage()
  n42 = Constant3Vector() {constant=<rgba 0 0 1 0>}
  n43 = Add(A=NamedRerouteUsage(), B=Multiply(A=Constant() {r=-0.1}, B=NamedRerouteUsage()))
  n44 = Multiply(A=Constant() {desc=Body Anim Speed, r=0.5}, B=NamedRerouteUsage())
  n45 = Multiply(A=Normalize(VectorInput=Transform(None=@n42) {transformSourceType=TRANSFORMSOURCE_Local}), B=Sine(None=Multiply(A=@n43, B=@n44)))
  n46 = Constant() {desc=Body Anim Amount, r=6}
  NamedRerouteDeclaration(None=Add(A=Multiply(A=@n40, B=@n41), B=Multiply(A=@n45, B=@n46))) {name=Glide WPO Anim, nodeColor=<rgba 0 1 0.4141 1>}
  NamedRerouteDeclaration(None=@n32.StepMin) {name=StepMin, nodeColor=<rgba 0 1 0.7188 1>}
  NamedRerouteDeclaration(None=@n32.StepMax) {name=StepMax, nodeColor=<rgba 1 0.3984 0 1>}
  NamedRerouteDeclaration(None=Add(A=Time(), B=DynamicParameter() {parameterIndex=2, paramNames [Animation Offset, Param2, Param3, Param4]}.Animation Offset)) {name=Offset Time, nodeColor=<rgba 1 0 0.4219 1>}
  n47 = NamedRerouteUsage()
  n48 = OneMinus(None=ComponentMask(None=TextureCoordinate()) {g=true})
  n49 = Power(Base=OneMinus(None=Multiply(A=@n47, B=@n48))) {constExponent=9}
  n50 = Multiply(A=TextureCoordinate() {uTiling=0.5}) {constB=3}
  n51 = MaterialFunctionCall(X=Constant(), Y=ScalarParameter() {defaultValue=-0.6, parameterName=Pan Speed}) {materialFunction=&MakeFloat2}
  n52 = Multiply(A=@n51.Result) {constB=0.7}
  n53 = TextureObject() {texture=/Game/Global/Textures/Default/Noise/T_StringyEnergy_Noise}
  n54 = LinearInterpolate(
   A=TextureSample(UVs=Panner(Coordinate=@n50, Speed=@n52) {speedY=-0.25}, Tex=@n53) {texture=/Game/Effects/FogVolume/T_TileNoise}.RGB
   B=TextureSample(UVs=Panner(Coordinate=@n50, Speed=@n51.Result) {speedY=-0.25}, Tex=@n53) {texture=&T_TileNoise}.RGB
  )
  NamedRerouteDeclaration(None=Multiply(A=Multiply(A=@n47, B=@n49), B=Subtract(A=@n54) {constB=0.5})) {name=Masked Smokey Wing Turb, nodeColor=<rgba 0.0312 0 1 1>}
  n55 = NamedRerouteUsage()
  n56 = Multiply(A=OneMinus(None=@n55), B=@n55)
  NamedRerouteDeclaration(None=Saturate(None=Multiply(A=@n56) {constB=3})) {name=Side Mask, nodeColor=<rgba 1 0 0.1641 1>}
  n57 = Transform(None=Constant3Vector() {constant=<rgba 0 0 1 0>}) {transformSourceType=TRANSFORMSOURCE_Local}
  n58 = NamedRerouteUsage()
  n59 = Constant() {r=0.2}
  n60 = Constant() {desc=Leg Anim Speed, r=1}
  n61 = NamedRerouteUsage()
  n62 = Sine(None=Multiply(A=Add(A=@n58, B=@n59), B=Multiply(A=@n60, B=@n61)))
  n63 = Multiply(A=Multiply(A=Normalize(VectorInput=@n57), B=Constant() {desc=Tail Move Amount, r=2}), B=@n62)
  n64 = Multiply(A=NamedRerouteUsage(), B=NamedRerouteUsage())
  n65 = Normalize(VectorInput=Transform(None=Constant3Vector() {constant=<rgba 1 0 0 0>}) {transformSourceType=TRANSFORMSOURCE_Local})
  n66 = Multiply(A=NamedRerouteUsage(), B=Multiply(A=Constant() {desc=Tail Rot Speed, r=0.7}, B=NamedRerouteUsage()))
  n67 = Multiply(A=Multiply(A=Sine(None=@n66), B=Constant() {desc=Tail Rot Amount, r=0.025}), B=@n64)
  n68 = MaterialFunctionCall() {materialFunction=&ComponentPivotLocation}
  n69 = WorldPosition()
  NamedRerouteDeclaration(None=Add(A=Multiply(A=@n63, B=@n64), B=RotateAboutAxis(NormalizedRotationAxis=@n65, RotationAngle=@n67, PivotPoint=@n68.Result, Position=@n69.XYZ))) {name=Tail WPO Anim, nodeColor=<rgba 1 0 0.7031 1>}
  n70 = Floor(None=Add(A=TextureCoordinate() {coordinateIndex=1}, B=Constant() {r=0.5}))
  NamedRerouteDeclaration(None=Clamp(None=ComponentMask(None=@n70) {g=true}) {desc=Wings L R}) {name=Wing Mask, nodeColor=<rgba 0 1 0.9766 1>}
  n71 = Floor(None=Add(A=TextureCoordinate() {coordinateIndex=2}, B=Constant() {r=0.5}))
  n72 = OneMinus(None=@n71)
  n73 = ComponentMask(None=@n72) {r=true}
  n74 = ComponentMask(None=@n71) {g=true}
  NamedRerouteDeclaration(None=Clamp(None=Multiply(A=@n73, B=@n74)) {desc=Wings}) {name=Wing End, nodeColor=<rgba 1 0.6563 0 1>}
  n75 = ComponentMask(None=@n71) {r=true}
  n76 = ComponentMask(None=@n71) {g=true}
  NamedRerouteDeclaration(None=Clamp(None=Multiply(A=@n75, B=@n76)) {desc=Legs}) {name=Wing Start, nodeColor=<rgba 0.3359 0 1 1>}
  n77 = ComponentMask(None=@n72) {r=true}
  n78 = ComponentMask(None=@n72) {g=true}
  NamedRerouteDeclaration(None=Clamp(None=Multiply(A=@n77, B=@n78)) {desc=Body}) {name=Body Tail, nodeColor=<rgba 0 1 0.0156 1>}
  NamedRerouteDeclaration(None=@n32.Animation Speed) {name=Animation Speed, nodeColor=<rgba 1 0 0.3047 1>}
  n79 = Constant3Vector() {constant=<rgba -0.5 0 1 0>}
  n80 = Multiply(A=Normalize(VectorInput=Transform(None=@n79) {transformSourceType=TRANSFORMSOURCE_Local}), B=@n62)
  n81 = Constant() {desc=Leg Anim Amount, r=2}
  n82 = NamedRerouteUsage()
  n83 = NamedRerouteUsage()
  NamedRerouteDeclaration(None=Multiply(A=Multiply(A=@n80, B=@n81), B=Multiply(A=@n82, B=@n83))) {name=Legs WPO Anim, nodeColor=<rgba 0 0.625 1 1>}
  n84 = Constant3Vector() {constant=<rgba 1 0 0.5 0>}
  n85 = NamedRerouteUsage()
  n86 = Multiply(A=Constant() {desc=Body Anim Speed, r=1}, B=NamedRerouteUsage())
  n87 = Multiply(A=Normalize(VectorInput=Transform(None=@n84) {transformSourceType=TRANSFORMSOURCE_Local}), B=Sine(None=Multiply(A=@n85, B=@n86)))
  n88 = Constant() {desc=Body Anim Amount, r=3}
  n89 = Add(A=OneMinus(None=NamedRerouteUsage()), B=NamedRerouteUsage())
  NamedRerouteDeclaration(None=Multiply(A=Multiply(A=@n87, B=@n88), B=Clamp(None=@n89))) {name=Body WPO Anim, nodeColor=<rgba 0.9453 1 0 1>}
  n90 = TextureSample(UVs=TextureCoordinate() {uTiling=2, vTiling=2}) {texture=&T_StringyEnergy_Noise}
  OneMinus(None=@n90.R)
  n91 = MaterialFunctionCall(Distance=ScalarParameter() {defaultValue=450, parameterName=Camera Clip Distance}) {materialFunction=&Near_Camera_Fade}
  n92 = ScalarParameter() {defaultValue=2, parameterName="Distance Fade  Pow"}
  n93 = StaticSwitchParameter(True=Multiply(A=@n12.A, B=Power(Base=@n91.Result, Exp=@n92)), False=@n12.A) {defaultValue=true, parameterName=Use_Camera_Distance_Fade}
  n94 = Multiply(A=@n90.R, B=@n90.R)
  n95 = SmoothStep(Min=NamedRerouteUsage(), Max=NamedRerouteUsage(), Value=Multiply(A=@n94, B=@n94))
  Reroute(None=Multiply(A=Multiply(A=@n93, B=@n95), B=@n11.A))
 ]
}
