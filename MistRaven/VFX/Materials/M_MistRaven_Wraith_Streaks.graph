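# Translucent unlit, two-sided Niagara material rendering the Mist Raven wraith's wispy wing streaks.
# Noise-distorted flow and mask textures build time-of-day-tinted emissive, combined with a colored wing texture and optional outer glow.
# Opacity multiplies wing alpha, masks and camera, depth and view-angle fades, also driving refraction on higher-quality deferred paths.
# Niagara dynamic parameters feed world position offset: sine waves, trailing bend, and switchable root masking and axis-rotation wing flapping.
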Material /MistRaven/VFX/Materials/M_MistRaven_Wraith_Streaks
graph {
 properties {
  bUseTranslucencyVertexFog=false
  bUsedWithNiagaraMeshParticles=true
  bUsedWithNiagaraRibbons=true
  bUsedWithNiagaraSprites=true
  blendMode=BLEND_Translucent
  pixelDepthOffsetMode=PDOM_Legacy
  refractionMethod=RM_IndexOfRefraction
  shadingModel=MSM_Unlit
  twoSided=true
  usedShadingModels=Unlit
 }
 nodes [
  n1 = TextureCoordinate()
  n2 = VectorParameter() {group=Noise, parameterName=Noise UV, defaultValue=<rgba 1 0.5 0 -0.2>}
  n3 = DynamicParameter() {defaultValue=<rgba 0 0 1 0>, paramNames [UVTilingRandom, TimeOffsetRandom, WPOStrength, TrailingModeON]}
  n4 = MaterialFunctionCall(X=@n2.R, Y=Add(A=@n3.UVTilingRandom, B=@n2.G)) {materialFunction=&MakeFloat2}
  n5 = MaterialFunctionCall(X=@n2.B, Y=@n2.B) {materialFunction=&MakeFloat2}
  n6 = Add(A=@n3.TimeOffsetRandom, B=Time())
  n7 = TextureSampleParameter2D(UVs=Add(A=Multiply(A=@n1, B=@n4.Result), B=Multiply(A=@n5.Result, B=@n6))) {
   group=Noise
   parameterName=Noise Texture
   samplerType=SAMPLERTYPE_Normal
   texture=&T_Weap_Distort_N_02
  }
  n8 = VectorParameter() {parameterName=Base, channelNames {a=NoiseStrength, b=MaskMultiply, g=DepthFade, r=Refraction}, defaultValue=<rgba 0 50 1 0.2>}
  n9 = Multiply(A=ComponentMask(None=@n7.RGB) {g=true, r=true}, B=@n8.NoiseStrength)
  n10 = Multiply(A=@n9, B=Constant2Vector() {r=1})
  n11 = TextureCoordinate()
  Add(A=@n10, B=@n11)
  n12 = Multiply(
   A=VectorParameter() {parameterName=Outer Color, defaultValue=<rgba 2 0 0.02 1>}.RGB
   B=Multiply(A=NamedRerouteUsage(), B=ScalarParameter() {defaultValue=2, parameterName=Outer Mask Scaling})
  )
  n13 = VectorParameter() {group=Flow, parameterName=Flow UV, defaultValue=<rgba 1 1 0.5 3>}
  n14 = DynamicParameter() {defaultValue=<rgba 0 0 1 0>, paramNames [UVTilingRandom, TimeOffsetRandom, WPOStrength, TrailingModeON]}
  n15 = Multiply(
   A=TextureCoordinate()
   B=MaterialFunctionCall(X=@n13.R, Y=Add(A=@n14.UVTilingRandom, B=@n13.G)) {materialFunction=&MakeFloat2}.Result
  )
  n16 = Multiply(
   A=MaterialFunctionCall(X=@n13.B, Y=@n13.B) {materialFunction=&MakeFloat2}.Result
   B=Add(A=@n14.TimeOffsetRandom, B=Time())
  )
  n17 = TextureSampleParameter2D!(UVs=Add(A=Add(A=@n15, B=@n16), B=@n9), Level=ScalarParameter() {parameterName=Detial Blur}, Apply View MipBias) {
   group=Flow
   mipValueMode=TMVM_MipLevel
   parameterName=Flow Texture
   samplerType=SAMPLERTYPE_LinearColor
   texture=/Game/Effects/Fort_Effects/Textures/T_WavePattern_Chromatic
  }
  n18 = VectorParameter() {group=Flow, parameterName=Flow Refine, defaultValue=<rgba 1 1 1 0.5>}
  n19 = Power(Base=@n17.A, Exp=@n18.R)
  n20 = ScalarParameter() {defaultValue=-10, group=Flow, parameterName=Flow Desaturation}
  n21 = VectorParameter() {group=Flow, parameterName=Flow Color, channelNames {a=Add, b="", g="", r=""}, defaultValue=<rgba 1 1 1 0>}
  n22 = Multiply(
   A=LinearInterpolate(A=Multiply(A=@n19, B=@n18.G), B=Multiply(A=@n17.A, B=@n18.B), Alpha=@n18.A)
   B=Add(A=Desaturation(None=@n17.RGB, Fraction=@n20), B=@n21.Add)
  )
  n23 = VectorParameter() {group=Flow, parameterName=Inside Color, channelNames {a=SizePower, b="", g="", r=""}, defaultValue=<rgba 0.3802 0 0 8>}
  n24 = TextureCoordinate()
  n25 = MaterialFunctionCall(Float2=TextureCoordinate() {uTiling=1.2}) {materialFunction=&BreakOutFloat2Components}
  n26 = DynamicParameter() {defaultValue=<rgba 0 0 1 0>, paramNames [UVTilingRandom, TimeOffsetRandom, WPOStrength, TrailingModeON]}
  n27 = VectorParameter() {group=Mask, parameterName=Mask UV, defaultValue=<rgba 0.5 0.3 0.55 0>}
  n28 = Add(A=@n26.UVTilingRandom, B=@n27.R)
  n29 = Multiply(
   A=MaterialFunctionCall(X=@n25.G, Y=Add(A=@n25.R) {constB=0.1}) {materialFunction=&MakeFloat2}.Result
   B=Add(
    A=MaterialFunctionCall(X=@n28, Y=@n27.G) {materialFunction=&MakeFloat2}.Result
    B=ScalarParameter() {defaultValue=0.5, group=Mask, parameterName=Mask V ofs}
   )
  )
  n30 = Multiply(
   A=MaterialFunctionCall(X=@n27.B, Y=@n27.B) {materialFunction=&MakeFloat2}.Result
   B=Time()
  )
  n31 = TextureSampleParameter2D!(UVs=Add(A=@n10, B=Add(A=@n29, B=@n30)), Apply View MipBias) {
   group=Mask
   parameterName=Mask Texture
   samplerType=SAMPLERTYPE_Grayscale
   texture=/Game/Effects/Fort_Effects/Textures/Smoke/T_Smoke_Trail_Soft_NoClamp
  }
  n32 = Multiply(A=Saturate(None=ComponentMask(None=@n24) {g=true}), B=@n31.R)
  n33 = Power(Base=@n32, Exp=@n23.SizePower)
  n34 = ParticleColor()
  n35 = Multiply(A=LinearInterpolate(A=Multiply(A=@n22, B=@n21), B=@n23, Alpha=Saturate(None=@n33)), B=@n34.RGB)
  n36 = VectorParameter() {group=ColorTimeOfDay, parameterName=ColorTimeOfDay, channelNames {a=Night, b=Evening, g=Day, r=Morning}, defaultValue=<rgba 1 1 1 1>}
  n37 = Multiply(A=@n36.Night, B=@n35)
  n38 = ScalarParameter() {group=ColorTimeOfDay, parameterName=DesaturationNight}
  n39 = NamedRerouteUsage()
  n40 = Constant2Vector() {g=0.05, r=0.2}
  n41 = TextureSampleParameter2D(UVs=Add(A=TextureCoordinate(), B=Multiply(A=@n39, B=@n40))) {parameterName=MainTex, texture=&/T_MistRavens_WraithWing}
  n42 = Add(
   A=MaterialFunctionCall(
    Color_Morning=Multiply(A=@n35, B=@n36.Morning)
    Color_Day=Multiply(A=@n35, B=@n36.Day)
    Color_Evening=Multiply(A=@n35, B=@n36.Evening)
    Color_Night=Desaturation(None=@n37, Fraction=@n38)
   ) {materialFunction=&Color_Time_of_Day}.Result
   B=Multiply(A=VectorParameter() {parameterName=WingColor, defaultValue=<rgba 0.0561 0.5642 0.6823 1>}.RGB, B=@n41.R)
  )
  MaterialFunctionCall(Emissive=StaticSwitchParameter(True=Add(A=@n12, B=@n42), False=@n42) {parameterName=USE_OuterColor}) {materialFunction=&MF_FX_EmissiveAdjustments}
  n43 = Constant() {r=64}
  n44 = Add(A=TextureCoordinate(), B=Constant2Vector() {g=-0.05})
  n45 = Constant() {r=0.5}
  n46 = ScalarParameter() {defaultValue=0.5}
  n47 = Multiply(
   A=Multiply(A=SphereMask(A=@n44, B=@n45, Radius=@n46) {hardnessPercent=0}) {constB=1.1}
   B=Multiply(A=ScalarParameter() {defaultValue=1, group=Mask, parameterName=Mask AlphaScale}, B=@n32)
  )
  n48 = Multiply(
   A=Multiply(A=@n34.A, B=MaterialFunctionCall(Distance=@n43) {materialFunction=&Near_Camera_Fade}.Result)
   B=Saturate(None=Multiply(A=@n47, B=@n8.MaskMultiply))
  )
  n49 = DepthFade(Opacity=@n48, FadeDistance=@n8.DepthFade)
  n50 = ShadingPathSwitch(Default=@n49, Deferred=@n49, Forward=@n48, Mobile=@n48)
  n51 = ScalarParameter() {defaultValue=0.1, parameterName="FadeSensitivity (optional) (S)"}
  Multiply(
   A=@n41.A
   B=Multiply(
    A=@n50
    B=MaterialFunctionCall("FadeSensitivity (optional)"=@n51) {materialFunction=&FadeBasedOnViewAngle}.Result
   )
  )
  n52 = Add(
   A=TextureCoordinate() {uTiling=0.5, vTiling=0.5}
   B=AppendVector(
    A=ScalarParameter() {defaultValue=0.25, parameterName=TextureMask_AddU}
    B=ScalarParameter() {defaultValue=0.48, parameterName=TextureMask_AddV}
   )
  )
  n53 = ScalarParameter() {parameterName=LinearMask_AddU}
  n54 = ScalarParameter() {parameterName=LinearMask_AddV}
  n55 = ComponentMask(None=Add(A=TextureCoordinate(), B=AppendVector(A=@n53, B=@n54))) {g=true}
  n56 = Saturate(None=Add(
    A=TextureSample(UVs=@n52) {samplerSource=SSM_Clamp_WorldGroupSettings, texture=/Game/WaterAndWind/Textures/WeaponFireVectorMask}.R
    B=OneMinus(None=@n55)
   ))
  n57 = Multiply(A=@n56, B=@n56) {constB=2}
  n58 = ScalarParameter() {defaultValue=3, parameterName=Mask2 multiply}
  NamedRerouteDeclaration(None=Saturate(None=Multiply(A=@n57, B=@n58) {constB=5})) {name=RootMask, nodeColor=<rgba 0.9922 1 0 1>}
  NamedRerouteDeclaration(None=Multiply(A=@n17.R, B=NamedRerouteUsage())) {name=Noise Wave, nodeColor=<rgba 0.5469 0 1 1>}
  n59 = OneMinus(None=ComponentMask(None=@n24) {r=true})
  n60 = Multiply(A=@n59, B=@n59)
  NamedRerouteDeclaration(None=Multiply(A=Multiply(A=@n60, B=@n60), B=ScalarParameter() {parameterName=Vib Mask})) {name=Vib_Mask, nodeColor=<rgba 0 1 0.0391 1>}
  n61 = VectorParameter() {group=Wing Flapping, parameterName=Flap3}
  n62 = DynamicParameter() {desc=Parameters set by Niagara, parameterIndex=1, defaultValue=<rgba 0 1 1 1>, paramNames [Side,Y,Z,WPOIntensity]}
  n63 = WorldPosition()
  n64 = ParticlePositionWS()
  n65 = Subtract(A=@n63.XYZ, B=@n64)
  n66 = Multiply(A=ComponentMask(None=Transform(None=@n65) {transformSourceType=TRANSFORMSOURCE_World, transformType=TRANSFORM_Local}) {g=true}) {constB=-1}
  n67 = Saturate(None=Multiply(A=@n62.Side, B=Multiply(B=@n66) {constA=0.003}))
  n68 = TextureCoordinate() {coordinateIndex=1}
  n69 = Multiply(A=Subtract(A=ComponentMask(None=@n68) {r=true}) {constB=0.01}) {constB=1.4}
  n70 = NamedRerouteDeclaration(None=Power(Base=Saturate(None=@n69)) {constExponent=0.5}) {name=Wing Mask, nodeColor=<rgba 1 0.0469 0 1>}
  n71 = Normalize(VectorInput=Transform(None=Constant3Vector() {constant=<rgba 1 0 0 1>}) {transformSourceType=TRANSFORMSOURCE_Local})
  n72 = ComponentMask(None=TextureCoordinate() {coordinateIndex=1}) {r=true}
  n73 = Saturate(None=Add(A=@n72, B=ScalarParameter() {defaultValue=0.3, group=Wing Flapping, parameterName=Add}))
  n74 = Constant() {desc=Time Delay for Wing Top, r=0.15}
  n75 = OneMinus(None=@n72)
  n76 = Multiply(
   A=Add(
    A=DynamicParameter() {parameterIndex=3, paramNames [WorldPosition_SphereMaskRadius control, orientation, Flapping Time, AppearMask]}.Flapping Time
    B=Multiply(A=@n74, B=@n75)
   )
   B=ScalarParameter() {defaultValue=1.51, group=Wing Flapping, parameterName=Anim Speed}
  )
  n77 = ConstantBiasScale(None=Sine(None=@n76)) {bias=1.5, scale=0.55}
  n78 = ScalarParameter() {defaultValue=1.9, group=Wing Flapping, parameterName=WingsWPOExp}
  n79 = Multiply(
   A=ConstantBiasScale(None=Power(Base=@n77, Exp=@n78)) {bias=-0.5, scale=2}
   B=ScalarParameter() {defaultValue=0.04, parameterName=Flap Adjustment}
  )
  n80 = Multiply(
   A=LinearInterpolate(A=Multiply(A=@n73, B=@n79) {constA=1}, Alpha=ScalarParameter() {parameterName=Bend})
   B=ScalarParameter() {defaultValue=0.5, parameterName=Angle Scale}
  )
  n81 = Add(A=@n64, B=Transform(None=VectorParameter() {group=Wing Flapping, parameterName=Ofs}.RGB) {transformSourceType=TRANSFORMSOURCE_Local})
  n82 = Multiply(
   A=@n70
   B=Multiply(
    A=RotateAboutAxis(NormalizedRotationAxis=@n71, RotationAngle=@n80, PivotPoint=@n81, Position=@n63.XYZ)
    B=ScalarParameter() {defaultValue=1, group=Wing Flapping, parameterName=Flapping WPO Scaling}
   )
  )
  n83 = VectorParameter() {group=Wing Flapping, parameterName=Flap2}
  n84 = DynamicParameter() {desc=Parameters set by Niagara, parameterIndex=1, defaultValue=<rgba 0 1 1 1>, paramNames [Side,Y,Z,WPOIntensity]}
  n85 = NamedRerouteUsage()
  n86 = Multiply(A=Multiply(A=@n70, B=Multiply(A=@n83.RGB, B=@n84.WPOIntensity)), B=Multiply(A=Add(A=@n85) {constB=-0.1}) {constB=-1})
  NamedRerouteDeclaration(None=Add(A=Multiply(A=@n61.RGB, B=@n67), B=Add(A=@n82, B=@n86))) {name=Wings WPO Anim, nodeColor=<rgba 0.3828 1 0 1>}
  n87 = Sine(None=Add(A=@n76, B=ScalarParameter() {defaultValue=0.5, group=Wing Flapping, parameterName=Time Ofs}))
  NamedRerouteDeclaration(None=Power(Base=ConstantBiasScale(None=@n87) {bias=1.4, scale=0.55}, Exp=@n78)) {name=Flap2, nodeColor=<rgba 0 0.0391 1 1>}
  n88 = OneMinus(None=Saturate(None=@n47))
  n89 = Power(
   Base=OneMinus(None=TextureSampleParameter2D(UVs=@n11) {parameterName=Base_mask, texture=/Game/Effects/Fort_Effects/Textures/Masks/T_Square_Gradient_M}.R)
   Exp=ScalarParameter() {defaultValue=5, parameterName=Param_1}
  ) {constExponent=5}
  n90 = Power(Base=@n56) {constExponent=5}
  NamedRerouteDeclaration(None=Multiply(A=Power(Base=@n88) {constExponent=0.6}, B=Multiply(A=@n89, B=@n90))) {name=UV Distortion, nodeColor=<rgba 0 1 0.625 1>}
  n91 = OneMinus(None=MaterialFunctionCall(Distance=ScalarParameter() {defaultValue=256, parameterName="Distance (S)"}) {materialFunction=&Near_Camera_Fade}.Result)
  n92 = LinearInterpolate(Alpha=Multiply(A=Multiply(A=@n50, B=@n91), B=@n8.Refraction)) {constA=1, constB=1.05}
  n93 = Constant() {r=1}
  n94 = QualitySwitch(Default=@n92, Low=@n93, High=@n92, Medium=@n93, Epic=@n92)
  ShadingPathSwitch(Default=@n94, Deferred=@n94, Forward=@n93, Mobile=@n93)
  n95 = Constant3Vector()
  n96 = QualitySwitch(Default=@n7.RGB, Low=@n95, High=@n7.RGB, Medium=@n95, Epic=@n7.RGB)
  ShadingPathSwitch(Default=@n96, Deferred=@n96, Forward=@n95, Mobile=@n95)
  n97 = NamedRerouteUsage()
  n98 = Multiply(A=@n97, B=@n97)
  n99 = VertexNormalWS()
  n100 = DynamicParameter() {defaultValue=<rgba 0 0 1 0>, paramNames [UVTilingRandom, TimeOffsetRandom, WPOStrength, TrailingModeON]}
  n101 = Add(A=Time(), B=@n100.TimeOffsetRandom)
  n102 = VectorParameter() {group=WPO, parameterName=WPO Sine, channelNames {a="", b=Sine Strength, g=Sine Tile, r=Sine Speed}, defaultValue=<rgba 2 2 1 1>}
  n103 = Add(A=@n102.Sine Speed) {constB=1.3}
  n104 = ComponentMask(None=TextureCoordinate()) {g=true}
  n105 = ScalarParameter() {parameterName=SIne Mask}
  n106 = Add(A=@n100.UVTilingRandom, B=Saturate(None=Add(A=@n104, B=@n105)))
  n107 = Add(A=@n102.Sine Tile) {constB=1.5}
  n108 = Sine(None=Add(A=Multiply(A=@n101, B=@n103), B=Multiply(A=@n106, B=@n107) {constB=2}))
  n109 = Multiply(A=DynamicParameter() {desc=Parameters set by Niagara, parameterIndex=2, paramNames [X,Y,Z,Velocity]}.Velocity) {constB=0.001}
  n110 = Multiply(A=NamedRerouteUsage(), B=Multiply(A=ScalarParameter() {parameterName=Noise Wave}, B=Clamp(None=@n109) {maxDefault=2}))
  n111 = Add(A=@n108, B=@n110)
  n112 = Add(A=@n108, B=Multiply(A=@n110) {constB=0.3})
  n113 = TextureCoordinate()
  n114 = LinearInterpolate(
   A=Constant3Vector()
   B=Multiply(
    A=MaterialFunctionCall(X=@n111, Y=@n111, Z=@n112) {materialFunction=&MakeFloat3}.Result
    B=@n102.Sine Strength
   ) {constB=0.2}
   Alpha=OneMinus(None=ComponentMask(None=@n113) {g=true})
  )
  n115 = DynamicParameter() {defaultValue=<rgba 0 0 1 0>, paramNames [UVTilingRandom, TimeOffsetRandom, WPOStrength, TrailingModeON]}
  n116 = Add(A=Multiply(A=Multiply(A=@n99, B=@n114), B=@n115.WPOStrength) {constB=10})
  n117 = DynamicParameter() {desc=Parameters set by Niagara, parameterIndex=2, paramNames [X,Y,Z,Velocity]}
  n118 = VectorParameter() {group=WPO, parameterName=WPO Control, channelNames {a=Speed, b=BendPower, g=Max, r=Min}, defaultValue=<rgba -50 50 3 3>}
  n119 = Clamp(
   None=Subtract(
    A=Constant3Vector() {constant=<rgba 1 1 1 0>}
    B=MaterialFunctionCall(X=@n117.X, Y=@n117.X, Z=@n117.X) {materialFunction=&MakeFloat3}.Result
   )
   Min=@n118.Min
   Max=@n118.Min
  ) {maxDefault=50, minDefault=-50}
  n120 = TextureCoordinate()
  n121 = Power(Base=OneMinus(None=ComponentMask(None=@n120) {g=true}), Exp=@n118.BendPower)
  n122 = LinearInterpolate(A=@n116, B=Add(A=LinearInterpolate(B=@n119, Alpha=@n121), B=@n116), Alpha=@n115.TrailingModeON)
  n123 = StaticSwitchParameter(True=Multiply(A=Multiply(A=@n98, B=@n98), B=@n122), False=@n122) {parameterName=WindRootMask}
  StaticSwitchParameter(True=Add(A=NamedRerouteUsage(), B=@n123), False=@n123) {parameterName=Wing Flap Anim}
 ]
}
