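# Niagara effect for MistRaven's DarkCoil hands, drawing short-lived flash sprites on a character's skeletal mesh.
# Only Bright_Flash_Decal has content: burst and rate spawns are partitioned to index bone sampling, re-run each update over filtered arm, hand and leg bones.
# User.Ascended picks the sprite color, and User.Ammo scales the alpha fade curve.
# Camera offset and animated dynamic material parameters shape the look; the other five emitters are empty.
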NiagaraSystem /MistRaven/VFX/NS_MistRaven_DarkCoil_Hands
dependencies {
 usedDataInterfaces [&NiagaraDataInterfaceSkeletalMesh, &NiagaraDataInterfaceCurve]
 usedDynamicInputs [
  &MakeFloatFromVector
  &MakeCustomFloatFromBool
  /Niagara/DynamicInputs/Add/Add_Vector2
  &Multiply_Vector2D_ByFloat
  &LerpLinearColors
  &Add_Float
  &Sine
 ]
 usedModules [
  &self:SystemSpawnScript.SystemScriptSource.SystemScriptGraph.NiagaraNodeAssignment_1.SetVariables_#1
  &SystemState
  &SampleSkeletalMesh
  &self:SystemSpawnScript.SystemScriptSource.SystemScriptGraph.NiagaraNodeAssignment_0.SetVariables_#2
 ]
}
emitters {
 Bright_Flash_Decal {
  emitterUpdateScript [
   &EmitterState(Inactive Response=::NewEnumerator1, Life Cycle Mode=::NewEnumerator1, Loop Delay=0.1, Loop Duration=6, UseLoopDelay=false)
   &SpawnRate(SpawnRate=20, Use Spawn Probability=false)
   &SpawnBurst_Instantaneous(Spawn Count=13, Spawn Time=0.1, Use Spawn Probability=false)
  ]
  particleSpawnScript [
   &PartitionParticles(Number Of Partitions=13)
   /Niagara/Modules/Spawn/Initialization/InitializeParticle(
    Color=&MakeCustomLinearColorFromBool(Bool=$User.Ascended, True Color=<rgba 0.5039 0 2 1>, False Color=<rgba 2 0 0.0404 1>)
    Lifetime=&RandomRangeFloat(Minimum=0.7, Maximum=0.8)
    Mass=&RandomRangeFloat(Minimum=0.6, Maximum=1.2)
    Module.Write Color=true
    Module.Write Lifetime=true
    Module.Write Mass=false
    Module.Write Material Random=true
    Module.Write Position=false
    Module.Write Scale=false
    Module.Write SpriteRotation=true
    Module.Write SpriteSize=true
    Sprite Rotation=&RandomRangeFloat(Minimum=0, Maximum=360)
    Sprite Size=&Vector2DFromFloat(Value=&Multiply_Float(A=80, B=&RandomRangeFloat(Minimum=1, Maximum=1.2)))
   )
   &CameraOffset(Camera Offset Amount=80)
  ]
  particleUpdateScript [
   &ParticleState()
   &SkeletalMeshLocation(
    Bone / Socket Index=$Particles.PartitionParticles.Partition
    Bone Sampling Mode=::NewEnumerator4
    MaskSpawnGroup=false
    Orientation Sampling=::NewEnumerator1
    Skeletal Mesh=DataInterface() {propertyValues {
      DefaultMesh=null
      ExcludeBoneName=root
      OptionalComponentClass=null
      PreviewMesh=/Game/Characters/Player/Female/Medium/Bodies/F_MED_Default_Soldier/Meshes/F_MED_Default_Soldier
      SkinningMode=SkinOnTheFly
      SoftSourceActor=null
      SourceMode=Default
      UvSetIndex=0
      WholeMeshLOD=-1
      bExcludeBone=false
      bReadDeformedGeometry=true
      bRequireCurrentFrameData=false
      ComponentTags []
      FilteredBones [upperarm_l,upperarm_r,lowerarm_l,lowerarm_r,hand_l,hand_r,thigh_l,thigh_r,calf_l,calf_r]
      FilteredSockets []
      MeshUserParameter {parameter {name=None, typeDefHandle {registeredTypeIndex=1}, varData []}}
      SamplingRegions []
     }}
   )
   &ScaleColor(
    Scale Alpha=&FloatFromCurve(
     FloatCurve=DataInterface() {propertyValues {Curve {keys [
         {arriveTangent=0.5269, interpMode=RCIM_Cubic, leaveTangent=0.5269, tangentMode=RCTM_User, time=0, value=1}
         {arriveTangent=-0.2619, interpMode=RCIM_Cubic, leaveTangent=-0.2619, tangentMode=RCTM_User, time=0.3, value=3}
         <at 1 0 RCIM_Cubic>
        ]}}}
     Scale Curve=$User.Ammo
    )
    Scale RGB=<0 0 0>
    ScaleA=true
    ScaleRGB=false
   )
   &DynamicMaterialParameters(
    Index 0 Param 1=&FloatFromCurve(FloatCurve=DataInterface() {propertyValues {Curve {keys [<at 0 5 RCIM_Cubic>,<at 0.0296 1.5 RCIM_Cubic>,<at 1 5 RCIM_Cubic>]}}})
    Index 0 Param 2=1
    Index 0 Param 3=&FloatFromCurve(FloatCurve=DataInterface() {propertyValues {Curve {keys [<at 0 0.35>,<at 1 0>]}}})
    Index 0 Param 4=&RandomRangeFloat(Minimum=0, Maximum=1)
    Index 1 Param 1=160
    Index 1 Param 2=5
    Module.Index0Float0Write=true
    Module.Index0Float1Write=true
    Module.Index0Float2Write=true
    Module.Index0Float3Write=true
    Module.Index1Float0Write=true
    Module.Index1Float1Write=true
    Module.Index1Float2Write=true
    Module.Index1Float3Write=true
    Write Parameter Index 1=::NewEnumerator1
   )
  ]
  rendererClasses [&NiagaraSpriteRendererProperties]
  renderers [{rendererClass=&NiagaraSpriteRendererProperties}]
 }
 Empty001 {rendererClasses [], renderers []}
 LWE_L_Connect {rendererClasses [], renderers []}
 LWE_R_Connect {rendererClasses [], renderers []}
 LW_Wraith_Streaks {rendererClasses [], renderers []}
 LW_Wraith_Streaks001 {rendererClasses [], renderers []}
}
summary {
 systemName=NS_MistRaven_DarkCoil_Hands
 emitters [
  {emitterName=Empty001, rendererClasses []}
  {emitterName=LW_Wraith_Streaks, rendererClasses []}
  {emitterName=LW_Wraith_Streaks001, rendererClasses []}
  {emitterName=LWE_R_Connect, rendererClasses []}
  {emitterName=LWE_L_Connect, rendererClasses []}
  {emitterName=Bright_Flash_Decal, rendererClasses [&NiagaraSpriteRendererProperties]}
 ]
 userVariables [
  User.Ammo {type=NiagaraFloat, defaultValue=1, description=""}
  User.Ascended {type=NiagaraBool, defaultValue=true}
 ]
}
