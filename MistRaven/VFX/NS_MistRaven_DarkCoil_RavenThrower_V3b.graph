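# Niagara effect for MistRaven's Dark Coil raven thrower, spraying a cone of mesh-and-sprite ravens along User.TargetVector at User.Velocity Scale.
# Lead Raven_Thrower particles seed tumbling feathers and darker secondary ravens via particle-read spawning, with GPU-sim counterparts for the PS4 variants.
# On collision, lead ravens stop and hide, while a per-particle timeline drives hit flashes.
# User.Ascended switches palettes and sprite size; secondary trails fade by camera-to-velocity angle.
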NiagaraSystem /MistRaven/VFX/NS_MistRaven_DarkCoil_RavenThrower_V3b
dependencies {
 usedDataInterfaces [
  &NiagaraDataInterfaceCurve
  &NiagaraDataInterfaceVector2DCurve
  &NiagaraDataInterfaceVectorCurve
  &NiagaraDataInterfaceColorCurve
  &NiagaraDataInterfaceSpriteRendererInfo
  &NiagaraDataInterfaceParticleRead
  &NiagaraDataInterfaceCamera
  &NiagaraDataInterfaceArrayColor
 ]
 usedDynamicInputs [&InvertBool]
 usedModules [
  &self:SystemSpawnScript.SystemScriptSource.SystemScriptGraph.NiagaraNodeAssignment_1.SetVariables_#1
  &SystemState
  &self:SystemSpawnScript.SystemScriptSource.SystemScriptGraph.NiagaraNodeAssignment_0.SetVariables_#2
 ]
}
emitters {
 Feather {
  emitterUpdateScript [
   &EmitterState()
   &SpawnParticlesFromOtherEmitter(
    Attribute Reader=DataInterface() {propertyValues {EmitterBinding {bindingMode=Other, emitterName=Raven_Thrower}}}
    Spawn Rate=20
    Spawn Rate Per Particle Cap=150
    Use Maximum Age=false
    Use Minimum Age=false
   )
  ]
  particleSpawnScript [
   &InitializeParticle(
    Color=&MakeCustomLinearColorFromBool(
     Bool=$User.Ascended
     True Color=&SelectLinearColorFromArray(
      Array Sampling Mode=::NewEnumerator0
      Direct Array Index=?
      Color Selection Array=DataInterface() {propertyValues {ColorData [<rgba 0.0177 0.0129 0.0273 1>, <rgba 0.0208 0.0077 0.0081 1>, <rgba 0.0003 0.0097 0.026 1>]}}
      Random Seed=0
     )
     False Color=&SelectLinearColorFromArray(
      Array Sampling Mode=::NewEnumerator0
      Direct Array Index=?
      Color Selection Array=DataInterface() {propertyValues {ColorData [<rgba 0.0177 0.0129 0.0273 1>, <rgba 0.0052 0 0.0001 1>, <rgba 0.0003 0.0097 0.026 1>]}}
      Random Seed=0
     )
    )
    Lifetime Max=1
    Lifetime Min=0.9
    Lifetime Mode=::NewEnumerator1
    Sprite Rotation Mode=::NewEnumerator2
    Sprite Size Mode=::NewEnumerator1
    Sprite UV Mode=::NewEnumerator3
    Uniform Sprite Size Max=16
    Uniform Sprite Size Min=10
    UsePositionOffset=false
    Write Material Random=true
   )
   &SampleParticlesFromOtherEmitter(
    Applied Position Offset:Vector3f=&Multiply_VectorByFloat(Vector=&RandomRangeVector(Minimum=<-5 -5 -5>, Maximum=<5 5 5>), Float=$User.Velocity Scale)
    Applied Velocity Scale:NiagaraFloat=&RandomRangeFloat(Minimum=0.2, Maximum=0.4)
    Velocity Sampling=::NewEnumerator0
   )
   &CameraOffset(Camera Offset Amount=-50)
   &SubUVAnimation(
    End Frame Range Override=30
    Evaluation Type=::NewEnumerator0
    "Sprite Renderer "=DataInterface() {propertyValues {SpriteRenderer=&self:Feather_0.NiagaraSpriteRendererProperties_2}}
    Start Frame Range Override=&RandomRangeInt(Minimum=0, Maximum=5)
    SubUV Animation Mode:ENiagara_SubUVLookupModeV2=::NewEnumerator2
    UseEndFrame=true
    UseStartFrame=true
   )
  ]
  particleUpdateScript [
   &ParticleState()
   &SpriteRotationRate(Rotation Rate=&UniformAOrBFloat(
     A=&RandomRangeFloat(Minimum=-1500, Maximum=-100)
     B=&RandomRangeFloat(Minimum=1500, Maximum=100)
     Evaluation Type=::NewEnumerator0
     Recalculate Random Each Loop=true
     Override Seed=false
    ))
   &ScaleSpriteSize(Uniform Curve Sprite Scale=DataInterface() {propertyValues {Curve {keys [
        {arriveTangent=3.2121, interpMode=RCIM_Cubic, leaveTangent=3.2121, tangentMode=RCTM_User, time=0, value=0.3}
        {tangentMode=RCTM_User, time=0.25, value=1.7}
       ]}}})
   &CurlNoiseForce(Bias Noise=false, Noise Quality / Cost=::NewEnumerator0, Noise Strength=1000, Pan Noise=false)
   &Drag(Drag=4, Use Linear Drag=true, Use Rotational Drag=false)
   &ScaleColor(
    Linear Color Curve=DataInterface() {propertyValues {
      AlphaCurve {keys [<at 0.0183 0>,<at 0.1211 2>,<at 0.8721 0.75>,<at 1 0>]}
      BlueCurve {keys [<at 0.1779 1>,<at 0.351 1>,<at 0.626 0>]}
      GreenCurve {keys [<at 0.1779 1>,<at 0.351 1>,<at 0.626 0>]}
      RedCurve {keys [<at 0.1779 1>,<at 0.351 1>,<at 0.626 0>]}
     }}
    Scale Mode=::NewEnumerator2
   )
   &GravityForce(Gravity=&Multiply_VectorByFloat(Vector=<0 0 -980>, Float=&RandomRangeFloat(Minimum=0.2, Maximum=0.5)))
   &SolveForcesAndVelocity(Clamp Velocity=false, Limit Acceleration=false)
  ]
  rendererClasses [&NiagaraSpriteRendererProperties]
  renderers [{rendererClass=&NiagaraSpriteRendererProperties}]
 }
 Feather_PS4 {
  simTarget=GPUComputeSim
  emitterUpdateScript [
   &EmitterState()
   &SpawnParticlesFromOtherEmitter(
    Attribute Reader=DataInterface() {propertyValues {EmitterBinding {bindingMode=Other, emitterName=Raven_Thrower_PS4}}}
    Spawn Rate=45
    Spawn Rate Per Particle Cap=150
    Use Maximum Age=false
    Use Minimum Age=false
   )
  ]
  particleSpawnScript [
   &InitializeParticle(
    Color=&MakeCustomLinearColorFromBool(
     Bool=$User.Ascended
     True Color=&SelectLinearColorFromArray(
      Array Sampling Mode=::NewEnumerator0
      Direct Array Index=?
      Color Selection Array=DataInterface() {propertyValues {ColorData [<rgba 0.0177 0.0129 0.0273 1>, <rgba 1 0 0.0278 1>, <rgba 0.0003 0.0097 0.026 1>]}}
      Random Seed=0
     )
     False Color=&SelectLinearColorFromArray(
      Array Sampling Mode=::NewEnumerator0
      Direct Array Index=?
      Color Selection Array=DataInterface() {propertyValues {ColorData [<rgba 0.0177 0.0129 0.0273 1>, <rgba 0.1562 0 0.0043 1>, <rgba 0.0003 0.0097 0.026 1>]}}
      Random Seed=0
     )
    )
    Lifetime Max=1
    Lifetime Min=0.9
    Lifetime Mode=::NewEnumerator1
    Sprite Rotation Mode=::NewEnumerator2
    Sprite Size Mode=::NewEnumerator1
    Sprite UV Mode=::NewEnumerator3
    Uniform Sprite Size Max=16
    Uniform Sprite Size Min=10
    UsePositionOffset=false
    Write Material Random=true
   )
   &SampleParticlesFromOtherEmitter(
    Applied Position Offset:Vector3f=&Multiply_VectorByFloat(Vector=&RandomRangeVector(Minimum=<-5 -5 -5>, Maximum=<5 5 5>), Float=$User.Velocity Scale)
    Applied Velocity Scale:NiagaraFloat=&RandomRangeFloat(Minimum=0.2, Maximum=0.4)
    Velocity Sampling=::NewEnumerator0
   )
   &CameraOffset(Camera Offset Amount=-50)
   &SubUVAnimation(
    End Frame Range Override=30
    Evaluation Type=::NewEnumerator0
    "Sprite Renderer "=DataInterface() {propertyValues {SpriteRenderer=&self:Feather_PS4_0.NiagaraSpriteRendererProperties_2}}
    Start Frame Range Override=&RandomRangeInt(Minimum=0, Maximum=5)
    SubUV Animation Mode:ENiagara_SubUVLookupModeV2=::NewEnumerator2
    UseEndFrame=true
    UseStartFrame=true
   )
  ]
  particleUpdateScript [
   &ParticleState()
   &SpriteRotationRate(Rotation Rate=&UniformAOrBFloat(
     A=&RandomRangeFloat(Minimum=-1500, Maximum=-100)
     B=&RandomRangeFloat(Minimum=1500, Maximum=100)
     Evaluation Type=::NewEnumerator0
     Recalculate Random Each Loop=true
     Override Seed=false
    ))
   &ScaleSpriteSize(Uniform Curve Sprite Scale=DataInterface() {propertyValues {Curve {keys [
        {arriveTangent=3.2121, interpMode=RCIM_Cubic, leaveTangent=3.2121, tangentMode=RCTM_User, time=0, value=0.3}
        {tangentMode=RCTM_User, time=0.25, value=1.7}
       ]}}})
   &CurlNoiseForce(Bias Noise=false, Noise Quality / Cost=::NewEnumerator0, Noise Strength=1000, Pan Noise=false)
   &Drag(Drag=4, Use Linear Drag=true, Use Rotational Drag=false)
   &ScaleColor(
    Linear Color Curve=DataInterface() {propertyValues {
      AlphaCurve {keys [<at 0.0183 0>,<at 0.1211 2>,<at 0.8721 0.75>,<at 1 0>]}
      BlueCurve {keys [<at 0.1779 1>,<at 0.351 1>,<at 0.626 0>]}
      GreenCurve {keys [<at 0.1779 1>,<at 0.351 1>,<at 0.626 0>]}
      RedCurve {keys [<at 0.1779 1>,<at 0.351 1>,<at 0.626 0>]}
     }}
    Scale Mode=::NewEnumerator2
   )
   &InheritSourceMovement(Applied Position Scale=<0.8 0.8 0.8>, ApplyForce=false, ApplyPosition=true, ApplyPositionFalloff=false)
   &GravityForce(Gravity=&Multiply_VectorByFloat(Vector=<0 0 -980>, Float=&RandomRangeFloat(Minimum=0.2, Maximum=0.5)))
   &SolveForcesAndVelocity(Clamp Velocity=false, Limit Acceleration=false)
  ]
  rendererClasses [&NiagaraSpriteRendererProperties]
  renderers [{rendererClass=&NiagaraSpriteRendererProperties}]
 }
 Raven_Thrower {
  emitterUpdateScript [
   &EmitterState(Life Cycle Mode=::NewEnumerator1, Loop Behavior=::NewEnumerator1, Loop Delay=0.2, Loop Duration Mode=::NewEnumerator1, UseLoopDelay=false)
   &SpawnRate(
    SpawnRate=&Multiply_Float(A=&RandomRangeFloat(Minimum=5.5, Maximum=6.5, Evaluation Type=::NewEnumerator1), B=1)
    Use Spawn Probability=false
   )
   &SpawnBurst_Instantaneous(
    Spawn Count=&RandomRangeInt(Minimum=1, Maximum=1)
    Spawn Group:NiagaraInt32=&ReturnExecIndex()
    Use Spawn Probability=false
    UseLoopCountLimit=false
   )
  ]
  particleSpawnScript [
   &InitializeParticle(
    Color=&MakeCustomLinearColorFromBool(Bool=$User.Ascended, True Color=<rgba 0.7481 0 2.5 1>, False Color=<rgba 5 0 0.2 1>)
    Lifetime=&Divide_Float(A=8, B=$User.Velocity Scale)
    Mesh Scale Mode=::NewEnumerator3
    Mesh Uniform Scale=&RandomRangeFloat(Minimum=0.4, Maximum=0.6)
    Sprite Size Mode=::NewEnumerator3
    Uniform Sprite Size=&MakeCustomFloatFromBool(Bool=$User.Ascended, True Float=300, False Float=200)
    UsePositionOffset=false
    Write Material Random=true
   )
   &ShapeLocation(
    Hemisphere Distribution=<0.5 1>
    MaskSpawnGroup=false
    Offset Coordinate Space=::Local
    Offset Mode=::NewEnumerator0
    Override Seed=false
    Rotation Axis=<0 0 1>
    Rotation Mode=::NewEnumerator1
    Sphere Radius=25
    UseOwnerScale=true
   )
   &InitialMeshOrientation(Mesh Orientation Mode=::NewEnumerator2, Use Rotation Vector=false)
   &AddVelocity(
    Cone Angle=$User.Cone Angle
    Cone Axis=$User.TargetVector
    Module.UseConeFalloff=false
    Module.UseVelDistribution=false
    Override Seed=false
    Velocity Mode=::NewEnumerator2
    Velocity Speed=&Multiply_Float(A=250, B=$User.Velocity Scale)
   )
   &CameraOffset(Camera Offset Amount=-25)
   /SetVariables(Particles.CamOfs2=30) {isSetParameters=true}
  ]
  particleUpdateScript [
   &ParticleState()
   &ScaleSpriteSize(
    Non-Uniform Curve Scale=<1.5 1>
    Non-Uniform Curve Sprite Scale=DataInterface() {propertyValues {
      XCurve {keys [<at 0.05 0 RCIM_Cubic>, {tangentMode=RCTM_User, time=0.3, value=1}]}
      YCurve {keys [<at 0.05 0 RCIM_Cubic>, {tangentMode=RCTM_User, time=0.3, value=1}]}
     }}
    Scale Sprite Size Mode=::NewEnumerator3
   )
   &ScaleMeshSize(Scale Factor=&Multiply_Vector(
     A=&VectorFromFloat(Value=&FloatFromCurve(
       FloatCurve=DataInterface() {propertyValues {Curve {defaultValue=1, keys [{tangentMode=RCTM_User, time=0, value=0.3}, <at 0.1 0.5>, {tangentMode=RCTM_User, time=0.2, value=1}]}}}
       Scale Curve=$System.Timeline.InverseNormalizedTime
      ))
     B=&VectorFromCurve(VectorCurve=DataInterface() {propertyValues {XCurve {keys [<at 0.1 1.5>,<at 0.2 1>,<at 1 1>]}, YCurve {keys [<at 0 1>,<at 1 1>]}, ZCurve {keys [<at 0 1>,<at 1 1>]}}})
    ))
   &Collision(
    Age Colliding Particles=false
    CPU Collision Trace Channel=ECC_WorldStatic
    Friction=50000
    KillOnCollision=false
    Maximum Penetration Correction Distance=50000
    Module.Control Rest State Via Surface Normal=false
    Module.EnableMaxCollisionCount=false
    Module.Randomize Collision Normal Vector=false
    Particle Radius=2
    Particle Radius Scale=2
    Percentage of Penetration Before Rest=99999
    Radius Calculation Type=::NewEnumerator3
    Restitution=0
   )
   &CurlNoiseForce(Bias Noise=false, Noise Quality / Cost=::NewEnumerator0, Noise Strength=500, Pan Noise=false)
   &DynamicMaterialParameters(
    Index 1 Param 1=&FloatFromCurve(FloatCurve=DataInterface() {propertyValues {Curve {keys [<at 0 1>,<at 0.3 0>]}}})
    Index 1 Param 2=&RandomRangeFloat(Minimum=2, Maximum=3)
    Index 1 Param 3=&FloatFromCurve(
     FloatCurve=DataInterface() {propertyValues {Curve {keys [<at 0.2 5>,<at 0.3 5>,<at 1 0.5>]}}}
     Scale Curve=&FloatFromCurve(
      FloatCurve=DataInterface() {propertyValues {Curve {defaultValue=1, keys [{tangentMode=RCTM_User, time=0, value=0.3}, {tangentMode=RCTM_User, time=0.15, value=1}]}}}
      CurveIndex=$Particles.Age
     )
    )
    Index 1 Param 4=-1
    Index 2 Param 1=&RandomRangeFloat(Minimum=0, Maximum=1111)
    Index 3 Param 1=&RandomRangeFloat(Minimum=-0.5, Maximum=3)
    Index 3 Param 2=&RandomRangeFloat(Minimum=0, Maximum=1)
    Module.Index0Float0Write=true
    Module.Index0Float1Write=true
    Module.Index0Float2Write=false
    Module.Index0Float3Write=true
    Module.Index1Float0Write=true
    Module.Index1Float1Write=true
    Module.Index1Float2Write=true
    Module.Index1Float3Write=true
    Module.Index2Float0Write=true
    Module.Index2Float1Write=false
    Module.Index2Float2Write=false
    Module.Index2Float3Write=false
    Module.Index3Float0Write=true
    Module.Index3Float1Write=true
    Module.Index3Float2Write=true
    Module.Index3Float3Write=true
    Write Parameter Index 1=::NewEnumerator1
    Write Parameter Index 2=::NewEnumerator1
    Write Parameter Index 3=::NewEnumerator1
   )
   /SetVariables(Particles.ColorBlack:LinearColor=&MakeLinearColorFromVectorAndFloat(
     "Vector (RGB)"=&MakeVectorFromLinearColorRGB(Color=&ColorFromCurve(ColorCurve=DataInterface() {propertyValues {
         AlphaCurve {keys [<at 0 0 RCIM_Cubic>]}
         BlueCurve {keys [<at 0.745 0 RCIM_Cubic>]}
         GreenCurve {keys [<at 0.745 0 RCIM_Cubic>]}
         RedCurve {keys [<at 0.745 0 RCIM_Cubic>]}
        }}))
     "Float (Alpha)"=&MakeCustomFloatFromBool(
      Bool=$Particles.HasCollided
      True Float=0
      False Float=&FloatFromCurve(FloatCurve=DataInterface() {propertyValues {Curve {keys [<at 0.55 1>,<at 1 0>]}}}, Scale Curve=2)
     )
    )) {isSetParameters=true, module=SetVariables_#3}
   /SetVariables(
    Particles.DMP_Trail:Vector4f=&MakeVector4(
     X=0
     Y=-0.1
     Z=1
     W=&FloatFromCurve(FloatCurve=DataInterface() {propertyValues {Curve {defaultValue=1, keys [{tangentMode=RCTM_User, time=0.05, value=0}, {tangentMode=RCTM_User, time=1, value=1}]}}})
    )
    Particles.DMP_Trail1:Vector4f=&MakeVector4(X=1, Y=1, Z=1, W=1)
    Particles.DMP_Trail2:Vector4f=&MakeVector4(X=1, Y=1, Z=1, W=1)
    Particles.DMP_Trail3:Vector4f=&MakeVector4(X=500, Y=0, Z=0, W=0)
    Particles.Trail Color:LinearColor=&MultiplyLinearColor(
     Linear Color=&MakeCustomLinearColorFromBool(Bool=$User.Ascended, True Color=<rgba 2 0 0.9407 1>, False Color=<rgba 0.9219 0.024 0.024 1>)
     Scale Factor=&MakeVector4(
      X=1
      Y=1
      Z=1
      W=&MakeCustomFloatFromBool(
       Bool=$Particles.HasCollided
       True Float=0
       False Float=&FloatFromCurve(FloatCurve=DataInterface() {propertyValues {Curve {keys [<at 0.9 1>,<at 1 0>]}}})
      )
     )
    )
    Particles.Trail Orientation:Quat4f=&MakeQuaternion(
     Quaternion From=0
     Axis=<1 0 0>
     Angle=&Add_Float(
      A=&RandomRangeFloat(Minimum=0, Maximum=360)
      B=&Multiply_Float(A=$Engine.Time, B=1000)
     )
     From=?
     To=?
     Angle Type=::NewEnumerator0
     Normalize Vectors=false
    )
    Particles.Trail Scale:Vector3f=&VectorFromCurve(VectorCurve=DataInterface() {propertyValues {
       XCurve {defaultValue=1, keys [{tangentMode=RCTM_User, time=0, value=1.25}, {tangentMode=RCTM_User, time=0.2, value=2.375}, <at 1 0.5>]}
       YCurve {defaultValue=1, keys [{tangentMode=RCTM_User, time=0, value=0.25}, {tangentMode=RCTM_User, time=0.2, value=1}, <at 1 2>]}
       ZCurve {defaultValue=1, keys [{tangentMode=RCTM_User, time=0, value=0.25}, {tangentMode=RCTM_User, time=0.2, value=1}, <at 1 2>]}
      }})
    Particles.Velocity=&MakeCustomVectorFromBool(Bool=$Particles.HasCollided, True Vector=<0 0 0>, False Vector=$Particles.Velocity)
   ) {isSetParameters=true, module=SetVariables_#4}
   &Timeline(Play=$Particles.HasCollided, Play Rate=2, Rewind Mode:ENiagara_TimelineMode=::NewEnumerator1)
   &ScaleColor(
    Scale RGB=&Multiply_VectorByFloat(
     Vector=&MakeVectorFromLinearColorRGB(Color=&ColorFromCurve(ColorCurve=DataInterface() {propertyValues {
         AlphaCurve {keys [<at 0 0.2>, {arriveTangent=-0.2, interpMode=RCIM_Cubic, leaveTangent=-0.2, time=0.1889, value=1}, <at 1 0 RCIM_Cubic>]}
         BlueCurve {keys [<at 0 1 RCIM_Cubic>]}
         GreenCurve {keys [<at 0 1 RCIM_Cubic>]}
         RedCurve {keys [<at 0 1 RCIM_Cubic>]}
        }}))
     Float=&MakeCustomFloatFromBool(Bool=$Particles.HasCollided, True Float=0, False Float=1)
    )
    ScaleA=false
    ScaleRGB=true
   )
   /SetVariables(
    Particles.Hit Color:LinearColor=&MakeLinearColorFromVectorAndFloat(
     "Vector (RGB)"=&MakeCustomVectorFromBool(
      Bool=$User.Ascended
      True Vector=&VectorFromCurve(
       VectorCurve=DataInterface() {propertyValues {
         XCurve {keys [{arriveTangent=-3.5433, time=0, value=0.625}, <at 0.5 0.1>]}
         YCurve {keys [<at 0.01 0.1>,<at 0.5 0>]}
         ZCurve {keys [<at 0 1.375>,<at 0.5 0.375>]}
        }}
       CurveIndex=$Particles.Timeline.NormalizedTime
      )
      False Vector=&VectorFromCurve(
       VectorCurve=DataInterface() {propertyValues {
         XCurve {keys [{arriveTangent=-12.0139, interpMode=RCIM_Cubic, leaveTangent=-12.0139, tangentMode=RCTM_User, time=0.01, value=2}, <at 0.5 0>]}
         YCurve {keys [<at 0.01 0.1>,<at 0.5 0>]}
         ZCurve {keys [<at 0.01 0.1>,<at 0.5 0>]}
        }}
       CurveIndex=$Particles.Timeline.NormalizedTime
      )
     )
     "Float (Alpha)"=&Multiply_Float(A=5, B=$Particles.Timeline.InverseNormalizedTime)
    )
    Particles.Hit Sprite:Vector2f=&Vector2DFromCurve(
     Vector2Curve=DataInterface() {propertyValues {
       XCurve {
        defaultValue=1
        keys [
         {arriveTangent=4.1785, interpMode=RCIM_Cubic, leaveTangent=4.1785, tangentMode=RCTM_User, time=0.01, value=0}
         {arriveTangent=1.1179, interpMode=RCIM_Cubic, leaveTangent=1.1179, tangentMode=RCTM_User, time=0.3, value=0.65}
         {arriveTangent=0.0224, interpMode=RCIM_Cubic, leaveTangent=0.0224, tangentMode=RCTM_User, time=1, value=1}
        ]
       }
       YCurve {
        defaultValue=1
        keys [
         {arriveTangent=4.1785, interpMode=RCIM_Cubic, leaveTangent=4.1785, tangentMode=RCTM_User, time=0.01, value=0}
         {arriveTangent=1.1179, interpMode=RCIM_Cubic, leaveTangent=1.1179, tangentMode=RCTM_User, time=0.3, value=0.65}
         {arriveTangent=0.0224, interpMode=RCIM_Cubic, leaveTangent=0.0224, tangentMode=RCTM_User, time=1, value=1}
        ]
       }
      }}
     CurveIndex=$Particles.Timeline.NormalizedTime
     ScaleCurve=<120 120>
    )
   ) {isSetParameters=true}
   &SubUVAnimation(
    Random Start Frame=true
    "Sprite Renderer "=DataInterface() {propertyValues {SpriteRenderer=&self:Raven_Thrower_0.NiagaraSpriteRendererProperties_1}}
    SubUV Lookup Index:NiagaraFloat=$StackContext.Timeline.NormalizedTime
    UseEndFrame=false
    UseStartFrame=false
   )
   &SolveForcesAndVelocity(Clamp Velocity=false, Limit Acceleration=false)
   &KillParticlesInVolume(Invert Volume=true, Sphere Radius:NiagaraFloat=$User.Raven Kill Distance) {enabled=false}
  ]
  rendererClasses [&NiagaraMeshRendererProperties, &NiagaraSpriteRendererProperties]
  renderers [
   {rendererClass=&NiagaraMeshRendererProperties, rendererIndex=0}
   {rendererClass=&NiagaraSpriteRendererProperties, rendererIndex=1}
   {rendererClass=&NiagaraMeshRendererProperties, rendererIndex=2}
   {rendererClass=&NiagaraSpriteRendererProperties, rendererIndex=3}
   {rendererClass=&NiagaraMeshRendererProperties, rendererIndex=4}
  ]
 }
 Raven_Thrower001 {
  emitterUpdateScript [
   &EmitterState(Life Cycle Mode=::NewEnumerator1, Loop Behavior=::NewEnumerator1, Loop Delay=0.07, Loop Duration Mode=::NewEnumerator1, UseLoopDelay=false)
   &SpawnParticlesFromOtherEmitter(
    Attribute Reader=DataInterface() {propertyValues {EmitterBinding {bindingMode=Other, emitterName=Raven_Thrower}}}
    Spawn Rate=&Multiply_Float(A=$User.Velocity Scale, B=1.2)
    Use Maximum Age=false
    Use Minimum Age=false
   )
  ]
  particleSpawnScript [
   &InitializeParticle(
    Color=<rgba 0.0024 0.002 0.005 2>
    Lifetime=&Divide_Float(A=15, B=$User.Velocity Scale)
    Mesh Scale Mode=::NewEnumerator3
    Mesh Uniform Scale=&RandomRangeFloat(Minimum=0.5, Maximum=0.65)
    Sprite Rotation Mode=::NewEnumerator2
    Sprite Size Mode=::NewEnumerator3
    Sprite UV Mode=::NewEnumerator3
    Uniform Sprite Size=110
    UsePositionOffset=false
    Write Material Random=true
   )
   &InitialMeshOrientation(Mesh Orientation Mode=::NewEnumerator2, Use Rotation Vector=false)
   &AddVelocity(
    Cone Angle=35
    Cone Axis=$User.TargetVector
    Module.UseConeFalloff=false
    Module.UseVelDistribution=false
    Override Seed=false
    Velocity Mode=::NewEnumerator2
    Velocity Speed=&Multiply_Float(A=&RandomRangeFloat(Minimum=30, Maximum=80), B=$User.Velocity Scale)
   )
   &SampleParticlesFromOtherEmitter(Lifetime Sampling:ENiagara_DirectReadSamplingMode=::NewEnumerator1)
   &ShapeLocation(
    Hemisphere Distribution=<0.5 1>
    MaskSpawnGroup=false
    Non Uniform Scale=<1.3 1.8 1.05>
    Offset=<-250 0 0>
    Offset Coordinate Space=::Local
    Offset Mode=::NewEnumerator0
    Override Seed=false
    Rotation Mode=::NewEnumerator1
    Sphere Radius=50
    UseOwnerScale=true
   )
   &CameraOffset(Camera Offset Amount=-50)
  ]
  particleUpdateScript [
   &ParticleState()
   &ScaleMeshSize(Scale Factor=&VectorFromFloat(Value=&FloatFromCurve(
      FloatCurve=DataInterface() {propertyValues {Curve {defaultValue=1, keys [{tangentMode=RCTM_User, time=0.07, value=0}, {tangentMode=RCTM_User, time=0.25, value=1}, <at 0.65 1>, <at 1 0>]}}}
      Scale Curve=$System.Timeline.InverseNormalizedTime
     )))
   &DynamicMaterialParameters(
    Index 1 Param 1=&FloatFromCurve(FloatCurve=DataInterface() {propertyValues {Curve {keys [<at 0 1>,<at 0.3 0>]}}})
    Index 1 Param 2=&RandomRangeFloat(Minimum=3.3, Maximum=4.2)
    Index 1 Param 3=&FloatFromCurve(FloatCurve=DataInterface() {propertyValues {Curve {keys [<at 0.11 1>,<at 0.16 -1.5>,<at 0.8 -1.5>,<at 1 1>]}}})
    Index 1 Param 4=1
    Index 2 Param 1=&RandomRangeFloat(Minimum=0, Maximum=1111)
    Module.Index1Float0Write=true
    Module.Index1Float1Write=true
    Module.Index1Float2Write=true
    Module.Index1Float3Write=true
    Module.Index2Float0Write=true
    Module.Index2Float1Write=false
    Module.Index2Float2Write=false
    Module.Index2Float3Write=false
    Module.Index3Float0Write=false
    Module.Index3Float1Write=false
    Module.Index3Float2Write=true
    Module.Index3Float3Write=true
    Write Parameter Index 0=::NewEnumerator0
    Write Parameter Index 1=::NewEnumerator1
    Write Parameter Index 2=::NewEnumerator1
    Write Parameter Index 3=::NewEnumerator1
   )
   &ScaleColor(
    Linear Color Curve=DataInterface() {propertyValues {
      AlphaCurve {keys [<at 0.0101 0>,<at 0.1087 1>,<at 0.4705 1 RCIM_Cubic>,<at 1 0 RCIM_Cubic>]}
      BlueCurve {keys [<at 0 1 RCIM_Cubic>]}
      GreenCurve {keys [<at 0 1 RCIM_Cubic>]}
      RedCurve {keys [<at 0 1 RCIM_Cubic>]}
     }}
    Scale Mode=::NewEnumerator2
   )
   &Collision(
    Age Colliding Particles=false
    Collision Enabled=&SetBoolByFloatComparison(A=$Particles.NormalizedAge, B=0.001)
    KillOnCollision=true
    Module.Control Rest State Via Surface Normal=false
    Module.EnableMaxCollisionCount=false
    Module.Randomize Collision Normal Vector=false
    Restitution=0
   )
   /SetVariables(Particles.ColorBlack:LinearColor=&MakeLinearColorFromVectorAndFloat(
     "Vector (RGB)"=<0 0 0>
     "Float (Alpha)"=&MakeCustomFloatFromBool(
      Bool=$Particles.HasCollided
      True Float=0
      False Float=&FloatFromCurve(FloatCurve=DataInterface() {propertyValues {Curve {keys [
           <at 0 0 RCIM_Cubic>
           {arriveTangent=4, leaveTangent=4, time=0.1, value=2}
           {arriveTangent=-8.1145, interpMode=RCIM_Cubic, leaveTangent=-8.1145, tangentMode=RCTM_User, time=0.9, value=2}
           <at 1 0 RCIM_Cubic>
          ]}}})
     )
    )) {isSetParameters=true, module=SetVariables_#5}
   &CurlNoiseForce(Bias Noise=false, Noise Quality / Cost=::NewEnumerator0, Noise Strength=400, Pan Noise=false)
   /SetVariables(
    Particles.DMP_Trail:Vector4f=&MakeVector4(
     X=0
     Y=-0.1
     Z=0
     W=&FloatFromCurve(FloatCurve=DataInterface() {propertyValues {Curve {defaultValue=1, keys [{tangentMode=RCTM_User, time=0.05, value=0}, {tangentMode=RCTM_User, time=1, value=1}]}}})
    )
    Particles.DMP_Trail1:Vector4f=&MakeVector4(X=1, Y=1, Z=1, W=1)
    Particles.DMP_Trail2:Vector4f=&MakeVector4(X=1, Y=1, Z=1, W=1)
    Particles.DMP_Trail3:Vector4f=&MakeVector4(X=100, Y=0, Z=0, W=0)
    Particles.Trail Color:LinearColor=&MultiplyLinearColor(
     Linear Color=&MakeCustomLinearColorFromBool(Bool=$User.Ascended, True Color=<rgba 0.1931 0 2 0.8>, False Color=<rgba 0.1771 0.0046 0.0046 0.8>)
     Scale Factor=&MakeVector4(
      X=1
      Y=1
      Z=1
      W=&MakeCustomFloatFromBool(
       Bool=$Particles.HasCollided
       True Float=0
       False Float=&FloatFromCurve(
        FloatCurve=DataInterface() {propertyValues {Curve {keys [<at 0.85 1>,<at 1 0>]}}}
        Scale Curve=&OneMinusFloat(Float=&Power(
          a=&Abs_Float(Float:NiagaraFloat=&DotProduct(
            Vector A=&ReturnCameraVector(Camera Vectors=::NewEnumerator1, Coordinate Space=::Simulation)
            Vector B=&NormalizeVector(Vector To Normalize=$Particles.Velocity, Safe Normalize=true, Safe Fallback Threshold=0, Safe Fallback Vector=<0 0 0.001>)
           ))
          b=$User.Culling Alpha Power
         ))
       )
      )
     )
    )
    Particles.Trail Orientation:Quat4f=&MakeQuaternion(
     Quaternion From=0
     Axis=<1 0 0>
     Angle=&Add_Float(A=&RandomRangeFloat(Minimum=0, Maximum=360), B=0)
     From=?
     To=?
     Angle Type=::NewEnumerator0
     Normalize Vectors=false
    )
    Particles.Trail Scale:Vector3f=&VectorFromCurve(
     VectorCurve=DataInterface() {propertyValues {
       XCurve {defaultValue=1, keys [{tangentMode=RCTM_User, time=0, value=0.625}, {tangentMode=RCTM_User, time=0.1, value=0.75}, <at 1 0.5>]}
       YCurve {defaultValue=1, keys [{tangentMode=RCTM_User, time=0, value=0.25}, {tangentMode=RCTM_User, time=0.2, value=1}, <at 1 2>]}
       ZCurve {defaultValue=1, keys [{tangentMode=RCTM_User, time=0, value=0.25}, {tangentMode=RCTM_User, time=0.2, value=1}, <at 1 2>]}
      }}
     Scale Curve=<0.5 0.05 0.05>
    )
    Particles.Velocity=&MakeCustomVectorFromBool(Bool=$Particles.HasCollided, True Vector=<0 0 0>, False Vector=$Particles.Velocity)
   ) {isSetParameters=true, module=SetVariables_#6}
   &SolveForcesAndVelocity(Clamp Velocity=false, Limit Acceleration=false)
   &KillParticlesInVolume(Invert Volume=true, Sphere Radius=1800)
   &KillParticles(Kill Particles=&SetBoolByFloatComparison(A=100, B=&VectorLength(Vector=$Particles.Velocity))) {enabled=false}
   /SetVariables(Particles.smoke Position:NiagaraPosition=&Add_Vector(
     A=Linked() {linkedVariable {name=Particles.Position, type=NiagaraPosition}}
     B=&TransformVector(Vector=<50 0 -5>, Destination Space=::Simulation)
    )) {isSetParameters=true, module=SetVariables_#7}
   &SubUVAnimation(
    Random Start Frame=true
    "Sprite Renderer "=DataInterface() {propertyValues {SpriteRenderer=&self:Raven_Thrower001_0.NiagaraSpriteRendererProperties_0}}
    UseEndFrame=false
    UseStartFrame=false
   )
  ]
  rendererClasses [&NiagaraMeshRendererProperties, &NiagaraSpriteRendererProperties]
  renderers [
   {rendererClass=&NiagaraMeshRendererProperties, rendererIndex=0}
   {rendererClass=&NiagaraMeshRendererProperties, rendererIndex=1}
   {rendererClass=&NiagaraMeshRendererProperties, rendererIndex=2}
   {rendererClass=&NiagaraSpriteRendererProperties, rendererIndex=3}
  ]
 }
 Raven_Thrower002 {
  simTarget=GPUComputeSim
  emitterUpdateScript [
   &EmitterState(Life Cycle Mode=::NewEnumerator1, Loop Behavior=::NewEnumerator1, Loop Delay=0.07, Loop Duration Mode=::NewEnumerator1, UseLoopDelay=false)
   &SpawnParticlesFromOtherEmitter(
    Attribute Reader=DataInterface() {propertyValues {EmitterBinding {bindingMode=Other, emitterName=Raven_Thrower_PS4}}}
    Spawn Rate=&Multiply_Float(A=$User.Velocity Scale, B=1.1)
    Use Maximum Age=false
    Use Minimum Age=false
   )
  ]
  particleSpawnScript [
   &InitializeParticle(
    Color=<rgba 0.0024 0.002 0.005 2>
    Lifetime=&Divide_Float(A=15, B=$User.Velocity Scale)
    Mesh Scale Mode=::NewEnumerator3
    Mesh Uniform Scale=&RandomRangeFloat(Minimum=0.5, Maximum=0.65)
    Sprite Rotation Mode=::NewEnumerator2
    Sprite Size Mode=::NewEnumerator3
    Sprite UV Mode=::NewEnumerator3
    Uniform Sprite Size=110
    UsePositionOffset=false
    Write Material Random=true
   )
   &InitialMeshOrientation(Mesh Orientation Mode=::NewEnumerator2, Use Rotation Vector=false)
   &AddVelocity(
    Cone Angle=25
    Cone Axis=$User.TargetVector
    Module.UseConeFalloff=false
    Module.UseVelDistribution=false
    Override Seed=false
    Velocity Mode=::NewEnumerator2
    Velocity Speed=&Multiply_Float(A=&RandomRangeFloat(Minimum=30, Maximum=80), B=$User.Velocity Scale)
   )
   &SampleParticlesFromOtherEmitter(Lifetime Sampling:ENiagara_DirectReadSamplingMode=::NewEnumerator1)
   &ShapeLocation(
    Hemisphere Distribution=<0.5 1>
    MaskSpawnGroup=false
    Non Uniform Scale=<1 2 1>
    Offset=<-140 0 0>
    Offset Coordinate Space=::Local
    Offset Mode=::NewEnumerator0
    Override Seed=false
    Rotation Mode=::NewEnumerator1
    Sphere Radius=40
    UseOwnerScale=true
   )
   &CameraOffset(Camera Offset Amount=-50)
  ]
  particleUpdateScript [
   &ParticleState()
   &ScaleMeshSize(Scale Factor=&VectorFromFloat(Value=&FloatFromCurve(
      FloatCurve=DataInterface() {propertyValues {Curve {defaultValue=1, keys [{tangentMode=RCTM_User, time=0.07, value=0}, {tangentMode=RCTM_User, time=0.25, value=1}, <at 0.65 1>, <at 1 0>]}}}
      Scale Curve=$System.Timeline.InverseNormalizedTime
     )))
   &DynamicMaterialParameters(
    Index 1 Param 1=&FloatFromCurve(FloatCurve=DataInterface() {propertyValues {Curve {keys [<at 0 1>,<at 0.3 0>]}}})
    Index 1 Param 2=&RandomRangeFloat(Minimum=3.3, Maximum=4.2)
    Index 1 Param 3=&FloatFromCurve(FloatCurve=DataInterface() {propertyValues {Curve {keys [<at 0.11 1>,<at 0.16 -1.5>,<at 0.8 -1.5>,<at 1 1>]}}})
    Index 1 Param 4=1
    Index 2 Param 1=&RandomRangeFloat(Minimum=0, Maximum=1111)
    Module.Index1Float0Write=true
    Module.Index1Float1Write=true
    Module.Index1Float2Write=true
    Module.Index1Float3Write=true
    Module.Index2Float0Write=true
    Module.Index2Float1Write=false
    Module.Index2Float2Write=false
    Module.Index2Float3Write=false
    Module.Index3Float0Write=false
    Module.Index3Float1Write=false
    Module.Index3Float2Write=true
    Module.Index3Float3Write=true
    Write Parameter Index 0=::NewEnumerator0
    Write Parameter Index 1=::NewEnumerator1
    Write Parameter Index 2=::NewEnumerator1
    Write Parameter Index 3=::NewEnumerator1
   )
   &ScaleColor(
    Linear Color Curve=DataInterface() {propertyValues {
      AlphaCurve {keys [<at 0.0101 0>,<at 0.1087 1>,<at 0.4705 1 RCIM_Cubic>,<at 1 0 RCIM_Cubic>]}
      BlueCurve {keys [<at 0 1 RCIM_Cubic>]}
      GreenCurve {keys [<at 0 1 RCIM_Cubic>]}
      RedCurve {keys [<at 0 1 RCIM_Cubic>]}
     }}
    Scale Mode=::NewEnumerator2
   )
   &Collision(
    Age Colliding Particles=false
    Collision Enabled=&SetBoolByFloatComparison(A=$Particles.NormalizedAge, B=0.001)
    Friction=50000
    Friction During a Bounce=50000
    GPU Collision Type=::NewEnumerator2
    KillOnCollision=true
    Module.Control Rest State Via Surface Normal=false
    Module.EnableMaxCollisionCount=false
    Module.Randomize Collision Normal Vector=false
    Restitution=0
   )
   /SetVariables(Particles.ColorBlack:LinearColor=&MakeLinearColorFromVectorAndFloat(
     "Vector (RGB)"=<0 0 0>
     "Float (Alpha)"=&MakeCustomFloatFromBool(
      Bool=$Particles.HasCollided
      True Float=0
      False Float=&FloatFromCurve(FloatCurve=DataInterface() {propertyValues {Curve {keys [
           <at 0 0 RCIM_Cubic>
           {arriveTangent=4, leaveTangent=4, time=0.1, value=2}
           {arriveTangent=-8.1145, interpMode=RCIM_Cubic, leaveTangent=-8.1145, tangentMode=RCTM_User, time=0.9, value=2}
           <at 1 0 RCIM_Cubic>
          ]}}})
     )
    )) {isSetParameters=true, module=SetVariables_#5}
   &CurlNoiseForce(Bias Noise=false, Noise Quality / Cost=::NewEnumerator0, Noise Strength=400, Pan Noise=false)
   /SetVariables(
    Particles.DMP_Trail:Vector4f=&MakeVector4(
     X=0
     Y=-0.1
     Z=0
     W=&FloatFromCurve(FloatCurve=DataInterface() {propertyValues {Curve {defaultValue=1, keys [{tangentMode=RCTM_User, time=0.05, value=0}, {tangentMode=RCTM_User, time=1, value=1}]}}})
    )
    Particles.DMP_Trail1:Vector4f=&MakeVector4(X=1, Y=1, Z=1, W=1)
    Particles.DMP_Trail2:Vector4f=&MakeVector4(X=1, Y=1, Z=1, W=1)
    Particles.DMP_Trail3:Vector4f=&MakeVector4(X=100, Y=0, Z=0, W=0)
    Particles.Trail Color:LinearColor=&MultiplyLinearColor(
     Linear Color=&MakeCustomLinearColorFromBool(Bool=$User.Ascended, True Color=<rgba 0.1931 0 2 0.8>, False Color=<rgba 0.1771 0.0046 0.0046 0.8>)
     Scale Factor=&MakeVector4(
      X=1
      Y=1
      Z=1
      W=&MakeCustomFloatFromBool(
       Bool=$Particles.HasCollided
       True Float=0
       False Float=&FloatFromCurve(
        FloatCurve=DataInterface() {propertyValues {Curve {keys [<at 0.85 1>,<at 1 0>]}}}
        Scale Curve=&OneMinusFloat(Float=&Power(
          a=&Abs_Float(Float:NiagaraFloat=&DotProduct(
            Vector A=&ReturnCameraVector(Camera Vectors=::NewEnumerator1, Coordinate Space=::Simulation)
            Vector B=&NormalizeVector(Vector To Normalize=$Particles.Velocity, Safe Normalize=true, Safe Fallback Threshold=0, Safe Fallback Vector=<0 0 0.001>)
           ))
          b=$User.Culling Alpha Power
         ))
       )
      )
     )
    )
    Particles.Trail Orientation:Quat4f=&MakeQuaternion(
     Quaternion From=0
     Axis=<1 0 0>
     Angle=&Add_Float(A=&RandomRangeFloat(Minimum=0, Maximum=360), B=0)
     From=?
     To=?
     Angle Type=::NewEnumerator0
     Normalize Vectors=false
    )
    Particles.Trail Scale:Vector3f=&VectorFromCurve(
     VectorCurve=DataInterface() {propertyValues {
       XCurve {defaultValue=1, keys [{tangentMode=RCTM_User, time=0, value=0.625}, {tangentMode=RCTM_User, time=0.1, value=0.75}, <at 1 0.5>]}
       YCurve {defaultValue=1, keys [{tangentMode=RCTM_User, time=0, value=0.25}, {tangentMode=RCTM_User, time=0.2, value=1}, <at 1 2>]}
       ZCurve {defaultValue=1, keys [{tangentMode=RCTM_User, time=0, value=0.25}, {tangentMode=RCTM_User, time=0.2, value=1}, <at 1 2>]}
      }}
     Scale Curve=<0.5 0.05 0.05>
    )
    Particles.Velocity=&MakeCustomVectorFromBool(Bool=$Particles.HasCollided, True Vector=<0 0 0>, False Vector=$Particles.Velocity)
   ) {isSetParameters=true, module=SetVariables_#6}
   &SolveForcesAndVelocity(Clamp Velocity=false, Limit Acceleration=false)
   &KillParticlesInVolume(Invert Volume=true, Sphere Radius:NiagaraFloat=$User.Raven Kill Distance)
   &KillParticles(Kill Particles=&SetBoolByFloatComparison(A=100, B=&VectorLength(Vector=$Particles.Velocity)))
   /SetVariables(Particles.smoke Position:NiagaraPosition=&Add_Vector(
     A=Linked() {linkedVariable {name=Particles.Position, type=NiagaraPosition}}
     B=&TransformVector(Vector=<50 0 -5>, Destination Space=::Simulation)
    )) {isSetParameters=true, module=SetVariables_#7}
   &SubUVAnimation(
    Random Start Frame=true
    "Sprite Renderer "=DataInterface() {propertyValues {SpriteRenderer=&self:Raven_Thrower002_0.NiagaraSpriteRendererProperties_0}}
    UseEndFrame=false
    UseStartFrame=false
   )
  ]
  rendererClasses [&NiagaraMeshRendererProperties, &NiagaraSpriteRendererProperties]
  renderers [
   {rendererClass=&NiagaraMeshRendererProperties, rendererIndex=0}
   {rendererClass=&NiagaraMeshRendererProperties, rendererIndex=1}
   {rendererClass=&NiagaraSpriteRendererProperties, rendererIndex=2}
  ]
 }
 Raven_Thrower_PS4 {
  simTarget=GPUComputeSim
  emitterUpdateScript [
   &EmitterState(Life Cycle Mode=::NewEnumerator1, Loop Behavior=::NewEnumerator1, Loop Duration Mode=::NewEnumerator1, UseLoopDelay=false)
   &SpawnRate(
    SpawnRate=&Multiply_Float(A=&RandomRangeFloat(Minimum=4, Maximum=6, Evaluation Type=::NewEnumerator1), B=1)
    Use Spawn Probability=false
   )
  ]
  particleSpawnScript [
   &InitializeParticle(
    Color=&MakeCustomLinearColorFromBool(Bool=$User.Ascended, True Color=<rgba 0.7481 0 2.5 1>, False Color=<rgba 5 0 0 1>)
    Lifetime=&Divide_Float(A=8, B=$User.Velocity Scale)
    Mesh Scale Mode=::NewEnumerator3
    Mesh Uniform Scale=&RandomRangeFloat(Minimum=0.4, Maximum=0.6)
    Sprite Size Mode=::NewEnumerator3
    Uniform Sprite Size=&MakeCustomFloatFromBool(Bool=$User.Ascended, True Float=300, False Float=200)
    UsePositionOffset=false
    Write Material Random=true
   )
   &ShapeLocation(
    Hemisphere Distribution=<0.5 1>
    MaskSpawnGroup=false
    Offset Mode=::NewEnumerator0
    Override Seed=false
    Rotation Mode=::NewEnumerator1
    Sphere Radius=25
    UseOwnerScale=true
   )
   &InitialMeshOrientation(Mesh Orientation Mode=::NewEnumerator2, Use Rotation Vector=false)
   /SetVariables(Particles.ColorBlack:LinearColor=<rgba 0 0 0 55>) {isSetParameters=true, module=SetVariables_#3}
   &AddVelocity(
    Cone Angle=$User.Cone Angle
    Cone Axis=$User.TargetVector
    Module.UseConeFalloff=false
    Module.UseVelDistribution=false
    Override Seed=false
    Velocity Mode=::NewEnumerator2
    Velocity Speed=&Multiply_Float(A=250, B=$User.Velocity Scale)
   )
   &CameraOffset(Camera Offset Amount=-25)
  ]
  particleUpdateScript [
   &ParticleState()
   &ScaleSpriteSize(Uniform Curve Sprite Scale=DataInterface() {propertyValues {Curve {keys [{tangentMode=RCTM_User, time=0, value=0}, {tangentMode=RCTM_User, time=0.25, value=1}]}}})
   &ScaleMeshSize(Scale Factor=&VectorFromFloat(Value=&FloatFromCurve(
      FloatCurve=DataInterface() {propertyValues {Curve {defaultValue=1, keys [{tangentMode=RCTM_User, time=0, value=0.3}, <at 0.1 0.5>, {tangentMode=RCTM_User, time=0.2, value=1}]}}}
      Scale Curve=$System.Timeline.InverseNormalizedTime
     )))
   &Collision(
    Age Colliding Particles=false
    Enable Rest State=false
    Friction=500000
    Friction During a Bounce=500000
    GPU Collision Type=::NewEnumerator2
    KillOnCollision=false
    Module.EnableMaxCollisionCount=false
    Module.Randomize Collision Normal Vector=false
    Restitution=0
   )
   &CurlNoiseForce(Bias Noise=false, Noise Quality / Cost=::NewEnumerator0, Noise Strength=500, Pan Noise=false)
   &DynamicMaterialParameters(
    Index 1 Param 1=&FloatFromCurve(FloatCurve=DataInterface() {propertyValues {Curve {keys [<at 0 1>,<at 0.3 0>]}}})
    Index 1 Param 2=&RandomRangeFloat(Minimum=3.3, Maximum=4.2)
    Index 1 Param 3=&FloatFromCurve(FloatCurve=DataInterface() {propertyValues {Curve {keys [<at 0.01 0>,<at 0.2 5>,<at 0.8 5>,<at 1 0.5>]}}})
    Index 2 Param 1=&RandomRangeFloat(Minimum=0, Maximum=1111)
    Index 3 Param 1=&RandomRangeFloat(Minimum=-0.5, Maximum=3)
    Index 3 Param 2=&RandomRangeFloat(Minimum=0, Maximum=1)
    Index 3 Param 3=&FloatFromCurve(FloatCurve=DataInterface() {propertyValues {Curve {keys [<at 0 2>,<at 0.6 1>]}}})
    Index 3 Param 4=&FloatFromCurve(FloatCurve=DataInterface() {propertyValues {Curve {keys [<at 0 1>,<at 0.3 0.004>]}}})
    Module.Index0Float0Write=true
    Module.Index0Float1Write=true
    Module.Index0Float2Write=false
    Module.Index0Float3Write=true
    Module.Index1Float0Write=true
    Module.Index1Float1Write=true
    Module.Index1Float2Write=true
    Module.Index1Float3Write=true
    Module.Index2Float0Write=true
    Module.Index2Float1Write=false
    Module.Index2Float2Write=false
    Module.Index2Float3Write=false
    Module.Index3Float0Write=true
    Module.Index3Float1Write=true
    Module.Index3Float2Write=true
    Module.Index3Float3Write=true
    Write Parameter Index 1=::NewEnumerator1
    Write Parameter Index 2=::NewEnumerator1
    Write Parameter Index 3=::NewEnumerator1
   )
   &InheritSourceMovement(ApplyForce=false, ApplyPosition=true, ApplyPositionFalloff=false) {module=InheritSourceMovement001}
   /SetVariables(Particles.Has GPU Collision:NiagaraBool=&self:ScratchDynamicInput_01(Default:Vector3f=<0 0 0>, Set Bool Speed:NiagaraFloat=500)) {isSetParameters=true}
   /SetVariables(
    Particles.DMP_Trail:Vector4f=&MakeVector4(
     X=0
     Y=-0.1
     Z=1
     W=&FloatFromCurve(
      FloatCurve=DataInterface() {propertyValues {Curve {defaultValue=1, keys [{tangentMode=RCTM_User, time=0.05, value=0}, {tangentMode=RCTM_User, time=1, value=1}]}}}
      Scale Curve=2
     )
    )
    Particles.DMP_Trail1:Vector4f=&MakeVector4(X=1, Y=1, Z=1, W=1)
    Particles.DMP_Trail2:Vector4f=&MakeVector4(X=1, Y=1, Z=1, W=1)
    Particles.DMP_Trail3:Vector4f=&MakeVector4(X=500, Y=0, Z=0, W=0)
    Particles.Trail Color:LinearColor=&MultiplyLinearColor(
     Linear Color=&MakeCustomLinearColorFromBool(Bool=$User.Ascended, True Color=<rgba 2 0 0.9407 1>, False Color=<rgba 0.9219 0.024 0.024 1>)
     Scale Factor=&MakeVector4(
      X=1
      Y=1
      Z=1
      W=&MakeCustomFloatFromBool(
       Bool=$Particles.HasCollided
       True Float=0
       False Float=&FloatFromCurve(FloatCurve=DataInterface() {propertyValues {Curve {keys [<at 0.9 1>,<at 1 0>]}}})
      )
     )
    )
    Particles.Trail Orientation:Quat4f=&MakeQuaternion(
     Quaternion From=0
     Axis=<1 0 0>
     Angle=&Add_Float(
      A=&RandomRangeFloat(Minimum=0, Maximum=360)
      B=&Multiply_Float(A=$Engine.Time, B=1000)
     )
     From=?
     To=?
     Angle Type=::NewEnumerator0
     Normalize Vectors=false
    )
    Particles.Velocity=&MakeCustomVectorFromBool(Bool=$Particles.Has GPU Collision, True Vector=<0 0 0>, False Vector=$Particles.Velocity)
   ) {isSetParameters=true, module=SetVariables_#4}
   &Timeline(Play=$Particles.Has GPU Collision, Play Rate=4, Rewind Mode:ENiagara_TimelineMode=::NewEnumerator1)
   &ScaleColor(
    Scale RGB=&Multiply_VectorByFloat(
     Vector=&MakeVectorFromLinearColorRGB(Color=&ColorFromCurve(ColorCurve=DataInterface() {propertyValues {
         AlphaCurve {keys [<at 0 1 RCIM_Cubic>,<at 1 0 RCIM_Cubic>]}
         BlueCurve {keys [<at 0 1 RCIM_Cubic>]}
         GreenCurve {keys [<at 0 1 RCIM_Cubic>]}
         RedCurve {keys [<at 0 1 RCIM_Cubic>]}
        }}))
     Float=&MakeCustomFloatFromBool(Bool=$Particles.Has GPU Collision, True Float=0, False Float=1)
    )
    ScaleA=false
    ScaleRGB=true
   )
   /SetVariables(Particles.ColorBlack:LinearColor=&MakeLinearColorFromVectorAndFloat(
     "Vector (RGB)"=&MakeVectorFromLinearColorRGB(Color=&ColorFromCurve(ColorCurve=DataInterface() {propertyValues {
         AlphaCurve {keys [<at 0 0 RCIM_Cubic>]}
         BlueCurve {keys [<at 0.5354 0 RCIM_Cubic>]}
         GreenCurve {keys [<at 0.5354 0 RCIM_Cubic>]}
         RedCurve {keys [<at 0.5354 0 RCIM_Cubic>]}
        }}))
     "Float (Alpha)"=&MakeCustomFloatFromBool(Bool=$Particles.Has GPU Collision, True Float=0, False Float=1)
    )) {isSetParameters=true}
   &SubUVAnimation(
    "Sprite Renderer "=DataInterface() {propertyValues {SpriteRenderer=&self:Raven_Thrower_PS4_0.NiagaraSpriteRendererProperties_1}}
    SubUV Lookup Index:NiagaraFloat=$Particles.Timeline.NormalizedTime
    UseEndFrame=false
    UseStartFrame=false
   )
   /SetVariables(
    Particles.Hit Color:LinearColor=&MakeLinearColorFromVectorAndFloat(
     "Vector (RGB)"=&MakeCustomVectorFromBool(
      Bool=$User.Ascended
      True Vector=&VectorFromCurve(
       VectorCurve=DataInterface() {propertyValues {
         XCurve {keys [{arriveTangent=-3.5433, time=0, value=0.625}, <at 0.5 0.1>]}
         YCurve {keys [<at 0.01 0.1>,<at 0.5 0>]}
         ZCurve {keys [<at 0 1.375>,<at 0.5 0.375>]}
        }}
       CurveIndex=$Particles.Timeline.NormalizedTime
      )
      False Vector=&VectorFromCurve(
       VectorCurve=DataInterface() {propertyValues {
         XCurve {keys [{arriveTangent=-12.0139, interpMode=RCIM_Cubic, leaveTangent=-12.0139, tangentMode=RCTM_User, time=0.01, value=2}, <at 0.5 0>]}
         YCurve {keys [<at 0.01 0.1>,<at 0.5 0>]}
         ZCurve {keys [<at 0.01 0.1>,<at 0.5 0>]}
        }}
       CurveIndex=$Particles.Timeline.NormalizedTime
      )
     )
     "Float (Alpha)"=&Multiply_Float(A=5, B=$Particles.Timeline.InverseNormalizedTime)
    )
    Particles.Hit Sprite:Vector2f=&Vector2DFromCurve(
     Vector2Curve=DataInterface() {propertyValues {
       XCurve {
        defaultValue=1
        keys [
         {arriveTangent=4.1785, interpMode=RCIM_Cubic, leaveTangent=4.1785, tangentMode=RCTM_User, time=0.01, value=0}
         {arriveTangent=1.1179, interpMode=RCIM_Cubic, leaveTangent=1.1179, tangentMode=RCTM_User, time=0.3, value=0.65}
         {arriveTangent=0.0224, interpMode=RCIM_Cubic, leaveTangent=0.0224, tangentMode=RCTM_User, time=1, value=1}
        ]
       }
       YCurve {
        defaultValue=1
        keys [
         {arriveTangent=4.1785, interpMode=RCIM_Cubic, leaveTangent=4.1785, tangentMode=RCTM_User, time=0.01, value=0}
         {arriveTangent=1.1179, interpMode=RCIM_Cubic, leaveTangent=1.1179, tangentMode=RCTM_User, time=0.3, value=0.65}
         {arriveTangent=0.0224, interpMode=RCIM_Cubic, leaveTangent=0.0224, tangentMode=RCTM_User, time=1, value=1}
        ]
       }
      }}
     CurveIndex=$Particles.Timeline.NormalizedTime
     ScaleCurve=<120 120>
    )
   ) {isSetParameters=true}
   &SolveForcesAndVelocity(Clamp Velocity=false, Limit Acceleration=false)
   &KillParticlesInVolume(Invert Volume=true, Sphere Radius:NiagaraFloat=$User.Raven Kill Distance)
  ]
  rendererClasses [&NiagaraMeshRendererProperties, &NiagaraSpriteRendererProperties]
  renderers [
   {rendererClass=&NiagaraMeshRendererProperties, rendererIndex=0}
   {rendererClass=&NiagaraSpriteRendererProperties, rendererIndex=1}
   {rendererClass=&NiagaraMeshRendererProperties, rendererIndex=2}
   {rendererClass=&NiagaraSpriteRendererProperties, rendererIndex=3}
  ]
 }
 Wind {
  emitterUpdateScript [
   &EmitterState(Life Cycle Mode=::NewEnumerator1, Loop Behavior=::NewEnumerator1, Loop Duration Mode=::NewEnumerator1, UseLoopDelay=false)
   &SpawnRate(SpawnRate=20, Use Spawn Probability=false)
  ]
  particleSpawnScript [
   &InitializeParticle(
    Color=&LerpLinearColors(
     StartColor=<rgba 0 0 0 1.1>
     EndColor=<rgba 0.0078 0.0035 0.0208 1>
     LerpFactor=&UniformAOrBFloat(A=0, Evaluation Type=::NewEnumerator0, Recalculate Random Each Loop=true, Override Seed=false)
    )
    Lifetime=&UniformRangedFloat(Minimum=0.3, Maximum=0.4)
    Mesh Scale=&Lerp_Vector(
     A=<10 10 35>
     B=<11 11 44>
     Alpha=&UniformAOrBFloat(A=0, Distribution Weight=0.6, Evaluation Type=::NewEnumerator0, Recalculate Random Each Loop=true, Override Seed=false)
    )
    Mesh Scale Mode=::NewEnumerator4
    Sprite Rotation Mode=::NewEnumerator2
    Sprite UV Mode=::NewEnumerator3
    UsePositionOffset=false
    Write Material Random=true
   ) {module=InitializeParticle001}
   &InitialMeshOrientation(
    Mesh Axis To Orient=<0 0 1>
    Mesh Orientation Mode=::NewEnumerator2
    Orientation Coordinate Space=::Local
    Rotation=&RandomRangeVector(Minimum=<0 -0.1 -0.1>, Maximum=<1 0.1 0.1>)
    Rotation Coordinate Space=::NewEnumerator2
    Use Rotation Vector=true
   ) {module=InitialMeshOrientation001}
   &AddVelocity(Velocity=$User.TargetVector, Velocity Speed Scale=&Multiply_Float(A=$User.Velocity Scale, B=250))
   &CameraOffset(Camera Offset Amount=-40)
  ]
  particleUpdateScript [
   &ParticleState()
   &ScaleSpriteSize(Uniform Curve Sprite Scale=DataInterface() {propertyValues {Curve {keys [
        {arriveTangent=5.8433, interpMode=RCIM_Cubic, leaveTangent=5.8433, tangentMode=RCTM_User, time=0, value=0.3}
        {tangentMode=RCTM_User, time=0.25, value=1.7}
        <at 1 0.5>
       ]}}})
   &InheritSourceMovement(Applied Position Scale=<0.8 0.8 0.8>, ApplyForce=false, ApplyPosition=true, ApplyPositionFalloff=false)
   &SolveForcesAndVelocity(Clamp Velocity=false, Limit Acceleration=false)
   &UpdateMeshOrientation(
    Rotation Coordinate Space=::NewEnumerator2
    Rotation Rate=&FloatFromCurve(
     FloatCurve=DataInterface() {propertyValues {Curve {keys [
         {arriveTangent=-2.0408, interpMode=RCIM_Cubic, time=0, value=1}
         {arriveTangent=-2.0603, interpMode=RCIM_Cubic, leaveTangent=-11.7273, tangentMode=RCTM_User, time=1, value=0.05}
        ]}}}
     Scale Curve=-1
    )
    Rotation Vector=<0.4 0.08 0.1>
   ) {module=UpdateMeshOrientation001}
   &ScaleMeshSize(Scale Factor=&VectorFromCurve(
     VectorCurve=DataInterface() {propertyValues {
       XCurve {keys [
         {arriveTangent=2.6817, interpMode=RCIM_Cubic, leaveTangent=2.1453, tangentMode=RCTM_User, time=0, value=0.5}
         {arriveTangent=0.2969, interpMode=RCIM_Cubic, leaveTangent=0.8713, tangentMode=RCTM_User, time=1, value=2}
        ]}
       YCurve {keys [
         {arriveTangent=3.0136, interpMode=RCIM_Cubic, leaveTangent=2.1453, tangentMode=RCTM_User, time=0, value=0.5}
         {arriveTangent=0.2969, interpMode=RCIM_Cubic, leaveTangent=0.5259, tangentMode=RCTM_User, time=1, value=2}
        ]}
       ZCurve {keys [
         {arriveTangent=2.7258, interpMode=RCIM_Cubic, time=0, value=0.625}
         {arriveTangent=2.1816, interpMode=RCIM_Cubic, leaveTangent=2.1816, tangentMode=RCTM_User, time=1, value=1.875}
        ]}
      }}
     Scale Curve=<0.4 0.4 0.52>
    ))
   &ScaleColor(
    Scale Alpha=&FloatFromCurve(
     FloatCurve=DataInterface() {propertyValues {Curve {keys [
         {arriveTangent=2.3251, interpMode=RCIM_Cubic, leaveTangent=2.3251, tangentMode=RCTM_User, time=0, value=0.55}
         {interpMode=RCIM_Cubic, tangentMode=RCTM_User, time=0.2, value=1}
         {arriveTangent=-1.25, interpMode=RCIM_Cubic, leaveTangent=-1.25, time=0.8, value=0.9}
         {arriveTangent=-1.8087, interpMode=RCIM_Cubic, leaveTangent=-1.8087, tangentMode=RCTM_User, time=1, value=0}
        ]}}}
     Scale Curve=0.5
    )
    ScaleA=true
    ScaleRGB=false
   )
   &DynamicMaterialParameters(
    Index 0 Param 1=&FloatFromCurve(FloatCurve=DataInterface() {propertyValues {Curve {keys [<at 0.5 0 RCIM_Cubic>, {arriveTangent=4.311, interpMode=RCIM_Cubic, leaveTangent=16.4602, tangentMode=RCTM_User, time=1, value=1}]}}})
    Index 0 Param 2=&FloatFromCurve(
     FloatCurve=DataInterface() {propertyValues {Curve {keys [
         {arriveTangent=6.2888, interpMode=RCIM_Cubic, leaveTangent=1.845, tangentMode=RCTM_User, time=0, value=0}
         {arriveTangent=0.4139, interpMode=RCIM_Cubic, leaveTangent=-0.8461, tangentMode=RCTM_User, time=1, value=1}
        ]}}}
     Scale Curve=-0.5
    )
    Index 0 Param 4=&FloatFromCurve(
     FloatCurve=DataInterface() {propertyValues {Curve {keys [
         {arriveTangent=0.2805, interpMode=RCIM_Cubic, time=0, value=1}
         {arriveTangent=2.4494, interpMode=RCIM_Cubic, leaveTangent=2.4494, tangentMode=RCTM_User, time=1, value=2}
        ]}}}
     Scale Curve=1.25
    )
    Module.Index0Float0Write=true
    Module.Index0Float1Write=true
    Module.Index0Float2Write=false
    Module.Index0Float3Write=true
   )
  ]
  rendererClasses [&NiagaraMeshRendererProperties]
  renderers [{rendererClass=&NiagaraMeshRendererProperties, rendererIndex=0}]
 }
}
summary {
 systemName=NS_MistRaven_DarkCoil_RavenThrower_V3b
 emitters [
  {emitterName=Raven_Thrower, rendererClasses [&NiagaraMeshRendererProperties, &NiagaraSpriteRendererProperties]}
  {emitterName=Wind, rendererClasses [&NiagaraMeshRendererProperties]}
  {
   emitterName=Raven_Thrower001
   rendererClasses [&NiagaraMeshRendererProperties, &NiagaraSpriteRendererProperties]
  }
  {
   emitterName=Raven_Thrower_PS4
   simTarget=GPUComputeSim
   rendererClasses [&NiagaraMeshRendererProperties, &NiagaraSpriteRendererProperties]
  }
  {emitterName=Feather, rendererClasses [&NiagaraSpriteRendererProperties]}
  {emitterName=Feather_PS4, simTarget=GPUComputeSim, rendererClasses [&NiagaraSpriteRendererProperties]}
  {
   emitterName=Raven_Thrower002
   simTarget=GPUComputeSim
   rendererClasses [&NiagaraMeshRendererProperties, &NiagaraSpriteRendererProperties]
  }
 ]
 userVariables [
  User.ActiveTarget {type=NiagaraBool}
  User.Ascended {type=NiagaraBool, defaultValue=true}
  User.Cone Angle {type=NiagaraFloat, defaultValue=20, description=""}
  User.Culling Alpha Power {type=NiagaraFloat, defaultValue=8, description=""}
  User.Raven Kill Distance {type=NiagaraFloat, defaultValue=2500, description=""}
  User.TargetVector {type=Vector3f, description="", defaultValue {struct=Vector3f, value=<1 0.075 0.02>}}
  User.Velocity Scale {type=NiagaraFloat, defaultValue=20, description=""}
 ]
}
