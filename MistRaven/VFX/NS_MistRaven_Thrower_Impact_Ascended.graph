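# Niagara impact effect for the Ascended variant of the MistRaven Thrower, judging by the asset name.
# It has five LW-prefixed, likely lightweight, emitters named for a feather burst, a smoke cloud and three glow sphere meshes.
# SystemState is the only listed module, and no renderers are listed for any emitter.
# An integer user parameter, IS_Damage, defaults to 1, with no visible consumer.
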NiagaraSystem /MistRaven/VFX/NS_MistRaven_Thrower_Impact_Ascended
dependencies {usedDataInterfaces [], usedDynamicInputs [], usedModules [&SystemState]}
emitters {
 LW_Feather {rendererClasses [], renderers []}
 LW_GlowSphereMesh2001 {rendererClasses [], renderers []}
 LW_GlowSphereMesh2002 {rendererClasses [], renderers []}
 LW_GlowSphereMesh2003 {rendererClasses [], renderers []}
 LW_Smoke_Cloud {rendererClasses [], renderers []}
}
summary {
 systemName=NS_MistRaven_Thrower_Impact_Ascended
 emitters [
  {emitterName=LW_GlowSphereMesh2001, rendererClasses []}
  {emitterName=LW_Feather, rendererClasses []}
  {emitterName=LW_Smoke_Cloud, rendererClasses []}
  {emitterName=LW_GlowSphereMesh2002, rendererClasses []}
  {emitterName=LW_GlowSphereMesh2003, rendererClasses []}
 ]
 userVariables [User.IS_Damage {type=NiagaraInt32, defaultValue=1}]
}
