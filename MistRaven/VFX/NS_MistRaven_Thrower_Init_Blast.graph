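# Niagara system for the initial blast effect of the MistRaven thrower.
# It combines glow sphere mesh, feather and energy emitters, with a smoke cloud emitter present but disabled; the LW_ names suggest lightweight emitters.
# A system spawn assignment accompanies SystemState, using a bool-to-linear-color dynamic input, likely tinting by the User.Ascended flag.
# It also exposes a float User.IsAir parameter, though its consumer is not shown.
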NiagaraSystem /MistRaven/VFX/NS_MistRaven_Thrower_Init_Blast
dependencies {
 usedDataInterfaces []
 usedDynamicInputs [&MakeCustomLinearColorFromBool]
 usedModules [
  &self:SystemSpawnScript.SystemScriptSource.SystemScriptGraph.NiagaraNodeAssignment_0.SetVariables_#1
  &SystemState
 ]
}
emitters {
 LW_Energy {rendererClasses [], renderers []}
 LW_Feather {rendererClasses [], renderers []}
 LW_GlowSphereMesh2001 {rendererClasses [], renderers []}
 LW_Smoke_Cloud {bEnabled=false, rendererClasses [], renderers []}
}
summary {
 systemName=NS_MistRaven_Thrower_Init_Blast
 emitters [
  {emitterName=LW_GlowSphereMesh2001, rendererClasses []}
  {emitterName=LW_Feather, rendererClasses []}
  {bEnabled=false, emitterName=LW_Smoke_Cloud, rendererClasses []}
  {emitterName=LW_Energy, rendererClasses []}
 ]
 userVariables [
  User.Ascended {type=NiagaraBool}
  User.IsAir {type=NiagaraFloat, defaultValue=1, description=""}
 ]
}
