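# Looping Niagara effect for MistRaven's active Wraith Form: flapping mesh wings, swirling raven ball, edge trail meshes, smoke, ribbons, and lightweight feather emitters.
# User.Ascended switches colors between red and purple, while User.Is Air gates several spawns and wing alpha.
# System timelines drive shrinking, dimming and a pulsing finish signal.
# Age curves bob wings and trails vertically, and owner velocity feeds ribbon alpha and wing material parameters.
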NiagaraSystem /MistRaven/VFX/NS_MistRaven_WraithForm_Active_Loop
dependencies {
 usedDataInterfaces [
  &NiagaraDataInterfaceCurve
  &NiagaraDataInterfaceVector2DCurve
  &NiagaraDataInterfaceSpriteRendererInfo
  &NiagaraDataInterfaceArrayFloat3
  &NiagaraDataInterfaceArrayFloat
  &NiagaraDataInterfaceArrayColor
  &NiagaraDataInterfaceMeshRendererInfo
  &NiagaraDataInterfaceVectorCurve
  &NiagaraDataInterfaceCamera
 ]
 usedDynamicInputs [
  &Vector2DFromCurve
  &RandomRangeVector2D
  &SetBoolByFloatComparison
  /Niagara/DynamicInputs/Math/Clamp_Vector
  &Multiply_Vector
  &Multiply_Vector2D_ByFloat
 ]
 usedModules [
  &self:SystemSpawnScript.SystemScriptSource.SystemScriptGraph.NiagaraNodeAssignment_1.SetVariables_#1
  &self:SystemSpawnScript.SystemScriptSource.SystemScriptGraph.NiagaraNodeAssignment_2.SetVariables_#2
  &self:SystemSpawnScript.SystemScriptSource.SystemScriptGraph.NiagaraNodeAssignment_3.SetVariables_#3
  &SystemState
  &Timeline
  &self:SystemSpawnScript.SystemScriptSource.SystemScriptGraph.NiagaraNodeAssignment_0.SetVariables_#4
  &self:SystemSpawnScript.SystemScriptSource.SystemScriptGraph.NiagaraNodeAssignment_4.SetVariables_#5
 ]
}
emitters {
 Em_Ravens_Ball {
  emitterSpawnScript [/SetVariables(Emitter.CameraOffset=10, Emitter.FlareSize:Vector2f=<350 140>, Emitter.SpriteSize:Vector2f=<50 50>) {isSetParameters=true}]
  emitterUpdateScript [
   &EmitterState(Life Cycle Mode=::NewEnumerator1, Loop Duration=0.4, UseLoopDelay=false)
   &SpawnRate(
    Spawn Probability=$User.Is Air
    SpawnRate=&FloatFromCurve(FloatCurve=DataInterface() {propertyValues {Curve {keys [<at 0.5 0>,<at 0.6 1>]}}}, CurveIndex=$Emitter.Age, Scale Curve=10)
    Use Spawn Probability=true
   )
   /SetVariables(Emitter.Sine Signal:NiagaraFloat=&Lerp_Float(
     A=&Sine(Normalized Angle=$Emitter.Age, Period=0.4, Scale=0.2, Bias=1)
     B=&Sine(Normalized Angle=$Emitter.Age, Period=0.2, Scale=1, Bias=2)
     Alpha=&FloatFromCurve(
      FloatCurve=DataInterface() {propertyValues {Curve {defaultValue=1, keys [{tangentMode=RCTM_User, time=-0.1, value=0}, {tangentMode=RCTM_User, time=0.1, value=1}]}}}
      CurveIndex=&Subtract_Float(A=$Emitter.Age, B=$System.Signal Start)
     )
    )) {isSetParameters=true}
   /SetVariables(
    Emitter.Color:LinearColor=&MakeLinearColorFromVectorAndFloat(
     "Vector (RGB)"=&MakeVectorFromLinearColorRGB(Color=&Multiply_LinearColorByFloat(
       Linear Color=&Multiply_LinearColorByFloat(
        Linear Color=&MultiplyLinearColor(
         Linear Color=&MakeCustomLinearColorFromBool(Bool=$User.Ascended, True Color=<rgba 0.2409 0 0.9375 1>, False Color=<rgba 1.5 0 0.02 1>)
         Scale Factor=&MakeVector4_FromFloat(Float=&FloatFromCurve(
           FloatCurve=DataInterface() {propertyValues {Curve {defaultValue=1, keys [<at 0 0.001>, {tangentMode=RCTM_User, time=0.8, value=0.001}, {tangentMode=RCTM_User, time=1.25, value=1}]}}}
           CurveIndex=$Engine.System.Age
          ))
        )
        Float=&Multiply_Float(
         A=&FloatFromCurve(
          FloatCurve=DataInterface() {propertyValues {Curve {
             defaultValue=0
             keys [{interpMode=RCIM_Cubic, tangentMode=RCTM_User, time=0.1, value=0}, {interpMode=RCIM_Cubic, tangentMode=RCTM_User, time=0.5, value=1}]
            }}}
          CurveIndex=$Emitter.Age
          Scale Curve=1.5
         )
         B=&FloatFromCurve(
          FloatCurve=DataInterface() {propertyValues {Curve {keys [<at 1 1>,<at 1.25 0>]}}}
          CurveIndex=&Subtract_Float(A=$Emitter.Age, B=$System.Signal Start)
         )
        )
       )
       Float=$Emitter.Sine Signal
      ))
     "Float (Alpha)"=1
    )
    Emitter.Position:NiagaraPosition=&TransformPosition(
     Position=&AddVectorToPosition(
      Position=$Engine.Emitter.SimulationPosition
      Vector=&MakeVector(
       X=0
       Y=0
       Z=&FloatFromCurve(
        FloatCurve=DataInterface() {propertyValues {Curve {keys [
            <at 0 0.0197>
            {arriveTangent=0.0896, interpMode=RCIM_Cubic, leaveTangent=0.0896, time=0.178, value=0.12}
            {arriveTangent=0.7228, interpMode=RCIM_Cubic, leaveTangent=0.7228, time=0.4154, value=0.0569}
            {arriveTangent=0.0626, interpMode=RCIM_Cubic, leaveTangent=0.0626, time=0.773, value=0.55}
            {arriveTangent=-0.142, interpMode=RCIM_Cubic, leaveTangent=-0.142, time=1.1037, value=0.1}
            {arriveTangent=-1.2076, interpMode=RCIM_Cubic, leaveTangent=-1.2076, tangentMode=RCTM_User, time=1.4772, value=0.45}
            {arriveTangent=-0.5153, interpMode=RCIM_Cubic, leaveTangent=-0.5153, tangentMode=RCTM_User, time=1.9788, value=0.1758}
            <at 2.5 0>
           ]}}}
        CurveIndex=$Engine.System.Age
        Scale Curve=80
       )
      )
     )
     Source Space=::Local
     Destination Space=::World
    )
   ) {isSetParameters=true}
  ]
  particleSpawnScript [
   &PartitionParticles()
   &InitializeParticle(
    Color=<rgba 0.016 0.0103 0.0273 5>
    Lifetime Max=1.2
    Lifetime Min=0.8
    Lifetime Mode=::NewEnumerator1
    Mesh Scale Mode=::NewEnumerator1
    Mesh Uniform Scale Max=0.3
    Mesh Uniform Scale Min=0.22
    UsePositionOffset=false
    Write Material Random=true
   )
   &ShapeLocation(MaskSpawnGroup=false, Override Seed=false, Sphere Radius=50, Sphere Surface Distribution=0.5833, UseOwnerScale=true)
  ]
  particleUpdateScript [
   &ParticleState()
   &ScaleMeshSize(Scale Factor=&VectorFromFloat(Value=&FloatFromCurve(
      FloatCurve=DataInterface() {propertyValues {Curve {
         defaultValue=1
         keys [
          {arriveTangent=3.4482, interpMode=RCIM_Cubic, leaveTangent=3.4482, tangentMode=RCTM_User, time=0, value=0}
          {arriveTangent=2, interpMode=RCIM_Cubic, leaveTangent=2, time=0.2, value=0.7}
          {interpMode=RCIM_Cubic, tangentMode=RCTM_User, time=0.5, value=1}
          {arriveTangent=-2, interpMode=RCIM_Cubic, leaveTangent=-2, time=0.8, value=0.8}
          {arriveTangent=-2.8613, tangentMode=RCTM_User, time=1, value=0}
         ]
        }}}
      Scale Curve=1.4
     )))
   &VortexForce(
    Origin Pull Amount=650
    Origin Pull Only On Vortex Axis=false
    UseOriginOffset=false
    Vortex Axis=&RandomRangeVector(Minimum=<-1 -1 -1>, Maximum=<1 1 1>)
    Vortex Force Amount=&Multiply_Float(A=150, B=&Add_Float(A=1, B=$System.Timeline.NormalizedTime))
   )
   &InheritSourceMovement(
    Applied Position Scale=&SelectVectorFromArray(
     Direct Array Index=$Particles.PartitionParticles.Partition
     Vector Selection Array=DataInterface() {propertyValues {FloatData [<0.6 0.6 0.6>, <0.7 0.7 0.7>]}}
    )
    ApplyForce=false
    ApplyPosition=true
    ApplyPositionFalloff=false
   )
   &SolveForcesAndVelocity(Clamp Velocity=false, Limit Acceleration=false)
   &DynamicMaterialParameters(
    Index 1 Param 2=&RandomRangeFloat(Minimum=2, Maximum=4)
    Index 1 Param 3=&FloatFromCurve(FloatCurve=DataInterface() {propertyValues {Curve {
        defaultValue=1
        keys [
         {interpMode=RCIM_Cubic, tangentMode=RCTM_User, time=0, value=0}
         {interpMode=RCIM_Cubic, tangentMode=RCTM_User, time=0.3, value=-2.375}
         {arriveTangent=3.3929, interpMode=RCIM_Cubic, leaveTangent=3.3929, time=0.8, value=-2.25}
         {tangentMode=RCTM_User, time=1, value=0}
        ]
       }}})
    Index 1 Param 4=1
    Module.Index0Float0Write=false
    Module.Index0Float1Write=false
    Module.Index0Float2Write=false
    Module.Index0Float3Write=false
    Module.Index1Float0Write=true
    Module.Index1Float1Write=true
    Module.Index1Float2Write=true
    Module.Index1Float3Write=true
    Module.Index3Float0Write=false
    Module.Index3Float1Write=false
    Module.Index3Float2Write=true
    Module.Index3Float3Write=true
    Write Parameter Index 1=::NewEnumerator1
    Write Parameter Index 3=::NewEnumerator1
   )
  ]
  rendererClasses [&NiagaraMeshRendererProperties]
  renderers [{rendererClass=&NiagaraMeshRendererProperties, rendererIndex=0}]
 }
 Em_Smoke_Cloud_Detail {
  emitterUpdateScript [
   &EmitterState()
   &SpawnRate(Spawn Probability=$User.Is Air, SpawnRate=6, Use Spawn Probability=false)
   &SpawnPerUnit(Module.Use Max Movement Threshold=true, Module.Use Movement Tolerance=false, Spawn Group=1, Spawn Spacing=45, Use Spawn Probability=false)
  ]
  particleSpawnScript [
   /SetVariables(Particles.VisibilityTag=$Emitter.SpawnGroup) {isSetParameters=true}
   &InitializeParticle(
    Color=<rgba 0.0101 0.0153 0.0273 1>
    Lifetime Min=0.95
    Lifetime Mode=::NewEnumerator1
    Sprite Rotation Mode=::NewEnumerator2
    Sprite Size Max=<100 100>
    Sprite Size Min=<80 80>
    Sprite Size Mode=::NewEnumerator2
    UsePositionOffset=false
    Write Material Random=true
   )
   &CameraOffset(Camera Offset Amount=30)
   &ShapeLocation(
    MaskSpawnGroup=false
    Offset Mode=::NewEnumerator0
    Override Seed=false
    Sphere Radius=&MaskFloatBySpawnGroup(Masked Float=50, Passthrough Float=10)
    UseOwnerScale=true
   )
  ]
  particleUpdateScript [
   &ParticleState()
   &SpriteRotationRate(Rotation Rate=&RandomRangeFloat(Minimum=-50, Maximum=50))
   &ScaleSpriteSize(
    Scale Sprite Size Mode=::NewEnumerator0
    Uniform Scale Factor=&FloatFromCurve(
     FloatCurve=DataInterface() {propertyValues {Curve {keys [<at 0 0.8>,<at 1 1.2>]}}}
     Scale Curve=&MaskFloatBySpawnGroup(Masked Float=1.3, Passthrough Float=1.9)
    )
   )
   &SubUVAnimation(
    End Frame Range Override=30
    "Sprite Renderer "=DataInterface() {propertyValues {SpriteRenderer=&self:Em_Smoke_Cloud_Detail_0.NiagaraSpriteRendererProperties_0}}
    Start Frame Range Override=&RandomRangeInt(Minimum=0, Maximum=5)
    UseEndFrame=true
    UseStartFrame=true
   )
   &SubUVAnimation(
    End Frame Range Override=30
    Random Start Frame=true
    "Sprite Renderer "=DataInterface() {propertyValues {SpriteRenderer=&self:Em_Smoke_Cloud_Detail_0.NiagaraSpriteRendererProperties_1}}
    Start Frame Range Override=&RandomRangeInt(Minimum=0, Maximum=5)
    SubUV Lookup Index Scale=0.4
    UseEndFrame=false
    UseStartFrame=false
   ) {module=SubUVAnimation001}
   &ScaleColor(
    Scale Alpha=&Multiply_Float(
     A=&RandomRangeFloat(Minimum=0.5, Maximum=1)
     B=&MaskFloatBySpawnGroup(
      Masked Float=&FloatFromCurve(
       FloatCurve=DataInterface() {propertyValues {Curve {keys [
           {interpMode=RCIM_Cubic, tangentMode=RCTM_User, time=0, value=0}
           {interpMode=RCIM_Cubic, tangentMode=RCTM_User, time=0.5, value=1}
           {tangentMode=RCTM_User, time=1, value=0}
          ]}}}
       Scale Curve=2
      )
      Passthrough Float=&FloatFromCurve(FloatCurve=DataInterface() {propertyValues {Curve {keys [
           {interpMode=RCIM_Cubic, tangentMode=RCTM_User, time=0, value=0.7}
           {interpMode=RCIM_Cubic, tangentMode=RCTM_User, time=0.6, value=1}
           {tangentMode=RCTM_User, time=1, value=0}
          ]}}})
     )
    )
    ScaleA=true
    ScaleRGB=false
   )
   &InheritSourceMovement(
    Applied Position Scale=&MaskVectorBySpawnGroup(Masked Vector=<1 1 1>, Passthrough Vector=<0 0 0>)
    ApplyForce=false
    ApplyPosition=true
    ApplyPositionFalloff=false
   )
   &SolveForcesAndVelocity(Clamp Velocity=false, Limit Acceleration=false)
  ]
  rendererClasses [&NiagaraSpriteRendererProperties]
  renderers [
   {rendererClass=&NiagaraSpriteRendererProperties}
   {rendererClass=&NiagaraSpriteRendererProperties, rendererIndex=1}
  ]
 }
 LW_Feather {rendererClasses [], renderers []}
 LW_Feather001 {rendererClasses [], renderers []}
 LW_Smoke_Cloud {rendererClasses [], renderers []}
 LW_Smoke_Mesh_Rotating {rendererClasses [], renderers []}
 Ribbon_Trail {
  emitterSpawnScript [/SetVariables(
    Emitter.Rand1:NiagaraFloat=&RandomRangeFloat(Minimum=-1, Maximum=1)
    Emitter.Rand2:NiagaraFloat=&RandomRangeFloat(Minimum=-1, Maximum=1)
    Emitter.Rand3:NiagaraFloat=&RandomRangeFloat(Minimum=-1, Maximum=1)
   ) {isSetParameters=true}]
  emitterUpdateScript [
   &EmitterState(Life Cycle Mode=::NewEnumerator1, Loop Duration=2, UseLoopDelay=false)
   &SpawnRate(SpawnRate=24, Use Spawn Probability=false)
   &SpawnRate(SpawnGroup=1, SpawnRate=25, Use Spawn Probability=false) {module=SpawnRate002}
   &SpawnRate(SpawnGroup=2, SpawnRate=25, Use Spawn Probability=false) {module=SpawnRate001}
   &SpawnRate(SpawnGroup=3, SpawnRate=30, Use Spawn Probability=false) {module=SpawnRate003}
   /SetVariables(Emitter.Ascended:NiagaraFloat=&MakeCustomFloatFromBool(Bool=$User.Ascended, True Float=1, False Float=0)) {isSetParameters=true}
  ]
  particleSpawnScript [
   /SetVariables(Particles.RibbonID=/Niagara/DynamicInputs/Meshes/MakeRibbonIDFromSpawnGroup()) {isSetParameters=true}
   &InitializeParticle(
    Color=&SelectLinearColorFromArray(
     Array Sampling Mode=::NewEnumerator1
     Direct Array Index=$Emitter.SpawnGroup
     Color Selection Array=DataInterface() {propertyValues {ColorData [<rgba 1 1 1 1.5>, <rgba 1 1 1 1.5>, <rgba 1 1 1 1.5>, <rgba 1 1 1 5>]}}
     Random Seed=?
    )
    Lifetime=&Subtract_Float(
     A=&SelectFloatFromArray(Direct Array Index=$Emitter.SpawnGroup, Float Selection Array=DataInterface() {propertyValues {FloatData [0.5,0.45,0.52,0.25]}})
     B=$System.Ribbon_falling
    )
    Position Offset=&SelectVectorFromArray(
     Direct Array Index=$Emitter.SpawnGroup
     Vector Selection Array=DataInterface() {propertyValues {FloatData [<-35 -30 0>, <-20 33 0>, <25 0 0>, <0 0 0>]}}
    )
    Position Offset Coordinate Space=::Local
    Ribbon Width=&SelectFloatFromArray(Direct Array Index=$Emitter.SpawnGroup, Float Selection Array=DataInterface() {propertyValues {FloatData [60,50,65,100]}})
    Ribbon Width Mode=::NewEnumerator1
    Sprite Rotation Mode=::NewEnumerator2
    UsePositionOffset=true
    Write Material Random=true
   )
  ]
  particleUpdateScript [
   &ParticleState()
   &ScaleColor(
    Scale Alpha=&FloatFromCurve(FloatCurve=DataInterface() {propertyValues {Curve {keys [<at -0.0046 1.0113 RCIM_Cubic>,<at 1.0011 0.002 RCIM_Cubic>]}}}, Scale Curve=0.7)
    ScaleA=true
    ScaleRGB=false
   )
   &CurlNoiseForce(Bias Noise=false, Noise Frequency=25, Noise Quality / Cost=::NewEnumerator0, Noise Strength=850, Pan Noise=false)
   &InheritSourceMovement(Applied Position Scale=<0.2 0.2 0.2>, ApplyForce=false, ApplyPosition=true, ApplyPositionFalloff=false)
   &SolveForcesAndVelocity(Clamp Velocity=false, Limit Acceleration=false)
   &ScaleColor(
    Scale Alpha=&Multiply_Float(
     A=&FloatFromCurve(FloatCurve=DataInterface() {propertyValues {Curve {defaultValue=1, keys [{tangentMode=RCTM_User, time=0, value=1}, <at 0.1 2>, {tangentMode=RCTM_User, time=1, value=0}]}}})
     B=&ClampFloat(Float=&Multiply_Float(A=&VectorLength(Vector=$Engine.Owner.Velocity), B=0.01))
    )
    ScaleA=true
    ScaleRGB=false
   ) {module=ScaleColor002}
   &DynamicMaterialParameters(
    Index 0 Param 2=0.5
    Index 0 Param 3=&MaskFloatBySpawnGroup(
     Masked Float=$Emitter.Rand1
     Passthrough Float=&MaskFloatBySpawnGroup(Masked Float=$Emitter.Rand2, Passthrough Float=$Emitter.Rand3, Spawn Group Mask=1)
    )
    Index 0 Param 4=&MaskFloatBySpawnGroup(
     Masked Float=$Emitter.Rand1
     Passthrough Float=&MaskFloatBySpawnGroup(Masked Float=$Emitter.Rand2, Passthrough Float=$Emitter.Rand3, Spawn Group Mask=1)
    )
    Module.Index0Float0Write=true
    Module.Index0Float1Write=true
    Module.Index0Float2Write=true
    Module.Index0Float3Write=true
   )
   &ScaleRibbonWidth(Ribbon Width Scale=$System.Timeline002.InverseNormalizedTime)
  ]
  rendererClasses [&NiagaraRibbonRendererProperties]
  renderers [{rendererClass=&NiagaraRibbonRendererProperties}]
 }
 Trail_Edge004 {
  emitterUpdateScript [
   &EmitterState(Life Cycle Mode=::NewEnumerator1, Loop Behavior=::NewEnumerator1, Loop Delay=0.1, Loop Duration Mode=::NewEnumerator1, UseLoopDelay=false)
   &SpawnRate(
    Spawn Probability=$User.Is Air
    SpawnRate=&RandomRangeFloat(Minimum=3, Maximum=4, Evaluation Type=::NewEnumerator1)
    Use Spawn Probability=true
   )
   &SpawnRate(
    Spawn Probability=$User.Is Air
    SpawnGroup=1
    SpawnRate=&RandomRangeFloat(Minimum=12, Maximum=14, Evaluation Type=::NewEnumerator1)
    Use Spawn Probability=true
   ) {module=SpawnRate001}
   /SetVariables(
    Emitter.Inside Color:LinearColor=&MakeLinearColorFromVectorAndFloat(
     "Vector (RGB)"=&MakeVectorFromLinearColorRGB(Color=&Multiply_LinearColorByFloat(
       Linear Color=&MakeCustomLinearColorFromBool(Bool=$User.Ascended, True Color=<rgba 0.2216 0 1 1>, False Color=<rgba 0.55 0 0.0055 1>)
       Float=$System.Timeline.InverseNormalizedTime
      ))
     "Float (Alpha)"=8
    )
    Emitter.Inside Colot2:LinearColor=&MakeLinearColorFromVectorAndFloat(
     "Vector (RGB)"=&MakeVectorFromLinearColorRGB(Color=&Multiply_LinearColorByFloat(
       Linear Color=&MakeCustomLinearColorFromBool(Bool=$User.Ascended, True Color=<rgba 0.3275 0.0169 1 1>, False Color=<rgba 1 0 0.01 8>)
       Float=$System.Timeline.InverseNormalizedTime
      ))
     "Float (Alpha)"=3
    )
    Emitter.PositionFix=1
   ) {isSetParameters=true, module=SetVariables_#6}
  ]
  particleSpawnScript [
   /SetVariables(Particles.VisibilityTag=$Emitter.SpawnGroup) {isSetParameters=true}
   &InitializeParticle(
    Color=&MaskLinearColorBySpawnGroup(Masked Linear Color=<rgba 1 1 1 2>, Passthrough Linear Color=<rgba 1 1 1 2>)
    Lifetime=&MaskFloatBySpawnGroup(Masked Float=1, Passthrough Float=0.95)
    Mesh Scale Mode=::NewEnumerator3
    Mesh Uniform Scale=&MaskFloatBySpawnGroup(Masked Float=&RandomRangeFloat(Minimum=0.5, Maximum=0.7), Passthrough Float=1)
    UsePositionOffset=true
    Write Material Random=true
   )
   &InitialMeshOrientation(
    Mesh Orientation Mode=::NewEnumerator2
    Rotation=&MaskVectorBySpawnGroup(
     Masked Vector=&RandomRangeVector(Minimum=<0 0 0>, Maximum=<1 1 1>)
     Passthrough Vector=&RandomRangeVector(Minimum=<0 -0.17 -0.2>, Maximum=<0 -0.17 0.2>)
    )
    Use Rotation Vector=true
   )
  ]
  particleUpdateScript [
   &ParticleState()
   &UpdateMeshOrientation(
    Rotation Rate=&MaskFloatBySpawnGroup(Masked Float=0.25, Passthrough Float=0.8)
    Rotation Vector=&MaskVectorBySpawnGroup(
     Masked Vector=&RandomRangeVector(Minimum=<0 1 -0.1>, Maximum=<1 0 0.1>)
     Passthrough Vector=&RandomRangeVector(Minimum=<0 0 -1>, Maximum=<0 0 1>)
    )
   )
   &UpdateMeshOrientation(
    Orientation Method=::NewEnumerator6
    Particle Radius Scale=500
    Rotation Rate=&Multiply_Float(A=$System.RotByVelo, B=0)
    Surface Normal:Vector3f=<0 1 0>
   ) {module=UpdateMeshOrientation002}
   &ScaleMeshSize(Scale Factor=&MaskVectorBySpawnGroup(
     Masked Vector=&VectorFromCurve(
      VectorCurve=DataInterface() {propertyValues {
        XCurve {
         defaultValue=1
         keys [
          {arriveTangent=3.7625, interpMode=RCIM_Cubic, leaveTangent=3.7625, tangentMode=RCTM_User, time=0, value=0}
          {arriveTangent=0.2212, interpMode=RCIM_Cubic, leaveTangent=0.2212, tangentMode=RCTM_User, time=0.5, value=1}
         ]
        }
        YCurve {
         defaultValue=1
         keys [
          {arriveTangent=3.7625, interpMode=RCIM_Cubic, leaveTangent=3.7625, tangentMode=RCTM_User, time=0, value=0}
          {arriveTangent=0.2212, interpMode=RCIM_Cubic, leaveTangent=0.2212, tangentMode=RCTM_User, time=0.5, value=1}
         ]
        }
        ZCurve {defaultValue=1, keys [{arriveTangent=0.2212, interpMode=RCIM_Cubic, leaveTangent=0.2212, tangentMode=RCTM_User, time=0.5, value=1}]}
       }}
      Scale Curve=&Multiply_VectorByFloat(
       Vector=&RandomRangeVector(Minimum=<0.7 0.7 0.7>, Maximum=<1 1.1 1>)
       Float=$System.Timeline001.InverseNormalizedTime
      )
     )
     Passthrough Vector=&VectorFromCurve(
      VectorCurve=DataInterface() {propertyValues {
        XCurve {
         defaultValue=1
         keys [
          {arriveTangent=3.7625, interpMode=RCIM_Cubic, leaveTangent=3.7625, tangentMode=RCTM_User, time=0, value=0.35}
          {arriveTangent=0.2212, interpMode=RCIM_Cubic, leaveTangent=0.2212, tangentMode=RCTM_User, time=0.5, value=1}
         ]
        }
        YCurve {
         defaultValue=1
         keys [
          {arriveTangent=3.7625, interpMode=RCIM_Cubic, leaveTangent=3.7625, tangentMode=RCTM_User, time=0, value=0.35}
          {arriveTangent=0.2212, interpMode=RCIM_Cubic, leaveTangent=0.2212, tangentMode=RCTM_User, time=0.5, value=1}
         ]
        }
        ZCurve {
         defaultValue=1
         keys [<at 0 0.35 RCIM_Cubic>, {arriveTangent=0.2212, interpMode=RCIM_Cubic, leaveTangent=0.2212, tangentMode=RCTM_User, time=0.5, value=1}]
        }
       }}
      Scale Curve=&Multiply_VectorByFloat(
       Vector=&RandomRangeVector(Minimum=<0.8 0.8 1>, Maximum=<1 1 1.25>)
       Float=$System.Timeline002.InverseNormalizedTime
      )
     )
    ))
   &DynamicMaterialParameters(
    Index 0 Param 1=&RandomRangeFloat(Minimum=-0.2, Maximum=0.2)
    Index 0 Param 2=&RandomRangeFloat(Minimum=0, Maximum=1)
    Index 0 Param 3=&MaskFloatBySpawnGroup(
     Masked Float=&RandomRangeFloat(Minimum=16, Maximum=22)
     Passthrough Float=&RandomRangeFloat(Minimum=2, Maximum=5)
    )
    Index 0 Param 4=1
    Index 2 Param 1=&MakeFloatFromVector(
     Vector=&AddVectorToPosition(
      Position=$Particles.Position
      Vector=&TransformVector(
       Vector=&FN_FrontEndVector3Modifier(Gameplay=$System.IGPOFix, Front End=$System.FEPoFix)
       Destination Space=::Simulation
      )
     )
     Channel=::NewEnumerator0
    )
    Index 2 Param 2=&MakeFloatFromVector(
     Vector=&AddVectorToPosition(
      Position=$Particles.Position
      Vector=&TransformVector(
       Vector=&FN_FrontEndVector3Modifier(Gameplay=$System.IGPOFix, Front End=$System.FEPoFix)
       Destination Space=::Simulation
      )
     )
     Channel=::NewEnumerator1
    )
    Index 2 Param 3=&MakeFloatFromVector(
     Vector=&AddVectorToPosition(
      Position=$Particles.Position
      Vector=&TransformVector(
       Vector=&FN_FrontEndVector3Modifier(Gameplay=$System.IGPOFix, Front End=$System.FEPoFix)
       Destination Space=::Simulation
      )
     )
     Channel=::NewEnumerator2
    )
    Index 3 Param 1=&FloatFromCurve(FloatCurve=DataInterface() {propertyValues {Curve {keys [<at 0 0 RCIM_Cubic>, {arriveTangent=502.1832, interpMode=RCIM_Cubic, leaveTangent=502.1832, tangentMode=RCTM_User, time=2, value=800}]}}})
    Module.Index0Float0Write=true
    Module.Index0Float1Write=true
    Module.Index0Float2Write=true
    Module.Index0Float3Write=true
    Module.Index2Float0Write=true
    Module.Index2Float1Write=true
    Module.Index2Float2Write=true
    Module.Index2Float3Write=true
    Module.Index3Float0Write=true
    Module.Index3Float1Write=false
    Module.Index3Float2Write=false
    Module.Index3Float3Write=true
    Write Parameter Index 2=::NewEnumerator1
    Write Parameter Index 3=::NewEnumerator1
   ) {module=DynamicMaterialParameters003}
   &InterpolateOverTime(
    Rate Of Change:NiagaraFloat=&MaskFloatBySpawnGroup(Masked Float=20, Passthrough Float=10)
    Target Vector Value:Vector3f=&AddVectorToPosition(
     Position=$Particles.Position
     Vector=&TransformVector(
      Vector=&FN_FrontEndVector3Modifier(Gameplay=$System.IGPOFix, Front End=$System.FEPoFix)
      Destination Space=::Simulation
     )
    )
    Variable Type=::NewEnumerator2
   )
   /SetVariables(Particles.InterpolateDynamic:Vector4f=&MakeFloat4FromVectorAndFloat(
     "Vector (RGB)"=$StackContext.InterpolateOverTime.InterpolateOverTime_V3_.Moving Average
     "Float (Alpha)"=&MaskFloatBySpawnGroup(Masked Float=20, Passthrough Float=10)
    )) {isSetParameters=true}
   &InheritSourceMovement(Applied Position Falloff Distance=1, ApplyForce=false, ApplyPosition=true, ApplyPositionFalloff=true, ApplyPositionFalloffCurve=false)
   &InheritVelocity(Inherited Velocity Amount Scale=<0.001 0.001 0.001>, Inherited Velocity Speed Limit=1, Speed Limit=true, UseSpeedThreshold=false) {module=InheritVelocity001}
   &SolveForcesAndVelocity(Clamp Velocity=false, Limit Acceleration=false)
   &ScaleColor(
    Scale Alpha=&FloatFromCurve(
     FloatCurve=DataInterface() {propertyValues {Curve {
        defaultValue=1
        keys [
         {arriveTangent=0.171, interpMode=RCIM_Cubic, leaveTangent=0.171, tangentMode=RCTM_User, time=0, value=0}
         {arriveTangent=-0.2135, interpMode=RCIM_Cubic, leaveTangent=-0.2135, tangentMode=RCTM_User, time=0.1408, value=1.2}
         {arriveTangent=-2.0687, interpMode=RCIM_Cubic, leaveTangent=-2.0687, time=0.4, value=1}
         {arriveTangent=-1.6667, interpMode=RCIM_Cubic, leaveTangent=-1.6667, time=0.6, value=0.25}
         {arriveTangent=-0.0405, interpMode=RCIM_Cubic, leaveTangent=-0.0405, tangentMode=RCTM_User, time=1, value=0}
        ]
       }}}
     Scale Curve=&MaskFloatBySpawnGroup(Masked Float=2, Passthrough Float=1)
    )
    Scale RGB=&MakeVectorFromLinearColorRGB(Color=&FNTimeOfDayColor(Night Color=<rgba 0.3 0.475 1 1>, Morning Color=<rgba 1 1 1 1>, Day Color=<rgba 1 1 1 1>, Evening Color=<rgba 1 1 1 1>))
    ScaleA=true
    ScaleRGB=true
   )
   /SetVariables(Particles.Position=&AddVectorToPosition(
     Position=$Engine.Emitter.SimulationPosition
     Vector=&MakeVector(
      X=0
      Y=0
      Z=&FloatFromCurve(
       FloatCurve=DataInterface() {propertyValues {Curve {keys [
           <at 0 0.0197>
           {arriveTangent=0.0896, interpMode=RCIM_Cubic, leaveTangent=0.0896, time=0.178, value=0.12}
           {arriveTangent=0.4944, interpMode=RCIM_Cubic, leaveTangent=0.4944, time=0.4154, value=0.0569}
           {arriveTangent=0.0626, interpMode=RCIM_Cubic, leaveTangent=0.0626, time=0.773, value=0.4141}
           {arriveTangent=-0.0593, interpMode=RCIM_Cubic, leaveTangent=-0.0593, time=1.1037, value=0.1}
           {arriveTangent=-1.2076, interpMode=RCIM_Cubic, leaveTangent=-1.2076, tangentMode=RCTM_User, time=1.4772, value=0.3724}
           {arriveTangent=-0.5153, interpMode=RCIM_Cubic, leaveTangent=-0.5153, tangentMode=RCTM_User, time=1.9788, value=0.1758}
           <at 2.5 0>
          ]}}}
       CurveIndex=$Engine.System.Age
       Scale Curve=80
      )
     )
    )) {isSetParameters=true}
  ]
  rendererClasses [&NiagaraMeshRendererProperties]
  renderers [{rendererClass=&NiagaraMeshRendererProperties, rendererIndex=0}]
 }
 Wings001 {
  emitterUpdateScript [
   &EmitterState(Life Cycle Mode=::NewEnumerator1, Loop Behavior=::NewEnumerator1, Loop Delay=0.1, Loop Duration=500, UseLoopDelay=false)
   &SpawnBurst_Instantaneous(Use Spawn Probability=false, UseLoopCountLimit=false)
   /SetVariables(Emitter.Fade:NiagaraFloat=&FN_FrontEndFloatModifier(Gameplay=0, Front End=1)) {isSetParameters=true}
   /SetVariables(
    Emitter.Inside Color:LinearColor=&MakeLinearColorFromVectorAndFloat(
     "Vector (RGB)"=&MakeVectorFromLinearColorRGB(Color=&Multiply_LinearColorByFloat(
       Linear Color=&MakeCustomLinearColorFromBool(Bool=$User.Ascended, True Color=<rgba 0.2638 0.0005 0.8177 1>, False Color=<rgba 0.2135 0 0.0021 8>)
       Float=$System.Timeline.InverseNormalizedTime
      ))
     "Float (Alpha)"=8
    )
    Emitter.Outer Color:LinearColor=&MakeLinearColorFromVectorAndFloat(
     "Vector (RGB)"=&MakeVectorFromLinearColorRGB(Color=&Multiply_LinearColorByFloat(
       Linear Color=&MakeCustomLinearColorFromBool(Bool=$User.Ascended, True Color=<rgba 0.2705 0 1 1>, False Color=<rgba 0.7 0 0.007 1>)
       Float=$System.Timeline.InverseNormalizedTime
      ))
     "Float (Alpha)"=8
    )
    Emitter.PositionFix:NiagaraFloat=&FN_FrontEndFloatModifier(Gameplay=1, Front End=1)
   ) {isSetParameters=true, module=SetVariables_#6}
   /SetVariables(Emitter.Flapping_Scale:NiagaraFloat=&Add_Float(
     A=&FloatFromCurve(
      FloatCurve=DataInterface() {propertyValues {Curve {
         defaultValue=1
         keys [
          {arriveTangent=86.0893, interpMode=RCIM_Cubic, leaveTangent=86.0893, tangentMode=RCTM_User, time=0, value=5}
          {arriveTangent=-2.5249, leaveTangent=-2.5249, time=0.2, value=10}
          {arriveTangent=-3.9811, interpMode=RCIM_Cubic, leaveTangent=-3.9811, tangentMode=RCTM_User, time=1.6, value=10}
          {arriveTangent=-3.3743, interpMode=RCIM_Cubic, leaveTangent=-3.3743, tangentMode=RCTM_User, time=2, value=1.5}
          <at 2.5 0 RCIM_Cubic>
         ]
        }}}
      CurveIndex=$Emitter.Age
      Scale Curve=0.52
     )
     B=&Multiply_Float(A=$System.Timeline002.NormalizedTime, B=11)
    )) {isSetParameters=true}
  ]
  particleSpawnScript [
   &InitializeParticle(Color=<rgba 1 1 1 1.5>, Mesh Scale=<2.75 2.45 2.1>, Mesh Scale Mode=::NewEnumerator4, UsePositionOffset=true, Write Material Random=true)
   &InitialMeshOrientation(Mesh Axis To Orient=<0 -5 0>, Mesh Orientation Mode=::NewEnumerator2, Use Rotation Vector=true) {module=InitialMeshOrientation002}
   &CameraOffset(Camera Offset Amount=70)
   &AddVelocity(Velocity=<1 0 0>, Velocity Speed Scale=0.05)
  ]
  particleUpdateScript [
   &ParticleState(Kill Particles When Lifetime Has Elapsed=false, Let Infinitely Lived Particles Die When Emitter Deactivates=true)
   &UpdateMeshOrientation(
    Bank Rate=30
    Banking Angle Min / Max:Vector2f=<-8 8>
    "Foward Vector (Mesh Space)":Vector3f=<0 -1 0>
    Initial Forward Vector:Vector3f=<1 0 -0.4>
    Orientation Method=::NewEnumerator3
    Upright Roll Correction Speed=10
   )
   &UpdateMeshOrientation(
    Orientation Method=::NewEnumerator6
    Particle Radius Scale=500
    Rotation Rate=&Multiply_Float(A=$System.RotByVelo, B=0.0009)
    Surface Normal:Vector3f=<1 0 0>
   ) {module=UpdateMeshOrientation001}
   &ScaleMeshSize(Scale Factor=&VectorFromCurve(
     VectorCurve=DataInterface() {propertyValues {
       XCurve {
        defaultValue=1
        keys [
         {arriveTangent=0.3844, interpMode=RCIM_Cubic, leaveTangent=0.3844, tangentMode=RCTM_User, time=0, value=0}
         {arriveTangent=2.8556, interpMode=RCIM_Cubic, leaveTangent=2.8556, tangentMode=RCTM_User, time=0.3, value=0.25}
         {arriveTangent=0.01, interpMode=RCIM_Cubic, leaveTangent=0.01, tangentMode=RCTM_User, time=1.2, value=1}
        ]
       }
       YCurve {
        defaultValue=1
        keys [
         {arriveTangent=0.3844, interpMode=RCIM_Cubic, leaveTangent=0.3844, tangentMode=RCTM_User, time=0, value=0}
         {arriveTangent=2.8556, interpMode=RCIM_Cubic, leaveTangent=2.8556, tangentMode=RCTM_User, time=0.3, value=0.25}
         {arriveTangent=0.01, interpMode=RCIM_Cubic, leaveTangent=0.01, tangentMode=RCTM_User, time=1.2, value=1}
        ]
       }
       ZCurve {
        defaultValue=1
        keys [<at 0 0.1 RCIM_Cubic>, {arriveTangent=0.01, interpMode=RCIM_Cubic, leaveTangent=0.01, tangentMode=RCTM_User, time=1.2, value=1}]
       }
      }}
     CurveIndex=$Emitter.Age
     Scale Curve=&Multiply_VectorByFloat(
      Vector=&Multiply_VectorByFloat(
       Vector=<2 2.5 0.7>
       Float=/Niagara/DynamicInputs/Camera/CameraFade(
        Camera Query:NiagaraDataInterfaceCamera=DataInterface()
        Fade Offset:NiagaraFloat=-250
        Fade Distance:NiagaraFloat=1000
        Use Owner Position:NiagaraBool=true
       )
      )
      Float=$System.Timeline002.InverseNormalizedTime
     )
    ))
   &DynamicMaterialParameters(
    Index 0 Param 1=&RandomRangeFloat(Minimum=0.6, Maximum=0.9)
    Index 0 Param 2=&RandomRangeFloat(Minimum=0.5, Maximum=1)
    Index 0 Param 3=&Multiply_Float(
     A=5
     B=&Lerp_Float(A=0.75, B=1.5, Alpha=&RandomRangeFloat(Minimum=0, Maximum=1))
    )
    Index 0 Param 4=1
    Index 1 Param 1=$System.YVelo
    Index 1 Param 4=&FloatFromCurve(
     FloatCurve=DataInterface() {propertyValues {Curve {keys [{arriveTangent=-1.1407, interpMode=RCIM_Cubic, leaveTangent=-1.1407, tangentMode=RCTM_User, time=0.8, value=1}, <at 2 0.25 RCIM_Cubic>]}}}
     CurveIndex=$Emitter.Age
    )
    Index 2 Param 1=&MakeFloatFromVector(
     Vector=&AddVectorToPosition(
      Position=$Particles.Position
      Vector=&TransformVector(Vector=$System.IGPOFix, Destination Space=::Simulation)
     )
     Channel=::NewEnumerator0
    )
    Index 2 Param 2=&MakeFloatFromVector(
     Vector=&AddVectorToPosition(
      Position=$Particles.Position
      Vector=&TransformVector(Vector=$System.IGPOFix, Destination Space=::Simulation)
     )
     Channel=::NewEnumerator1
    )
    Index 2 Param 3=&MakeFloatFromVector(
     Vector=&AddVectorToPosition(
      Position=$Particles.Position
      Vector=&TransformVector(Vector=$System.IGPOFix, Destination Space=::Simulation)
     )
     Channel=::NewEnumerator2
    )
    Index 2 Param 4=&VectorLength(Vector=$Engine.Owner.Velocity)
    Index 3 Param 1=&FloatFromCurve(FloatCurve=DataInterface() {propertyValues {Curve {keys [<at 0 0 RCIM_Cubic>, {arriveTangent=502.1832, interpMode=RCIM_Cubic, leaveTangent=502.1832, tangentMode=RCTM_User, time=2, value=800}]}}})
    Index 3 Param 2=$Output.UpdateMeshOrientation.CurrentBankValue
    Index 3 Param 3=&Add_Float(A=$Emitter.Age, B=0.5)
    Module.Param0WriteEnabled=true
    Module.Param0WriteEnabled1=true
    Module.Param0WriteEnabled2=false
    Module.Param0WriteEnabled3=true
    Module.Param1WriteEnabled=true
    Module.Param1WriteEnabled1=false
    Module.Param1WriteEnabled2=false
    Module.Param1WriteEnabled3=true
    Module.Param2WriteEnabled=true
    Module.Param2WriteEnabled1=false
    Module.Param2WriteEnabled2=false
    Module.Param2WriteEnabled3=true
    Module.Param3WriteEnabled=true
    Module.Param3WriteEnabled1=true
    Module.Param3WriteEnabled2=true
    Module.Param3WriteEnabled3=true
    Write Parameter Index 1=true
    Write Parameter Index 2=true
    Write Parameter Index 3=true
   ) {module=DynamicMaterialParameters003}
   &InheritVelocity(Inherited Velocity Amount Scale=<0.0001 0.0001 0.0001>, Inherited Velocity Speed Limit=0.1, Speed Limit=true, UseSpeedThreshold=false) {module=InheritVelocity001}
   &SolveForcesAndVelocity(Clamp Velocity=false, Limit Acceleration=false)
   &ScaleColor(Scale Alpha=$User.Is Air, ScaleA=true, ScaleRGB=false)
   /SetVariables(Particles.Position=&AddVectorToPosition(
     Position=$Engine.Emitter.SimulationPosition
     Vector=&MakeVector(
      X=0
      Y=0
      Z=&FloatFromCurve(
       FloatCurve=DataInterface() {propertyValues {Curve {keys [
           <at 0 0.0197>
           {arriveTangent=0.0896, interpMode=RCIM_Cubic, leaveTangent=0.0896, time=0.178, value=0.12}
           {arriveTangent=1.0589, interpMode=RCIM_Cubic, leaveTangent=1.0589, time=0.4154, value=0.0569}
           {arriveTangent=0.0626, interpMode=RCIM_Cubic, leaveTangent=0.0626, time=0.773, value=0.75}
           {arriveTangent=-0.142, interpMode=RCIM_Cubic, leaveTangent=-0.142, time=1.1037, value=0.1}
           {arriveTangent=-1.2076, interpMode=RCIM_Cubic, leaveTangent=-1.2076, tangentMode=RCTM_User, time=1.4772, value=0.65}
           {arriveTangent=-0.5153, interpMode=RCIM_Cubic, leaveTangent=-0.5153, tangentMode=RCTM_User, time=2, value=0.15}
           <at 2.5 0>
          ]}}}
       CurveIndex=$Engine.System.Age
       Scale Curve=80
      )
     )
    )) {isSetParameters=true}
  ]
  rendererClasses [&NiagaraMeshRendererProperties]
  renderers [
   {rendererClass=&NiagaraMeshRendererProperties, rendererIndex=0}
   {rendererClass=&NiagaraMeshRendererProperties, rendererIndex=1}
  ]
 }
}
summary {
 systemName=NS_MistRaven_WraithForm_Active_Loop
 emitters [
  {emitterName=Em_Smoke_Cloud_Detail, rendererClasses [&NiagaraSpriteRendererProperties]}
  {emitterName=Em_Ravens_Ball, rendererClasses [&NiagaraMeshRendererProperties]}
  {emitterName=LW_Smoke_Mesh_Rotating, rendererClasses []}
  {emitterName=LW_Smoke_Cloud, rendererClasses []}
  {emitterName=LW_Feather, rendererClasses []}
  {emitterName=Ribbon_Trail, rendererClasses [&NiagaraRibbonRendererProperties]}
  {emitterName=Trail_Edge004, rendererClasses [&NiagaraMeshRendererProperties]}
  {emitterName=Wings001, rendererClasses [&NiagaraMeshRendererProperties]}
  {emitterName=LW_Feather001, rendererClasses []}
 ]
 userVariables [
  User.Ascended {type=NiagaraBool, defaultValue=true}
  User.bIsFrontEnd {type=NiagaraBool}
  User.bIsFrontEndPreview {type=NiagaraBool}
  User.Finish Signal Duration {type=NiagaraFloat, defaultValue=1.15, description=""}
  User.Is Air {type=NiagaraFloat, defaultValue=1, description=""}
  User.Looping Duration {type=NiagaraFloat, defaultValue=4, description=""}
 ]
}
